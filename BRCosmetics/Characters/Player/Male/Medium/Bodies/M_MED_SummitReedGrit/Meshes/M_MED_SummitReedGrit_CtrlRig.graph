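# Control Rig for the SummitReedGrit male medium cosmetic body, adding procedural corrective deformation for the body, thigh straps, logo and feet.
# A shared DeformRig function runs first, then a sequence caches Leg ROM outputs into variables for later steps.
# Parent and aim constraints blend strap and foot helper bones between pelvis, leg and deform bones.
# Spherical pose readers, leg ROM values, foot rotation, pelvis height and crouch/DBNO flags weight additional transform offsets.
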ControlRigBlueprint /BRCosmetics/Characters/Player/Male/Medium/Bodies/M_MED_SummitReedGrit/Meshes/M_MED_SummitReedGrit_CtrlRig
graphs [
 RigVMFunctionLibrary {links [], nodes []}
 RigVMFunction_Sequence_1_ContainedGraph {nodes [
   n1 = RigVMFunctionEntryNode()
   RigVMFunction_Sequence = &RigVMFunction_Sequence(ExecuteContext=@n1.ExecuteContext)
   n2 = &RigVMFunction_Sequence(ExecuteContext=@RigVMFunction_Sequence.B)
   n3 = &RigVMFunction_Sequence(ExecuteContext=@n2.B)
   n4 = &RigVMFunction_Sequence(ExecuteContext=@n3.B)
   RigVMFunctionReturnNode(A=@RigVMFunction_Sequence.A, B=@n2.A, C=@n3.A, D=@n4.A, E=@n4.B)
  ]}
 RigVMModel {nodes [
   RigUnit_BeginExecution = &RigUnit_BeginExecution()
   &RigUnit_ModifyTransforms(
    ItemToModify="((Item=(Type=Bone,Name=\"thigh_l\"),Transform=(Rotation=(X=0,Y=0,Z=0,W=1),Translation=(X=0,Y=0,Z=0),Scale3D=(X=1,Y=1,Z=1))),(Item=(Type=Bone,Name=\"thigh_r\"),Transform=(Rotation=(X=0,Y=0,Z=0,W=1),Translation=(X=0,Y=0,Z=0),Scale3D=(X=1,Y=1,Z=1))),(Item=(Type=B ...(truncated)"
    Weight=1
    WeightMinimum=0
    ExecutePin=@RigUnit_BeginExecution.ExecutePin
   )
   > FN_DeformRig_FNC = RigVMFunctionReferenceNode(
    DeformRig_ Curve="(deformRig_l_upperarm_up_alpha_72_#1=False,deformRig_r_upperarm_up_alpha_73_#2=False,deformRig_l_upperarm_fr_alpha_74_#3=False,deformRig_r_upperarm_fr_alpha_75_CB51B8 ...(truncated)"
    DeformRig_Bool="(upperarm_1_#4=True,elbow_3_#5=True,wrist_7_#6=True,groin_8_#7=True,glute_17_#8=True,knee_9_40C408324CA261 ...(truncated)"
    DeformRig_Debug=(upperarm_debug_18_#4=False,wrist_debug_20_#6=False,lat_debug_26_#9=False,groinGlute_debug_27_#7=False)
    DeformRig_Value="(upperarm_up_18_#10=(Rotation=(X=0.0,Y=-0.0,Z=0.0,W=1.0),Translation=(X=0.0,Y=0.0,Z=8.0),Scale3D=(X=0.70,Y=1.0,Z=1.0)),upperarm_fwd_fr_19_#11=(Rotation=(X=0.0,Y=-0.0,Z=0.0,W=1.0),Translation=(X=0.0, ...(truncated)"
   )
   > RigVMFunction_Sequence = RigVMAggregateNode()
   Leg ROM = RigVMFunctionReferenceNode(
    Include Ankles=false
    Include Distance Checks=true
    Include Knees=true
    Interp Output Speed=5
    Interp Results=true
    Result Alpha=1
    SPR Optional Parent Item=(Type=Bone,Name="None")
    ExecuteContext=@RigVMFunction_Sequence.A
   )
   > RigVMVariableNode(Value=@"Leg ROM".L Thigh Raised Fwd) {variable=Leg ROM_L Thigh Raised Fwd}
   > RigVMVariableNode(Value=@"Leg ROM".R Thigh Raised Fwd) {variable=Leg ROM_R Thigh Raised Fwd}
   > RigVMVariableNode(Value=@"Leg ROM".L Thigh Raised Side) {variable=Leg ROM_L Thigh Raised Side}
   > RigVMVariableNode(Value=@"Leg ROM".R Thigh Raised Side) {variable=Leg ROM_R Thigh Raised Side}
   > RigVMVariableNode(Value=@"Leg ROM".L Thigh Raised Rear) {variable=Leg ROM_L Thigh Raised Rear}
   > RigVMVariableNode(Value=@"Leg ROM".R Thigh Raised Rear) {variable=Leg ROM_R Thigh Raised Rear}
   > RigVMVariableNode(Value=@"Leg ROM".L Knee) {variable=Leg ROM_L Thigh Fwd Knee Bend}
   > RigVMVariableNode(Value=@"Leg ROM".R Knee) {variable=Leg ROM_R Thigh Fwd Knee Bend}
   > RigVMVariableNode(Value=@"Leg ROM".L Thigh Fwd Knee Bend) {variable=Leg ROM_L Thigh Fwd Knee Bend_0}
   &RigUnit_ParentConstraint(
    AdvancedSettings=(InterpolationType=Average,RotationOrderForFilter=XZY)
    Child=(Type=Bone,Name="MSTR_strape02_fr_l")
    Parents=((Item=(Type=Bone,Name="deform_groin_l"),Weight=0.2),(Item=(Type=Bone,Name="pelvis"),Weight=0.2),(Item=(Type=Bone,Name="thigh_twist_01_l"),Weight=0.6))
    Weight=1
    ExecutePin=@RigVMFunction_Sequence.B
   )
   > &RigUnit_ParentConstraint(
    AdvancedSettings=(InterpolationType=Average,RotationOrderForFilter=XZY)
    Child=(Type=Bone,Name="MSTR_strape02_bk_l")
    Parents=((Item=(Type=Bone,Name="pelvis"),Weight=0.3),(Item=(Type=Bone,Name="deform_glute_l"),Weight=0.23),(Item=(Type=Bone,Name="thigh_twist_01_l"),Weight=0.44))
    Weight=1
   )
   > &RigUnit_ParentConstraint(
    AdvancedSettings=(InterpolationType=Average,RotationOrderForFilter=XZY)
    Child=(Type=Bone,Name="MSTR_strape02_fr_r")
    Parents=((Item=(Type=Bone,Name="deform_groin_r"),Weight=0.2),(Item=(Type=Bone,Name="pelvis"),Weight=0.2),(Item=(Type=Bone,Name="thigh_twist_01_r"),Weight=0.6))
    Weight=1
   )
   > &RigUnit_ParentConstraint(
    AdvancedSettings=(InterpolationType=Average,RotationOrderForFilter=XZY)
    Child=(Type=Bone,Name="MSTR_strape02_bk_r")
    Parents=((Item=(Type=Bone,Name="pelvis"),Weight=0.3),(Item=(Type=Bone,Name="deform_glute_r"),Weight=0.23),(Item=(Type=Bone,Name="thigh_twist_01_r"),Weight=0.44))
    Weight=1
   )
   > &RigUnit_ParentConstraint(
    AdvancedSettings=(InterpolationType=Average,RotationOrderForFilter=XZY)
    Child=(Type=Bone,Name="def_foot_fr_l")
    Parents=((Item=(Type=Bone,Name="foot_l"),Weight=0.5),(Item=(Type=Bone,Name="calf_twist_01_l"),Weight=1))
    Weight=1
   )
   > &RigUnit_ParentConstraint(
    AdvancedSettings=(InterpolationType=Average,RotationOrderForFilter=XZY)
    Child=(Type=Bone,Name="def_foot_fr_r")
    Parents=((Item=(Type=Bone,Name="foot_r"),Weight=0.5),(Item=(Type=Bone,Name="calf_twist_01_r"),Weight=1))
    Weight=1
   )
   Add = &RigVMFunction_MathDoubleAdd(
    A=&RigVMFunction_MathDoubleRemap(SourceMaximum=1, SourceMinimum=0.25, TargetMaximum=1, TargetMinimum=0, Value=RigVMVariableNode() {variable=Leg ROM_L Thigh Raised Fwd}.Value).Result
    B=&RigVMFunction_MathDoubleRemap(SourceMaximum=1, SourceMinimum=0.25, TargetMaximum=1, TargetMinimum=0, Value=RigVMVariableNode() {variable=Leg ROM_R Thigh Raised Fwd}.Value).Result
   )
   Multiply = &RigVMFunction_MathDoubleMul(B=0.5, A=@Add.Result)
   &RigUnit_ModifyTransforms(
    ItemToModify=((Item=(Type=Bone,Name="def_logo"),Transform=(Rotation=(X=0,Y=0,Z=0,W=1),Translation=(X=1,Y=0,Z=7),Scale3D=(X=1,Y=1,Z=1))))
    WeightMinimum=0
    Weight=@Multiply.Result
    ExecutePin=@RigVMFunction_Sequence.C
   )
   Remap = &RigVMFunction_MathDoubleRemap(SourceMaximum=1, SourceMinimum=0.5, TargetMaximum=1, TargetMinimum=0, Value=RigVMVariableNode() {variable=Leg ROM_L Thigh Raised Fwd}.Value)
   > &RigUnit_ModifyTransforms(
    ItemToModify=((Item=(Type=Bone,Name="def_logo"),Transform=(Rotation=(X=-0.3007,Y=0,Z=0,W=0.9537),Translation=(X=0,Y=0,Z=0),Scale3D=(X=1,Y=1,Z=1))))
    WeightMinimum=0
    Weight=@Remap.Result
   )
   > &RigUnit_ModifyTransforms(
    ItemToModify=((Item=(Type=Bone,Name="def_logo"),Transform=(Rotation=(X=0.3007,Y=0,Z=0,W=0.9537),Translation=(X=0,Y=0,Z=0),Scale3D=(X=1,Y=1,Z=1))))
    WeightMinimum=0
    Weight=&RigVMFunction_MathDoubleRemap(SourceMaximum=1, SourceMinimum=0.5, TargetMaximum=1, TargetMinimum=0, Value=RigVMVariableNode() {variable=Leg ROM_R Thigh Raised Fwd}.Value).Result
   )
   Cast = &RigVMFunction_MathBoolToFloat(Value=RigVMVariableNode() {variable=isDBNO}.Value)
   > &RigUnit_ModifyTransforms(
    ItemToModify=((Item=(Type=Bone,Name="def_logo"),Transform=(Rotation=(X=0,Y=0,Z=0,W=1),Translation=(X=0,Y=0,Z=0.5),Scale3D=(X=1,Y=0.8,Z=1))))
    WeightMinimum=0
    Weight=@Cast.Result
   )
   > n1 = &RigUnit_SphericalPoseReader(
    ActiveRegionScaleFactors=(PositiveWidth=1,NegativeWidth=1,PositiveHeight=1,NegativeHeight=1)
    ActiveRegionSize=0.1
    Debug=(bDrawDebug=false,bDraw2D=False,bDrawLocalAxes=False,DebugScale=25,DebugSegments=20,DebugThickness=0.25)
    DriverAxis=(X=-1,Y=0,Z=0)
    DriverItem=(Type=Bone,Name="thigh_l")
    FalloffRegionScaleFactors=(PositiveWidth=1,NegativeWidth=1,PositiveHeight=1,NegativeHeight=1)
    FalloffSize=0.25
    FlipHeightScaling=False
    FlipWidthScaling=False
    OptionalParentItem=(Type=Bone,Name="None")
    RotationOffset=(X=165,Y=0,Z=0)
   )
   > n2 = &RigUnit_SphericalPoseReader(
    ActiveRegionScaleFactors=(PositiveWidth=1,NegativeWidth=1,PositiveHeight=1,NegativeHeight=1)
    ActiveRegionSize=0.1
    Debug=(bDrawDebug=false,bDraw2D=False,bDrawLocalAxes=False,DebugScale=25,DebugSegments=20,DebugThickness=0.25)
    DriverAxis=(X=1,Y=0,Z=0)
    DriverItem=(Type=Bone,Name="thigh_r")
    FalloffRegionScaleFactors=(PositiveWidth=1,NegativeWidth=1,PositiveHeight=1,NegativeHeight=1)
    FalloffSize=0.25
    FlipHeightScaling=False
    FlipWidthScaling=False
    OptionalParentItem=(Type=Bone,Name="None")
    RotationOffset=(X=-20,Y=0,Z=0)
   )
   > &RigUnit_ModifyTransforms(
    ItemToModify=((Item=(Type=Bone,Name="def_logo"),Transform=(Rotation=(X=0,Y=0,Z=0,W=1),Translation=(X=1,Y=0,Z=4),Scale3D=(X=1,Y=1,Z=1))))
    WeightMinimum=0
    Weight=&RigVMFunction_MathDoubleMul(A=@n1.OutputParam, B=@n2.OutputParam).Result
   )
   RigVMFunction_MathQuaternionToEuler = &RigVMFunction_MathQuaternionToEuler(
    RotationOrder=XYZ
    Value=&RigUnit_GetTransform(Item=(Type=Bone,Name="foot_l"), Space=LocalSpace, bInitial=False).Transform.Rotation
   )
   > &RigUnit_ModifyTransforms(
    ItemToModify=((Item=(Type=Bone,Name="def_foot_bk_l"),Transform=(Rotation=(X=0,Y=-0.3827,Z=0,W=0.9239),Translation=(X=3.475,Y=0,Z=5.9217),Scale3D=(X=1,Y=1,Z=1))))
    WeightMinimum=0
    Weight=&RigVMFunction_MathDoubleRemap(SourceMaximum=75, SourceMinimum=0, TargetMaximum=1, TargetMinimum=0, Value=@RigVMFunction_MathQuaternionToEuler.Result.Z).Result
   )
   n3 = &RigUnit_GetTransform(Item=(Type=Bone,Name="foot_r"), Space=LocalSpace, bInitial=False)
   > &RigUnit_ModifyTransforms(
    ItemToModify=((Item=(Type=Bone,Name="def_foot_bk_r"),Transform=(Rotation=(X=0,Y=-0.3827,Z=0,W=0.9239),Translation=(X=-3.4,Y=0,Z=-6),Scale3D=(X=1,Y=1,Z=0.5))))
    WeightMinimum=0
    Weight=&RigVMFunction_MathDoubleRemap(
     SourceMaximum=75
     SourceMinimum=0
     TargetMaximum=1
     TargetMinimum=0
     Value=&RigVMFunction_MathQuaternionToEuler(RotationOrder=XYZ, Value=@n3.Transform.Rotation).Result.Z
    ).Result
   )
   &RigUnit_AimConstraintLocalSpaceOffset(
    Child=(Type=Bone,Name="MSTR_strape02_fr_l")
    Filter=(bX=true,bY=true,bZ=true)
    Parents=((Item=(Type=Bone,Name="deform_groin_l"),Weight=0),(Item=(Type=Bone,Name="pelvis"),Weight=0),(Item=(Type=Bone,Name="thigh_twist_01_l"),Weight=1))
    Weight=1
    WorldUp=(Target=(X=0,Y=0,Z=1),Kind=Location,Space=(Type=Bone,Name="thigh_l"))
    bMaintainOffset=true
    ExecutePin=@RigVMFunction_Sequence.D
   )
   > &RigUnit_AimConstraintLocalSpaceOffset(
    Child=(Type=Bone,Name="MSTR_strape02_bk_l")
    Filter=(bX=true,bY=true,bZ=true)
    Parents=((Item=(Type=Bone,Name="deform_glute_l"),Weight=0),(Item=(Type=Bone,Name="pelvis"),Weight=0),(Item=(Type=Bone,Name="thigh_twist_01_l"),Weight=1))
    Weight=1
    WorldUp=(Target=(X=0,Y=0,Z=1),Kind=Location,Space=(Type=Bone,Name="thigh_l"))
   )
   > &RigUnit_AimConstraintLocalSpaceOffset(
    Child=(Type=Bone,Name="MSTR_strape02_fr_r")
    Filter=(bX=true,bY=true,bZ=true)
    Parents=((Item=(Type=Bone,Name="deform_groin_r"),Weight=0),(Item=(Type=Bone,Name="pelvis"),Weight=0),(Item=(Type=Bone,Name="thigh_twist_01_r"),Weight=1))
    Weight=1
    WorldUp=(Target=(X=0,Y=0,Z=1),Kind=Location,Space=(Type=Bone,Name="thigh_r"))
    bMaintainOffset=true
   )
   > &RigUnit_AimConstraintLocalSpaceOffset(
    Child=(Type=Bone,Name="MSTR_strape02_bk_r")
    Filter=(bX=true,bY=true,bZ=true)
    Parents=((Item=(Type=Bone,Name="deform_glute_r"),Weight=0),(Item=(Type=Bone,Name="pelvis"),Weight=0),(Item=(Type=Bone,Name="thigh_twist_01_r"),Weight=1),(Item=(Type=Bone,Name="calf_r"),Weight=0))
    Weight=1
    WorldUp=(Target=(X=0,Y=0,Z=1),Kind=Location,Space=(Type=Bone,Name="thigh_r"))
   )
   n4 = &RigUnit_SphericalPoseReader(
    ActiveRegionScaleFactors=(PositiveWidth=3.5,NegativeWidth=0.5,PositiveHeight=3,NegativeHeight=3)
    ActiveRegionSize=0.1
    Debug=(bDrawDebug=false,DebugScale=25,DebugSegments=20,DebugThickness=0.25)
    DriverAxis=(X=-1,Y=0,Z=0)
    DriverItem=(Type=Bone,Name="thigh_l")
    FalloffRegionScaleFactors=(PositiveWidth=1,NegativeWidth=1,PositiveHeight=1,NegativeHeight=1)
    FalloffSize=0.8
    OptionalParentItem=(Type=Bone,Name="spine_03")
    RotationOffset=(X=110,Y=0,Z=-30)
    ExecutePin=@RigVMFunction_Sequence.E
   )
   n5 = RigVMVariableNode() {variable=isCrouching}
   n6 = &RigVMFunction_MathDoubleRemap(
    SourceMaximum=42
    SourceMinimum=70
    TargetMaximum=1
    TargetMinimum=0
    Value=&RigUnit_GetRelativeTransformForItem(Child=(Type=Bone,Name="pelvis"), Parent=(Type=Bone,Name="root"), bChildInitial=False, bParentInitial=False).RelativeTransform.Translation.Z
   )
   > &RigUnit_ModifyTransforms(
    ItemToModify=((Item=(Type=Bone,Name="MSTR_strape02_bk_l"),Transform=(Rotation=(X=0,Y=0,Z=0,W=1),Translation=(X=-5,Y=-5,Z=0),Scale3D=(X=1,Y=1,Z=1))))
    WeightMinimum=0
    Weight=&RigVMFunction_MathDoubleMul(
     A=&RigVMFunction_MathBoolToFloat(Value=@n5.Value).Result
     B=&RigVMFunction_MathDoubleMul(B=@n6.Result, A=@n4.OutputParam).Result
    ).Result
   )
   > &RigUnit_ModifyTransforms(
    ItemToModify="((Item=(Type=Bone,Name=\"MSTR_strape02_bk_l\"),Transform=(Rotation=(X=0,Y=0,Z=0,W=1),Translation=(X=0,Y=-3,Z=-2),Scale3D=(X=1,Y=1,Z=1))),(Item=(Type=Bone,Name=\"MSTR_strape02_fr_l\"),Transform=(Rotation=(X=0.13,Y=0.0864,Z=-0.0114,W=0.9877),Translation=(X=0,Y=0 ...(truncated)"
    WeightMinimum=0
    Weight=&RigVMFunction_MathDoubleRemap(SourceMaximum=0.4, SourceMinimum=0, TargetMaximum=1, TargetMinimum=0, Value=RigVMVariableNode() {variable=Leg ROM_L Thigh Raised Rear}.Value).Result
   )
   > &RigUnit_ModifyTransforms(
    ItemToModify="((Item=(Type=Bone,Name=\"MSTR_strape02_bk_r\"),Transform=(Rotation=(X=0,Y=0,Z=0,W=1),Translation=(X=0,Y=3,Z=2),Scale3D=(X=1,Y=1,Z=1))),(Item=(Type=Bone,Name=\"MSTR_strape02_fr_r\"),Transform=(Rotation=(X=0,Y=0,Z=0,W=1),Translation=(X=0,Y=0,Z=0),Scale3D=(X=1,Y= ...(truncated)"
    WeightMinimum=0
    Weight=&RigVMFunction_MathDoubleRemap(SourceMaximum=0.4, SourceMinimum=0, TargetMaximum=1, TargetMinimum=0, Value=RigVMVariableNode() {variable=Leg ROM_R Thigh Raised Rear}.Value).Result
   )
   n7 = RigVMVariableNode() {variable=Leg ROM_R Thigh Fwd Knee Bend}
   n8 = &RigUnit_GetRelativeTransformForItem(Child=(Type=Bone,Name="pelvis"), Parent=(Type=Bone,Name="root"), bChildInitial=False, bParentInitial=False)
   n9 = &RigVMFunction_MathDoubleRemap(SourceMaximum=25, SourceMinimum=35, TargetMaximum=0, TargetMinimum=1, Value=@n8.RelativeTransform.Translation.Z)
   > &RigUnit_ModifyTransforms(
    ItemToModify="((Item=(Type=Bone,Name=\"MSTR_strape02_bk_r\"),Transform=(Rotation=(X=-0.3675,Y=-0.1296,Z=0.0871,W=0.9168),Translation=(X=5,Y=-2,Z=-2),Scale3D=(X=1,Y=1,Z=1))),(Item=(Type=Bone,Name=\"MSTR_strape02_fr_r\"),Transform=(Rotation=(X=-0.0868,Y=0.0868,Z=0.0076,W=0.99 ...(truncated)"
    WeightMinimum=0
    Weight=&RigVMFunction_MathDoubleMul(
     B=&RigVMFunction_MathDoubleRemap(SourceMaximum=1, SourceMinimum=0.6, TargetMaximum=1, TargetMinimum=0, Value=@n7.Value).Result
     A=@n9.Result
    ).Result
   )
   n10 = RigVMVariableNode() {variable=Leg ROM_L Thigh Fwd Knee Bend}
   > &RigUnit_ModifyTransforms(
    ItemToModify="((Item=(Type=Bone,Name=\"MSTR_strape02_bk_l\"),Transform=(Rotation=(X=-0.3675,Y=-0.1296,Z=0.0871,W=0.9168),Translation=(X=-5,Y=2,Z=2),Scale3D=(X=1,Y=1,Z=1))),(Item=(Type=Bone,Name=\"MSTR_strape02_fr_l\"),Transform=(Rotation=(X=0.0868,Y=0.0868,Z=-0.0076,W=0.992 ...(truncated)"
    WeightMinimum=0
    Weight=&RigVMFunction_MathDoubleMul(
     B=&RigVMFunction_MathDoubleRemap(SourceMaximum=1, SourceMinimum=0.6, TargetMaximum=1, TargetMinimum=0, Value=@n10.Value).Result
     A=@n9.Result
    ).Result
   )
   > n11 = &RigUnit_SphericalPoseReader(
    ActiveRegionScaleFactors=(PositiveWidth=1,NegativeWidth=1.5,PositiveHeight=1.5,NegativeHeight=1.5)
    ActiveRegionSize=0.1
    Debug=(bDrawDebug=True,bDraw2D=False,bDrawLocalAxes=False,DebugScale=25,DebugSegments=20,DebugThickness=0.25)
    DriverAxis=(X=1,Y=0,Z=0)
    DriverItem=(Type=Bone,Name="thigh_r")
    FalloffRegionScaleFactors=(PositiveWidth=1,NegativeWidth=1,PositiveHeight=1,NegativeHeight=1)
    FalloffSize=0.25
    FlipHeightScaling=False
    FlipWidthScaling=False
    OptionalParentItem=(Type=Bone,Name="None")
    RotationOffset=(X=-60,Y=-35,Z=0)
   )
   > n12 = &RigUnit_SphericalPoseReader(
    ActiveRegionScaleFactors=(PositiveWidth=1,NegativeWidth=1.5,PositiveHeight=1.5,NegativeHeight=1.5)
    ActiveRegionSize=0.1
    Debug=(bDrawDebug=True,bDraw2D=False,bDrawLocalAxes=False,DebugScale=25,DebugSegments=20,DebugThickness=0.25)
    DriverAxis=(X=-1,Y=0,Z=0)
    DriverItem=(Type=Bone,Name="thigh_l")
    FalloffRegionScaleFactors=(PositiveWidth=1,NegativeWidth=1,PositiveHeight=1,NegativeHeight=1)
    FalloffSize=0.25
    FlipHeightScaling=False
    FlipWidthScaling=true
    OptionalParentItem=(Type=Bone,Name="None")
    RotationOffset=(X=120,Y=-35,Z=20)
   )
   > &RigUnit_ModifyTransforms(
    ItemToModify="((Item=(Type=Bone,Name=\"MSTR_strape02_bk_l\"),Transform=(Rotation=(X=0,Y=-0.0436,Z=0,W=0.999),Translation=(X=0,Y=2,Z=0),Scale3D=(X=1,Y=1,Z=1))),(Item=(Type=Bone,Name=\"MSTR_strape02_fr_l\"),Transform=(Rotation=(X=0,Y=0.0436,Z=0,W=0.999),Translation=(X=0,Y=-1, ...(truncated)"
    WeightMinimum=0
    Weight=&RigVMFunction_MathDoubleMul(
     B=&RigVMFunction_MathDoubleMul(A=@n11.OutputParam, B=@n12.OutputParam).Result
     A=&RigVMFunction_MathDoubleRemap(SourceMaximum=11, SourceMinimum=35, TargetMaximum=1, TargetMinimum=0, Value=@n8.RelativeTransform.Translation.Z).Result
    ).Result
   )
   > &RigUnit_ModifyTransforms(
    ItemToModify="((Item=(Type=Bone,Name=\"MSTR_strape02_bk_l\"),Transform=(Rotation=(X=0,Y=0,Z=0,W=1),Translation=(X=0,Y=0,Z=-1),Scale3D=(X=1,Y=1,Z=1))),(Item=(Type=Bone,Name=\"MSTR_strape02_fr_l\"),Transform=(Rotation=(X=-0.13,Y=-0.0114,Z=0.0864,W=0.9877),Translation=(X=1,Y=0 ...(truncated)"
    WeightMinimum=0
    Weight=&RigVMFunction_MathBoolToFloat(Value=RigVMVariableNode() {variable=isDBNO}.Value).Result
   )
   &RigVMFunction_MathDoubleRemap(SourceMaximum=45, SourceMinimum=70, TargetMaximum=1, TargetMinimum=0, Value=@n8.RelativeTransform.Translation.Z)
   RigVMRerouteNode(Value=&RigVMFunction_MathDoubleRemap(SourceMaximum=25, SourceMinimum=35, TargetMaximum=1, TargetMinimum=0, Value=@n8.RelativeTransform.Translation.Z).Result)
  ]}
]
hierarchy [
 root {loc=<0 0 0>, rot=<0 0 0 1>}
 pelvis {parent=root, loc=<0 -0.7071 93.1191>, rot=<0.0127 0.707 -0.0127 -0.707>}
 spine_01 {parent=pelvis, loc=<6.7494 -0.6388 0>, rot=<0 0 0.034 -0.9994>}
 spine_02 {parent=spine_01, loc=<7.2518 -0.035 0>, rot=<0 0 0.0238 -0.9997>}
 spine_03 {parent=spine_02, loc=<6.0104 0 0>, rot=<0 0 0.0149 -0.9999>}
 spine_04 {parent=spine_03, loc=<8.1024 0 0>, rot=<0 0 -0.0132 -0.9999>}
 spine_05 {parent=spine_04, loc=<10.0146 0 0>, rot=<0 0 -0.0015 -1>}
 neck_01 {parent=spine_05, loc=<11.3138 0.1529 0>, rot=<0 0 -0.2003 -0.9797>}
 neck_02 {parent=neck_01, loc=<4.5627 0 0>, rot=<0 0 0.0504 -0.9987>}
 head {parent=neck_02, loc=<5.3914 0 0>, rot=<0 0 0.115 -0.9934>}
 dyn_knot {parent=head, loc=<4.5292 -8.7624 -0.9059>, rot=<0.4555 -0.5408 0.451 -0.5446>}
 dyn_knot_up {parent=dyn_knot, loc=<0.4491 -0.198 0.7955>, rot=<-0.1511 0.16 -0.5785 -0.7854>}
 dyn_knot_down {parent=dyn_knot, loc=<0.6421 -0.062 -0.4195>, rot=<0.2064 -0.1602 -0.4465 -0.8558>}
 MSTR_bandana_main {parent=head, loc=<3.9355 -8.3379 0.9675>, rot=<0.1559 -0.9878 -0.0001 0.0005>}
 MSTR_bandana {parent=MSTR_bandana_main, loc=<1.7227 -0.0786 -0.0215>, rot=<0 0 0 1>}
 dyn_bandana_01 {parent=MSTR_bandana, loc=<0 0 0>, rot=<0 0 0 1>}
 dyn_bandana_02 {parent=dyn_bandana_01, loc=<8.5245 0 0>, rot=<-0.0024 0.0084 -0.0246 -0.9997>}
 dyn_bandana_03 {parent=dyn_bandana_02, loc=<7.4348 0 0>, rot=<-0.0003 0.0011 -0.0199 -0.9998>}
 dyn_bandana_end {parent=dyn_bandana_03, loc=<8.5563 0 0>, rot=<0 0 0 1>}
 MSTR_bandana_l {parent=MSTR_bandana_main, loc=<0.362 4.2025 8.0339>, rot=<0.046 0.1108 0.0457 -0.9917>}
 dyn_bandana_01_l {parent=MSTR_bandana_l, loc=<0 0 0>, rot=<0 0 0 1>}
 dyn_bandana_02_l {parent=dyn_bandana_01_l, loc=<8.8775 0 0>, rot=<-0.0784 -0.0225 0.0126 -0.9966>}
 dyn_bandana_03_l {parent=dyn_bandana_02_l, loc=<7.6998 0 0>, rot=<0.0077 -0.0174 0.0084 -0.9998>}
 dyn_bandana_end_l {parent=dyn_bandana_03_l, loc=<8.2893 0 0>, rot=<0 0 0 1>}
 MSTR_bandana_r {parent=MSTR_bandana_main, loc=<0.4016 4.4546 -5.947>, rot=<-0.1263 -0.1081 0.0812 -0.9827>}
 dyn_bandana_01_r {parent=MSTR_bandana_r, loc=<0 0 0>, rot=<0 0 0 1>}
 dyn_bandana_02_r {parent=dyn_bandana_01_r, loc=<9.5635 0 0>, rot=<-0.0081 0.01 -0.0276 -0.9995>}
 dyn_bandana_03_r {parent=dyn_bandana_02_r, loc=<7.6765 0 0>, rot=<0.0776 0.0215 -0.0131 -0.9967>}
 dyn_bandana_end_r {parent=dyn_bandana_03_r, loc=<7.4213 0 0>, rot=<0 0 0 1>}
 MSTR_hair_fr_l {parent=head, loc=<4.0837 4.857 -6.2257>, rot=<0.0401 -0.7067 -0.706 0.0204>}
 dyn_hair_01_fr_l {parent=MSTR_hair_fr_l, loc=<0 0 0>, rot=<0 0 0 1>}
 dyn_hair_02_fr_l {parent=dyn_hair_01_fr_l, loc=<5.1821 0 0>, rot=<0.0021 0.0014 -0.0272 -0.9996>}
 dyn_hair_03_fr_l {parent=dyn_hair_02_fr_l, loc=<5.4763 0 0>, rot=<0 0 0 1>}
 MSTR_hair_side_l {parent=head, loc=<4.2936 0.7988 -6.8112>, rot=<-0.0322 0.7065 0.7066 0.0227>}
 dyn_hair_01_side_l {parent=MSTR_hair_side_l, loc=<0 0 0>, rot=<0 0 0 1>}
 dyn_hair_02_side_l {parent=dyn_hair_01_side_l, loc=<5.1605 0 0>, rot=<0.0002 0.0026 -0.0633 -0.998>}
 dyn_hair_03_side_l {parent=dyn_hair_02_side_l, loc=<5.3365 0 0>, rot=<0 0 0 1>}
 MSTR_hair_bk_l {parent=head, loc=<4.2215 -4.5293 -5.7339>, rot=<0.0458 -0.9057 -0.4214 -0.0075>}
 dyn_hair_01_bk_l {parent=MSTR_hair_bk_l, loc=<0 0 0>, rot=<0 0 0 1>}
 dyn_hair_02_bk_l {parent=dyn_hair_01_bk_l, loc=<5.1428 0 0>, rot=<-0.0003 0.0036 0.003 -1>}
 dyn_hair_03_bk_l {parent=dyn_hair_02_bk_l, loc=<4.3533 0 0>, rot=<0 0 0 1>}
 MSTR_hair_bk {parent=head, loc=<4.4845 -7.2133 -0.1486>, rot=<0.0873 -0.9962 0 0>}
 dyn_hair_01_bk {parent=MSTR_hair_bk, loc=<0 0 0>, rot=<0 0 0 1>}
 dyn_hair_02_bk {parent=dyn_hair_01_bk, loc=<5.0351 0 0>, rot=<0 0 -0.0619 -0.9981>}
 dyn_hair_03_bk {parent=dyn_hair_02_bk, loc=<4.7301 0 0>, rot=<0 0 0 1>}
 MSTR_hair_fr_r {parent=head, loc=<4.0838 4.857 6.2257>, rot=<-0.7067 -0.0401 -0.0204 -0.706>}
 dyn_hair_01_fr_r {parent=MSTR_hair_fr_r, loc=<0 0 0>, rot=<0 0 0 1>}
 dyn_hair_02_fr_r {parent=dyn_hair_01_fr_r, loc=<-5.1825 0 0>, rot=<0.0021 0.0014 -0.0272 -0.9996>}
 dyn_hair_03_fr_r {parent=dyn_hair_02_fr_r, loc=<-5.4762 0 0>, rot=<0 0 0 1>}
 MSTR_hair_side_r {parent=head, loc=<4.2933 0.7988 6.8112>, rot=<-0.7065 -0.0322 0.0227 -0.7066>}
 dyn_hair_01_side_r {parent=MSTR_hair_side_r, loc=<0 0 0>, rot=<0 0 0 1>}
 dyn_hair_02_side_r {parent=dyn_hair_01_side_r, loc=<-5.1606 0 0>, rot=<0.0002 0.0026 -0.0633 -0.998>}
 dyn_hair_03_side_r {parent=dyn_hair_02_side_r, loc=<-5.3364 0 0>, rot=<0 0 0 1>}
 MSTR_hair_bk_r {parent=head, loc=<4.2214 -4.5293 5.7339>, rot=<0.9057 0.0458 -0.0075 0.4214>}
 dyn_hair_01_bk_r {parent=MSTR_hair_bk_r, loc=<0 0 0>, rot=<0 0 0 1>}
 dyn_hair_02_bk_r {parent=dyn_hair_01_bk_r, loc=<-5.1431 0 0>, rot=<-0.0003 0.0036 0.003 -1>}
 dyn_hair_03_bk_r {parent=dyn_hair_02_bk_r, loc=<-4.353 0 0>, rot=<0 0 0 -1>}
 attach_cape {parent=spine_05, loc=<9.017 -11.714 0>, rot=<0.0616 -0.7044 -0.0616 -0.7044>}
 attach_backpack {parent=spine_05, loc=<-0.123 -13.473 0>, rot=<-0.0093 -0.707 0.0093 -0.707>}
 clavicle_l {parent=spine_05, loc=<4.3575 -0.1466 -2.3223>, rot=<0.0465 -0.7289 -0.009 -0.6829>}
 upperarm_l {parent=clavicle_l, loc=<15.5111 0 0>, rot=<0.0226 -0.3234 -0.0088 -0.9459>}
 lowerarm_l {parent=upperarm_l, loc=<26.6514 0 0>, rot=<0.0391 -0.0099 -0.2453 -0.9686>}
 hand_l {parent=lowerarm_l, loc=<22.3491 0 0>, rot=<-0.7081 -0.0507 0.0975 -0.6975>}
 weapon_l {parent=hand_l, loc=<-0.1974 -3.4111 0.3811>, rot=<0 0 0 1>}
 thumb_01_l {parent=hand_l, loc=<1.3118 -0.3191 -2.459>, rot=<0.6546 -0.4042 0.0664 -0.6354>}
 thumb_02_l {parent=thumb_01_l, loc=<5.9336 -0.0448 0.0401>, rot=<-0.0037 0.0213 0.298 -0.9543>}
 thumb_03_l {parent=thumb_02_l, loc=<2.2223 0.3682 -0.0907>, rot=<0.0118 -0.013 0.0066 -0.9998>}
 index_metacarpal_l {parent=hand_l, loc=<1.6535 0.0183 -2.96>, rot=<0.0158 -0.0124 -0.0173 -0.9996>}
 index_01_l {parent=index_metacarpal_l, loc=<8.3008 0.0357 -0.0524>, rot=<-0.0044 0.0485 0.0924 -0.9945>}
 index_02_l {parent=index_01_l, loc=<3.3587 -0.3004 0.0452>, rot=<-0.002 0.0027 0.3225 -0.9466>}
 index_03_l {parent=index_02_l, loc=<2.5578 -0.123 -0.0252>, rot=<-0.002 0.0067 0.2558 -0.9667>}
 middle_metacarpal_l {parent=hand_l, loc=<2.1856 -0.1725 -0.5984>, rot=<0.006 0.0168 -0.0329 -0.9993>}
 middle_01_l {parent=middle_metacarpal_l, loc=<7.3275 -0.2447 0.0129>, rot=<-0.0208 0.0042 0.0858 -0.9961>}
 middle_02_l {parent=middle_01_l, loc=<4.0123 0.1401 0.1175>, rot=<-0.0218 0.0258 0.3512 -0.9357>}
 middle_03_l {parent=middle_02_l, loc=<2.6077 0.0534 0.1373>, rot=<0.0082 0.0188 0.1708 -0.9851>}
 ring_metacarpal_l {parent=hand_l, loc=<2.2074 -0.3625 1.3638>, rot=<-0.0284 0.0722 -0.0458 -0.9959>}
 ring_01_l {parent=ring_metacarpal_l, loc=<7.0295 -0.6197 -0.1446>, rot=<-0.0489 -0.0252 0.1892 -0.9804>}
 ring_02_l {parent=ring_01_l, loc=<3.6174 0.4647 -0.1027>, rot=<0 0 0.2806 -0.9598>}
 ring_03_l {parent=ring_02_l, loc=<2.4152 0.114 0.0199>, rot=<0 0 0.1189 -0.9929>}
 pinky_metacarpal_l {parent=hand_l, loc=<1.7616 -0.9753 3.3814>, rot=<-0.0397 0.0814 -0.069 -0.9935>}
 pinky_01_l {parent=pinky_metacarpal_l, loc=<7.1709 -0.4831 -0.1094>, rot=<-0.0654 -0.0035 0.2242 -0.9723>}
 pinky_02_l {parent=pinky_01_l, loc=<2.8274 0.0893 -0.1043>, rot=<0.0037 0.0011 0.2854 -0.9584>}
 pinky_03_l {parent=pinky_02_l, loc=<2.2046 0.0307 -0.0078>, rot=<-0.0002 0.0098 0.1445 -0.9895>}
 hand_attach_l {parent=hand_l, loc=<8.5 -3 0>, rot=<0.7071 0.7071 0 0>}
 deform_wrist_dn_base_l {parent=hand_l, loc=<0 -2.2342 0>, rot=<0 0 0 -1>}
 deform_wrist_dn_l {parent=deform_wrist_dn_base_l, loc=<0 0 0>, rot=<0 0 0 1>}
 deform_wrist_up_base_l {parent=hand_l, loc=<0 2.2682 0>, rot=<0 0 0 -1>}
 deform_wrist_up_l {parent=deform_wrist_up_base_l, loc=<0 0 0>, rot=<0 0 0 1>}
 lowerarm_twist_02_l {parent=lowerarm_l, loc=<9.6199 0 0>, rot=<0 0 0 -1>}
 lowerarm_twist_01_l {parent=lowerarm_l, loc=<17.8549 0 0>, rot=<0 0 0 -1>}
 deform_elbow_in_base_l {parent=lowerarm_l, loc=<0.2112 5.259 0.8274>, rot=<0 0 0 -1>}
 deform_elbow_in_l {parent=deform_elbow_in_base_l, loc=<0 0 0>, rot=<0 0 0 1>}
 deform_elbow_out_base_l {parent=lowerarm_l, loc=<0 -4 0>, rot=<0 0 0 -1>}
 deform_elbow_out_l {parent=deform_elbow_out_base_l, loc=<0 0 0>, rot=<0 0 0 1>}
 upperarm_twist_01_l {parent=upperarm_l, loc=<0.0041 0 0>, rot=<0 0 0 -1>}
 deform_upperarm_bk_base_l {parent=upperarm_twist_01_l, loc=<1.0813 -7.2193 -1.9453>, rot=<0 0 0 1>}
 deform_upperarm_bk_l {parent=deform_upperarm_bk_base_l, loc=<0 0 0>, rot=<0 0 0 1>}
 deform_upperarm_fr_base_l {parent=upperarm_twist_01_l, loc=<-0.3691 7.0506 -2.765>, rot=<0 0 0 1>}
 deform_upperarm_fr_l {parent=deform_upperarm_fr_base_l, loc=<0 0 0>, rot=<0 0 0 1>}
 deform_upperarm_up_base_l {parent=upperarm_twist_01_l, loc=<-3.2216 -1.1391 3.3311>, rot=<0 0 0 1>}
 deform_upperarm_up_l {parent=deform_upperarm_up_base_l, loc=<0 0 0>, rot=<0 0 0 1>}
 upperarm_twist_02_l {parent=upperarm_l, loc=<13.7941 0 0>, rot=<0 0 0 -1>}
 def_chainCorrUp_l {parent=clavicle_l, loc=<5.2579 -0.0783 7.531>, rot=<0.0008 0.0336 -0.0252 -0.9991>}
 clavicle_r {parent=spine_05, loc=<4.3575 -0.1466 2.3223>, rot=<0.7289 0.0465 -0.6829 0.009>}
 upperarm_r {parent=clavicle_r, loc=<-15.5111 0 0>, rot=<0.0226 -0.3234 -0.0088 -0.9459>}
 lowerarm_r {parent=upperarm_r, loc=<-26.6514 0 0>, rot=<0.0391 -0.0099 -0.2453 -0.9686>}
 hand_r {parent=lowerarm_r, loc=<-22.3491 0 0>, rot=<-0.7081 -0.0507 0.0975 -0.6975>}
 weapon_r {parent=hand_r, loc=<0.1974 3.4111 -0.3811>, rot=<0 0 0 1>}
 thumb_01_r {parent=hand_r, loc=<-1.3118 0.3191 2.459>, rot=<0.6546 -0.4042 0.0664 -0.6354>}
 thumb_02_r {parent=thumb_01_r, loc=<-5.9336 0.0448 -0.0401>, rot=<-0.0037 0.0213 0.298 -0.9543>}
 thumb_03_r {parent=thumb_02_r, loc=<-2.2223 -0.3682 0.0907>, rot=<0.0118 -0.013 0.0066 -0.9998>}
 index_metacarpal_r {parent=hand_r, loc=<-1.6535 -0.0183 2.96>, rot=<0.0158 -0.0124 -0.0173 -0.9996>}
 index_01_r {parent=index_metacarpal_r, loc=<-8.3008 -0.0357 0.0524>, rot=<-0.0044 0.0485 0.0924 -0.9945>}
 index_02_r {parent=index_01_r, loc=<-3.3587 0.3004 -0.0452>, rot=<-0.002 0.0027 0.3225 -0.9466>}
 index_03_r {parent=index_02_r, loc=<-2.5578 0.123 0.0252>, rot=<-0.002 0.0067 0.2558 -0.9667>}
 middle_metacarpal_r {parent=hand_r, loc=<-2.1856 0.1725 0.5984>, rot=<0.006 0.0168 -0.0329 -0.9993>}
 middle_01_r {parent=middle_metacarpal_r, loc=<-7.3275 0.2447 -0.0129>, rot=<-0.0208 0.0042 0.0858 -0.9961>}
 middle_02_r {parent=middle_01_r, loc=<-4.0123 -0.1401 -0.1175>, rot=<-0.0218 0.0258 0.3512 -0.9357>}
 middle_03_r {parent=middle_02_r, loc=<-2.6077 -0.0534 -0.1373>, rot=<0.0082 0.0188 0.1708 -0.9851>}
 ring_metacarpal_r {parent=hand_r, loc=<-2.2074 0.3625 -1.3638>, rot=<-0.0284 0.0722 -0.0458 -0.9959>}
 ring_01_r {parent=ring_metacarpal_r, loc=<-7.0295 0.6197 0.1446>, rot=<-0.0489 -0.0252 0.1892 -0.9804>}
 ring_02_r {parent=ring_01_r, loc=<-3.6174 -0.4647 0.1027>, rot=<0 0 0.2806 -0.9598>}
 ring_03_r {parent=ring_02_r, loc=<-2.4152 -0.114 -0.0199>, rot=<0 0 0.1189 -0.9929>}
 pinky_metacarpal_r {parent=hand_r, loc=<-1.7616 0.9753 -3.3814>, rot=<-0.0397 0.0814 -0.069 -0.9935>}
 pinky_01_r {parent=pinky_metacarpal_r, loc=<-7.1709 0.4831 0.1094>, rot=<-0.0654 -0.0035 0.2242 -0.9723>}
 pinky_02_r {parent=pinky_01_r, loc=<-2.8274 -0.0893 0.1043>, rot=<0.0037 0.0011 0.2854 -0.9584>}
 pinky_03_r {parent=pinky_02_r, loc=<-2.2046 -0.0307 0.0078>, rot=<-0.0002 0.0098 0.1445 -0.9895>}
 hand_attach_r {parent=hand_r, loc=<-8.5001 2.9999 0>, rot=<0 0 -0.7071 -0.7071>}
 deform_wrist_dn_base_r {parent=hand_r, loc=<-0.0004 2.2345 -0.0001>, rot=<0 0 0 -1>}
 deform_wrist_dn_r {parent=deform_wrist_dn_base_r, loc=<0 0 0>, rot=<0 0 0 1>}
 deform_wrist_up_base_r {parent=hand_r, loc=<0.0003 -2.2684 0.0001>, rot=<0 0 0 -1>}
 deform_wrist_up_r {parent=deform_wrist_up_base_r, loc=<0 0 0>, rot=<0 0 0 1>}
 lowerarm_twist_02_r {parent=lowerarm_r, loc=<-9.6199 0 0>, rot=<0 0 0 -1>}
 lowerarm_twist_01_r {parent=lowerarm_r, loc=<-17.8549 0 0>, rot=<0 0 0 -1>}
 deform_elbow_in_base_r {parent=lowerarm_r, loc=<-0.2111 -5.2591 -0.8275>, rot=<0 0 0 -1>}
 deform_elbow_in_r {parent=deform_elbow_in_base_r, loc=<0 0 0>, rot=<0 0 0 1>}
 deform_elbow_out_base_r {parent=lowerarm_r, loc=<0 4 -0.0001>, rot=<0 0 0 -1>}
 deform_elbow_out_r {parent=deform_elbow_out_base_r, loc=<0 0 0>, rot=<0 0 0 1>}
 upperarm_twist_01_r {parent=upperarm_r, loc=<-0.0041 0 0>, rot=<0 0 0 -1>}
 deform_upperarm_bk_base_r {parent=upperarm_twist_01_r, loc=<-1.0814 7.2192 1.9455>, rot=<0 0 0 1>}
 deform_upperarm_bk_r {parent=deform_upperarm_bk_base_r, loc=<0 0 0>, rot=<0 0 0 1>}
 deform_upperarm_fr_base_r {parent=upperarm_twist_01_r, loc=<0.3689 -7.0506 2.7652>, rot=<0 0 0 1>}
 deform_upperarm_fr_r {parent=deform_upperarm_fr_base_r, loc=<0 0 0>, rot=<0 0 0 1>}
 deform_upperarm_up_base_r {parent=upperarm_twist_01_r, loc=<3.2214 1.1391 -3.3308>, rot=<0 0 0 1>}
 deform_upperarm_up_r {parent=deform_upperarm_up_base_r, loc=<0 0 0>, rot=<0 0 0 -1>}
 upperarm_twist_02_r {parent=upperarm_r, loc=<-13.7941 0 0>, rot=<0 0 0 -1>}
 def_chainCorrUp_r {parent=clavicle_r, loc=<-5.2579 0.0783 -7.531>, rot=<0.0252 -0.9991 0.0008 -0.0336>}
 def_necklace_01 {parent=spine_05, loc=<6.0447 7.52 0>, rot=<-0.2533 -0.9674 0 0>}
 def_necklace_02 {parent=spine_05, loc=<3.6034 8.8552 -0.0315>, rot=<-0.2116 -0.9774 0 0>}
 dyn_chain_l {parent=spine_05, loc=<3.4813 9.5283 -3.524>, rot=<-0.195 -0.8895 -0.089 0.4036>}
 dyn_chain_r {parent=spine_05, loc=<3.879 9.6787 3.2631>, rot=<-0.1831 -0.9161 0.1451 -0.3258>}
 dyn_chain_rear {parent=spine_05, loc=<0.8791 9.8472 -0.0557>, rot=<-0.0922 -0.9957 -0.0005 0>}
 dyn_chain {parent=dyn_chain_rear, loc=<-0.6018 0.6923 0.3494>, rot=<0 0 -0.0781 -0.9969>}
 dyn_rope01_l {parent=spine_05, loc=<7.2201 6.625 -6.0192>, rot=<-0.2348 -0.9702 0.0188 0.0573>}
 dyn_rope02_l {parent=dyn_rope01_l, loc=<3.2016 0 0>, rot=<-0.0357 -0.0626 0.0372 -0.9967>}
 dyn_rope03_l {parent=dyn_rope02_l, loc=<2.5324 0 0>, rot=<-0.0043 -0.0095 0.0568 -0.9983>}
 dyn_rope04_l {parent=dyn_rope03_l, loc=<3.7809 0 0>, rot=<-0.0299 -0.0873 0.0371 -0.995>}
 dyn_rope05_l {parent=dyn_rope04_l, loc=<2.8313 0 0>, rot=<-0.0578 -0.2752 0.082 -0.9561>}
 dyn_rope01_r {parent=spine_05, loc=<7.2203 6.625 6.0191>, rot=<-0.2348 -0.9702 -0.0188 -0.0573>}
 dyn_rope02_r {parent=dyn_rope01_r, loc=<3.2017 -0.0001 0>, rot=<0.0357 0.0626 0.0372 -0.9967>}
 dyn_rope03_r {parent=dyn_rope02_r, loc=<2.5323 0.0001 0>, rot=<0.0043 0.0095 0.0568 -0.9983>}
 dyn_rope04_r {parent=dyn_rope03_r, loc=<3.7809 0 0>, rot=<0.0299 0.0873 0.0371 -0.995>}
 dyn_rope05_r {parent=dyn_rope04_r, loc=<2.8318 -0.0002 -0.0003>, rot=<0.0578 0.2752 0.082 -0.9561>}
 deform_pec_aim_base_l {parent=spine_04, loc=<7.8732 12.6388 -2.4123>, rot=<0.0102 -0.4602 0.3979 -0.7936>}
 deform_pec_aim_l {parent=deform_pec_aim_base_l, loc=<0 0 0>, rot=<0 0 0 1>}
 deform_pec_base_l {parent=deform_pec_aim_l, loc=<10.766 0 0>, rot=<0 0 0 1>}
 deform_pec_l {parent=deform_pec_base_l, loc=<0.055 0 0>, rot=<0 0 0 1>}
 deform_pec_aim_base_r {parent=spine_04, loc=<7.8727 12.6388 2.4123>, rot=<-0.4602 -0.0102 0.7936 0.3979>}
 deform_pec_aim_r {parent=deform_pec_aim_base_r, loc=<0 0 0>, rot=<0 0 0 1>}
 deform_pec_base_r {parent=deform_pec_aim_r, loc=<-10.7662 -0.0006 -0.0007>, rot=<0 0 0 1>}
 deform_pec_r {parent=deform_pec_base_r, loc=<0 0 0>, rot=<0 0 0 1>}
 thigh_l {parent=pelvis, loc=<-2.5693 -0.0109 -7.7501>, rot=<0.0012 0.0572 0.0155 -0.9982>}
 calf_l {parent=thigh_l, loc=<-39.0484 0 0>, rot=<0.0006 -0.0161 -0.0391 -0.9991>}
 foot_l {parent=calf_l, loc=<-42.561 0.1077 0.4292>, rot=<0.0418 -0.0455 0.0343 -0.9975>}
 ball_l {parent=foot_l, loc=<-6.1599 13.8264 0.6806>, rot=<0.0001 -0.0001 -0.7063 -0.7079>}
 dyn_shoeTag01_l {parent=foot_l, loc=<0.1869 7.1163 -4.8862>, rot=<0.0479 -0.9406 -0.2782 -0.1883>}
 dyn_shoeTag02_l {parent=dyn_shoeTag01_l, loc=<3.1182 0 0>, rot=<-0.2588 0 0 -0.9659>}
 calf_twist_02_l {parent=calf_l, loc=<-18.8064 -0.0967 0.0397>, rot=<0.0237 0.0095 -0.0035 -0.9997>}
 dyn_pants_cloth_03_l {parent=calf_twist_02_l, loc=<-9.8888 6.6084 -3.209>, rot=<-0.4034 -0.606 -0.3433 -0.5934>}
 dyn_pants_cloth_02_l {parent=calf_twist_02_l, loc=<-12.0292 -1.0735 7.707>, rot=<0.7417 -0.0498 0.6688 0.0123>}
 dyn_pants_cloth_04_l {parent=calf_twist_02_l, loc=<-2.0923 1.1693 8.4651>, rot=<0.7356 0.0695 0.6626 0.1224>}
 dyn_pants_cloth_05_l {parent=calf_twist_02_l, loc=<-4.4236 -4.4817 -7.1903>, rot=<0.0863 -0.7399 0.1468 -0.6508>}
 dyn_pants_cloth_01_l {parent=calf_twist_02_l, loc=<-12.7452 -5.0639 -5.6994>, rot=<0.2046 -0.7147 0.2383 -0.625>}
 calf_twist_01_l {parent=calf_l, loc=<-35.813 -0.0967 0.0397>, rot=<0.0237 0.0095 -0.0035 -0.9997>}
 def_foot_bk_l {parent=calf_twist_01_l, loc=<-3.955 -5.607 -0.058>, rot=<0.4892 -0.5597 0.4816 -0.4642>}
 def_foot_fr_l {parent=calf_twist_01_l, loc=<-2.9736 5.8883 0.2022>, rot=<-0.0137 -0.0299 -0.3047 -0.9519>}
 deform_knee_in_base_l {parent=calf_l, loc=<1.298 -5.2957 -0.0496>, rot=<0 0 0 -1>}
 deform_knee_in_l {parent=deform_knee_in_base_l, loc=<0 0 0>, rot=<0 0 0 1>}
 deform_knee_out_base_l {parent=calf_l, loc=<0 4 0>, rot=<0 0 0 -1>}
 deform_knee_out_l {parent=deform_knee_out_base_l, loc=<0 0 0>, rot=<0 0 0 1>}
 thigh_twist_01_l {parent=thigh_l, loc=<-21.6303 0 0>, rot=<0.0013 0 0 -1>}
 deform_groin_base_l {parent=thigh_l, loc=<-2.1908 8.1142 -1.7442>, rot=<0 0 0 -1>}
 deform_groin_l {parent=deform_groin_base_l, loc=<0 0 0>, rot=<0 0 0 1>}
 MSTR_strape02_fr_l {parent=thigh_l, loc=<-5.9316 7.2208 -4.9228>, rot=<-0.0771 -0.0026 0.9969 -0.0118>}
 dyn_strape02_fr_l {parent=MSTR_strape02_fr_l, loc=<0 0 0>, rot=<0 0 0 1>}
 dyn_strape03_fr_l {parent=dyn_strape02_fr_l, loc=<10.7103 0 0>, rot=<0.0024 0.0322 -0.0503 -0.9982>}
 dyn_strape04_fr_l {parent=dyn_strape03_fr_l, loc=<10.6751 0 0>, rot=<0.0189 0.0182 -0.2204 -0.9751>}
 dyn_strape05_fr_l {parent=dyn_strape04_fr_l, loc=<8.4782 0 -0.0001>, rot=<0.034 0.054 -0.408 -0.9107>}
 dyn_strape06_fr_l {parent=dyn_strape05_fr_l, loc=<7.0651 -0.0001 0>, rot=<0 0 -0.0884 -0.9961>}
 thigh_r {parent=pelvis, loc=<-2.5693 -0.0109 7.7501>, rot=<0.0572 -0.0012 0.9982 0.0155>}
 calf_r {parent=thigh_r, loc=<39.0484 0 0>, rot=<0.0006 -0.0161 -0.0391 -0.9991>}
 foot_r {parent=calf_r, loc=<42.561 -0.1077 -0.4292>, rot=<0.0418 -0.0455 0.0343 -0.9975>}
 ball_r {parent=foot_r, loc=<6.1599 -13.8264 -0.6806>, rot=<0.0001 -0.0001 -0.7063 -0.7079>}
 dyn_shoeTag01_r {parent=foot_r, loc=<-0.1869 -7.1163 4.8862>, rot=<0.0479 -0.9406 -0.2782 -0.1883>}
 dyn_shoeTag02_r {parent=dyn_shoeTag01_r, loc=<-3.1182 0 0.0001>, rot=<-0.2588 0 0 -0.9659>}
 calf_twist_02_r {parent=calf_r, loc=<18.8064 0.0967 -0.0397>, rot=<0.0237 0.0095 -0.0035 -0.9997>}
 dyn_pants_cloth_05_r {parent=calf_twist_02_r, loc=<12.4702 7.8538 -0.8089>, rot=<0.4816 0.4642 -0.4892 -0.5597>}
 dyn_pants_cloth_04_r {parent=calf_twist_02_r, loc=<9.3059 -0.5423 7.5809>, rot=<0.0167 0.6687 0.0449 -0.742>}
 dyn_pants_cloth_02_r {parent=calf_twist_02_r, loc=<6.0285 4.4236 -8.1743>, rot=<-0.6676 -0.0419 0.7352 0.1097>}
 dyn_pants_cloth_03_r {parent=calf_twist_02_r, loc=<11.61 -5.6711 -4.3982>, rot=<-0.5948 0.3549 0.6133 -0.3796>}
 calf_twist_01_r {parent=calf_r, loc=<35.813 0.0967 -0.0397>, rot=<0.0237 0.0095 -0.0035 -0.9997>}
 def_foot_fr_r {parent=calf_twist_01_r, loc=<2.9736 -5.8883 -0.2021>, rot=<-0.0137 -0.0299 -0.3047 -0.9519>}
 def_foot_bk_r {parent=calf_twist_01_r, loc=<3.955 5.607 0.058>, rot=<0.4892 -0.5597 0.4816 -0.4642>}
 deform_knee_in_base_r {parent=calf_r, loc=<-1.298 5.2957 0.0496>, rot=<0 0 0 -1>}
 deform_knee_in_r {parent=deform_knee_in_base_r, loc=<0 0 0>, rot=<0 0 0 1>}
 deform_knee_out_base_r {parent=calf_r, loc=<0 -4 0>, rot=<0 0 0 -1>}
 deform_knee_out_r {parent=deform_knee_out_base_r, loc=<0 0 0>, rot=<0 0 0 1>}
 thigh_twist_01_r {parent=thigh_r, loc=<21.6303 0 0>, rot=<0.0013 0 0 -1>}
 deform_groin_base_r {parent=thigh_r, loc=<2.1908 -8.1142 1.7442>, rot=<0 0 0 -1>}
 deform_groin_r {parent=deform_groin_base_r, loc=<0 0 0>, rot=<0 0 0 1>}
 MSTR_strape02_fr_r {parent=thigh_r, loc=<5.9316 -7.2208 4.9228>, rot=<-0.0771 -0.0026 0.9969 -0.0118>}
 dyn_strape02_fr_r {parent=MSTR_strape02_fr_r, loc=<0 0 0>, rot=<0 0 0 1>}
 dyn_strape03_fr_r {parent=dyn_strape02_fr_r, loc=<-10.7103 0 0>, rot=<0.0024 0.0322 -0.0503 -0.9982>}
 dyn_strape04_fr_r {parent=dyn_strape03_fr_r, loc=<-10.6751 0 0>, rot=<0.0189 0.0182 -0.2204 -0.9751>}
 dyn_strape05_fr_r {parent=dyn_strape04_fr_r, loc=<-8.4782 0 0.0001>, rot=<0.034 0.054 -0.408 -0.9107>}
 dyn_strape06_fr_r {parent=dyn_strape05_fr_r, loc=<-7.0651 0.0001 0>, rot=<0 0 -0.0884 -0.9961>}
 deform_glute_base_l {parent=pelvis, loc=<-8.8064 -8.8222 -8.4287>, rot=<0.0275 0.0501 -0.4472 -0.8926>}
 deform_glute_l {parent=deform_glute_base_l, loc=<0 0 0>, rot=<0 0 0 1>}
 deform_glute_base_r {parent=pelvis, loc=<-8.8062 -8.8222 8.4287>, rot=<0.0501 -0.0275 0.8926 -0.4472>}
 deform_glute_r {parent=deform_glute_base_r, loc=<0 0 0>, rot=<0 0 0 1>}
 def_logo {parent=pelvis, loc=<0.0327 9.0694 0>, rot=<-0.5089 -0.491 -0.491 -0.5089>}
 dyn_shirt_l {parent=pelvis, loc=<1.4955 -3.9755 15.6297>, rot=<0.0059 -0.1345 0.9908 0.0166>}
 dyn_shirt_r {parent=pelvis, loc=<-1.2303 -4.4742 -15.805>, rot=<-0.0146 -0.9932 0.1095 -0.0371>}
 dyn_shirt {parent=pelvis, loc=<-4.5798 -13.4827 0>, rot=<-0.0258 -0.9997 0 0>}
 MSTR_strape02_bk_r {parent=pelvis, loc=<-13.1948 -11.2193 11.048>, rot=<0.9911 -0.0313 -0.1293 -0.0041>}
 dyn_strape02_bk_r {parent=MSTR_strape02_bk_r, loc=<0 0 0>, rot=<0 0 0 1>}
 dyn_strape03_bk_r {parent=dyn_strape02_bk_r, loc=<-9.2729 -0.0001 0>, rot=<0.0129 0.0196 -0.0417 -0.9989>}
 dyn_strape04_bk_r {parent=dyn_strape03_bk_r, loc=<-8.9832 0 0>, rot=<0.0217 0.0909 -0.2076 -0.9737>}
 dyn_strape05_bk_r {parent=dyn_strape04_bk_r, loc=<-8.4412 0.0001 0.0001>, rot=<-0.1908 -0.0752 -0.5468 -0.8118>}
 dyn_strape06_bk_r {parent=dyn_strape05_bk_r, loc=<-7.0651 -0.0001 0>, rot=<0 0 0.0909 -0.9959>}
 MSTR_strape02_bk_l {parent=pelvis, loc=<-13.1948 -11.2193 -11.048>, rot=<-0.0313 -0.9911 0.0041 -0.1293>}
 dyn_strape02_bk_l {parent=MSTR_strape02_bk_l, loc=<0 0 0>, rot=<0 0 0 1>}
 dyn_strape03_bk_l {parent=dyn_strape02_bk_l, loc=<9.2729 0.0001 0>, rot=<0.0129 0.0196 -0.0417 -0.9989>}
 dyn_strape04_bk_l {parent=dyn_strape03_bk_l, loc=<8.9832 0 0>, rot=<0.0217 0.0909 -0.2076 -0.9737>}
 dyn_strape05_bk_l {parent=dyn_strape04_bk_l, loc=<8.4412 -0.0001 -0.0001>, rot=<-0.1908 -0.0752 -0.5468 -0.8118>}
 dyn_strape06_bk_l {parent=dyn_strape05_bk_l, loc=<7.0651 0.0001 0>, rot=<0 0 0.0909 -0.9959>}
]
variables [
 Leg ROM_L Thigh Fwd Knee Bend {type=double, default=0}
 Leg ROM_L Thigh Fwd Knee Bend_0 {type=double, default=0}
 Leg ROM_L Thigh Raised Fwd {type=float, default=0}
 Leg ROM_L Thigh Raised Rear {type=float, default=0}
 Leg ROM_L Thigh Raised Side {type=float, default=0}
 Leg ROM_R Thigh Fwd Knee Bend {type=double, default=0}
 Leg ROM_R Thigh Raised Fwd {type=double, default=0}
 Leg ROM_R Thigh Raised Rear {type=float, default=0}
 Leg ROM_R Thigh Raised Side {type=float, default=0}
 isCrouching {type=bool, default=False}
 isDBNO {type=bool, default=False}
 isSliding {type=bool, default=False}
]
