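# Physics asset for the M_MED_SummitReedGrit male medium body, setting up physics for a dangling left-side strap.
# Front and back chains of capsule bodies run from small MSTR anchor spheres through limited-swing constraints.
# Their last links are constrained to each other in both directions with free swing, closing the strap into a loop.
# Pelvis, left thigh and calf capsules, plus a large root capsule below the origin as a floor, provide collision shapes.
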PhysicsAsset /BRCosmetics/Characters/Player/Male/Medium/Bodies/M_MED_SummitReedGrit/Meshes/M_MED_SummitReedGrit_Physics2
bodies {
 MSTR_strape02_bk_l [{shapeType=Sphere, radius=0.505, shapeName=MSTR_strape02_bk_l_sphere}]
 MSTR_strape02_fr_l [{shapeType=Sphere, radius=0.505, shapeName=MSTR_strape02_fr_l_sphere}]
 calf_l [{
   shapeType=Capsule
   length=24.4879
   radius=9.4546
   shapeName=pelvis_capsule
   center=<-15.7764 -0.6692 0.7962>
   rotation=<rot 89.2106 179.9998 179.9998>
  }]
 dyn_strape02_bk_l [{
   shapeType=Capsule
   length=3.8399
   radius=4.3396
   shapeName=dyn_strape01_bk_l_capsule
   center=<4.8516 -3.2272 0.0737>
   rotation=<rot -0.1584 97.1627 -88.7012>
  }]
 dyn_strape02_fr_l [{
   shapeType=Capsule
   length=5.0337
   radius=4.5812
   shapeName=dyn_strape01_bk_l_capsule
   center=<5.0161 -4.7281 -0.639>
   rotation=<rot 29.6428 93.2801 -91.5096>
  }]
 dyn_strape03_bk_l [{
   shapeType=Capsule
   length=4.1018
   radius=4.0778
   shapeName=dyn_strape01_bk_l_capsule
   center=<4.7405 -2.0749 0.0712>
   rotation=<rot 0.0486 88.0813 -88.6925>
  }]
 dyn_strape03_fr_l [{
   shapeType=Capsule
   length=4.646
   radius=4.5812
   shapeName=dyn_strape01_bk_l_capsule
   center=<5.112 -4.5161 -0.4992>
   rotation=<rot 30.0284 88.8413 -88.4898>
  }]
 dyn_strape04_bk_l [{
   shapeType=Capsule
   length=3.3341
   radius=4.0812
   shapeName=dyn_strape01_bk_l_capsule
   center=<4.0794 -2.5876 1.0687>
   rotation=<rot 0.0914 86.2028 -88.6948>
  }]
 dyn_strape04_fr_l [{
   shapeType=Capsule
   length=6.7732
   radius=4.5812
   shapeName=dyn_strape01_bk_l_capsule
   center=<3.8459 -4.1992 -2.1011>
   rotation=<rot 60.4461 75.6903 -101.1641>
  }]
 dyn_strape05_bk_l [{
   shapeType=Capsule
   length=3.5984
   radius=4.5812
   shapeName=dyn_strape01_bk_l_capsule
   center=<3.6787 -0.0004 4.4927>
   rotation=<rot 0.1288 84.5579 -88.698>
  }]
 dyn_strape05_fr_l [{
   shapeType=Capsule
   length=3.5984
   radius=4.5812
   shapeName=dyn_strape01_bk_l_capsule
   center=<4.3987 0.0004 -4.4061>
   rotation=<rot -0.1859 98.3745 -88.7049>
  }]
 pelvis [{shapeType=Capsule, length=12.2442, radius=10.6662, shapeName=pelvis_capsule, center=<-11.8546 -1.1608 0.3601>}]
 root [{shapeType=Capsule, length=0.505, radius=100, shapeName=root_capsule, center=<0 0 -99.338>, rotation=<rot 0 -90 0>}]
 thigh_l [{
   shapeType=Capsule
   length=24.7183
   radius=9.8809
   shapeName=pelvis_capsule
   center=<-21.2476 -0.1962 1.9465>
   rotation=<rot 89.4867 0.0009 0.0011>
  }]
}
constraints [
 {
  bone1Name=dyn_strape02_bk_l
  bone2Name=MSTR_strape02_bk_l
  swing1LimitDegrees=10
  swing2LimitDegrees=10
  twistLimitDegrees=20
  twistMotion=Locked
 }
 {bone1Name=dyn_strape03_bk_l, bone2Name=dyn_strape02_bk_l, swing1LimitDegrees=10, swing2LimitDegrees=20, twistLimitDegrees=10}
 {bone1Name=dyn_strape04_bk_l, bone2Name=dyn_strape03_bk_l, swing1LimitDegrees=20, swing2LimitDegrees=20, twistLimitDegrees=5}
 {bone1Name=dyn_strape05_bk_l, bone2Name=dyn_strape04_bk_l, swing1LimitDegrees=20, swing2LimitDegrees=20, twistLimitDegrees=5}
 {
  bone1Name=dyn_strape05_bk_l
  bone2Name=dyn_strape05_fr_l
  swing1LimitDegrees=45
  swing1Motion=Free
  swing2LimitDegrees=45
  swing2Motion=Free
  twistLimitDegrees=15
 }
 {
  bone1Name=dyn_strape05_fr_l
  bone2Name=dyn_strape05_bk_l
  swing1LimitDegrees=45
  swing1Motion=Free
  swing2LimitDegrees=45
  swing2Motion=Free
  twistLimitDegrees=15
 }
 {bone1Name=dyn_strape02_fr_l, bone2Name=MSTR_strape02_fr_l, swing1LimitDegrees=10, swing2LimitDegrees=10, twistLimitDegrees=0}
 {bone1Name=dyn_strape03_fr_l, bone2Name=dyn_strape02_fr_l, swing1LimitDegrees=45, swing2LimitDegrees=15, twistLimitDegrees=5}
 {bone1Name=dyn_strape04_fr_l, bone2Name=dyn_strape03_fr_l, swing1LimitDegrees=30, swing2LimitDegrees=30, twistLimitDegrees=5}
 {bone1Name=dyn_strape05_fr_l, bone2Name=dyn_strape04_fr_l, swing1LimitDegrees=20, swing2LimitDegrees=20, twistLimitDegrees=5}
]
