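# Animation Blueprint for the SummitReedGrit male body cosmetic, adding secondary motion to clothing and accessories.
# It copies the attached parent's pose, then runs a Control Rig fed crouch, slide and DBNO states.
# Cached-pose stages layer AnimDynamics on pant cloth and shoe tags, and weighted multi-bone constraints on ropes.
# bRigidBodyEnabled selects RigidBody simulation or thigh-relative AnimDynamics strap chains for output, both reading the ShirtDynamics cache.
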AnimBlueprint /BRCosmetics/Characters/Player/Male/Medium/Bodies/M_MED_SummitReedGrit/Meshes/M_MED_SummitReedGrit_AnimBP
target_skeleton=&/M_MED_SummitReedGrit_Skeleton
graphs [
 AnimGraph {nodes [
   n1 = Anim Dynamics - Bone: dyn_rope04_r(Alpha=1, ExternalForce="0,0,0") {
    angularDampingOverride=0.99
    angularSpringConstant=900
    bAngularSpring=true
    bOverrideAngularDamping=true
    bUseGravityOverride=true
    linearDampingOverride=0.9999
    linearSpringConstant=600
    simulationSpace=RootRelative
    boundBone {boneName=dyn_rope04_r}
    componentAppliedLinearAccClamp=<1500 1500 1500>
    componentLinearAccScale=<0.1 0.1 0.3>
    componentLinearVelScale=<0.1 0.1 0.1>
    gravityOverride {z=-980}
    physicsBodyDefinitions [{
      boundBone {boneName=dyn_rope04_r}
      boxExtents=<12 12 12>
      constraintSetup {angularLimitsMax=<0 10 30>, angularLimitsMin=<0 -5 0>, angularTarget=<1 0 0>}
      localJointOffset=<1 0 0>
     }]
    simSpaceSettings {maxAngularAcceleration=1500, maxAngularVelocity=1500, simSpaceAngularAlpha=0.1}
   }
   Anim Dynamics - Bone: dyn_rope05_r(ComponentPose=@n1, Alpha=1, ExternalForce="0,0,0") {
    angularDampingOverride=0.99
    angularSpringConstant=900
    bAngularSpring=true
    bOverrideAngularDamping=true
    bUseGravityOverride=true
    linearDampingOverride=0.9999
    linearSpringConstant=600
    simulationSpace=RootRelative
    boundBone {boneName=dyn_rope05_r}
    componentAppliedLinearAccClamp=<1500 1500 1500>
    componentLinearAccScale=<0.1 0.1 0.3>
    componentLinearVelScale=<0.1 0.1 0.1>
    gravityOverride {z=-980}
    physicsBodyDefinitions [{
      boundBone {boneName=dyn_rope05_r}
      boxExtents=<12 12 12>
      constraintSetup {angularLimitsMax=<0 7 10>, angularLimitsMin=<0 -7 0>, angularTarget=<1 0 0>}
      localJointOffset=<1 0 0>
     }]
    simSpaceSettings {maxAngularAcceleration=1500, maxAngularVelocity=1500, simSpaceAngularAlpha=0.1}
   }
   n2 = Anim Dynamics - Bone: dyn_rope04_l(Alpha=1, ExternalForce="0,0,0") {
    angularDampingOverride=0.99
    angularSpringConstant=900
    bAngularSpring=true
    bOverrideAngularDamping=true
    bUseGravityOverride=true
    linearDampingOverride=0.9999
    linearSpringConstant=600
    simulationSpace=RootRelative
    boundBone {boneName=dyn_rope04_l}
    componentAppliedLinearAccClamp=<1500 1500 1500>
    componentLinearAccScale=<0.1 0.1 0.3>
    componentLinearVelScale=<0.1 0.1 0.1>
    gravityOverride {z=-980}
    physicsBodyDefinitions [{
      boundBone {boneName=dyn_rope04_l}
      boxExtents=<12 12 12>
      constraintSetup {angularLimitsMax=<0 10 30>, angularLimitsMin=<0 -5 0>, angularTarget=<1 0 0>}
      localJointOffset=<1 0 0>
     }]
    simSpaceSettings {maxAngularAcceleration=1500, maxAngularVelocity=1500, simSpaceAngularAlpha=0.1}
   }
   Anim Dynamics - Bone: dyn_rope05_l(ComponentPose=@n2, Alpha=1, ExternalForce="0,0,0") {
    angularDampingOverride=0.99
    angularSpringConstant=900
    bAngularSpring=true
    bOverrideAngularDamping=true
    bUseGravityOverride=true
    linearDampingOverride=0.9999
    linearSpringConstant=600
    simulationSpace=RootRelative
    boundBone {boneName=dyn_rope05_l}
    componentAppliedLinearAccClamp=<1500 1500 1500>
    componentLinearAccScale=<0.1 0.1 0.3>
    componentLinearVelScale=<0.1 0.1 0.1>
    gravityOverride {z=-980}
    physicsBodyDefinitions [{
      boundBone {boneName=dyn_rope05_l}
      boxExtents=<12 12 12>
      constraintSetup {angularLimitsMax=<0 7 10>, angularLimitsMin=<0 -7 0>, angularTarget=<1 0 0>}
      localJointOffset=<1 0 0>
     }]
    simSpaceSettings {maxAngularAcceleration=1500, maxAngularVelocity=1500, simSpaceAngularAlpha=0.1}
   }
   n3 = RigidBody!(
    ComponentPose=Local To Component(LocalPose=Use cached pose 'ShirtDynamics'())
    Alpha
    OverrideWorldGravity:Vector="0,0,-980"
    ExternalForce
    SimSpaceSettings:Sim Space Settings Structure=(WorldAlpha=1,VelocityScaleZ=1,DampingAlpha=0,MaxLinearVelocity=5,MaxAngularVelocity=15,MaxLinearAcceleration=150,MaxAngularAcceleration=150,ExternalLinearDragV=(X=0,Y=0,Z=0),ExternalLinearVelocity=(X=0,Y=0,Z=0),ExternalAngularVelocity=(X=0,Y=0,Z=0))
   ) -> (Pose) {
    bClampLinearTranslationLimitToRefPose=true
    bOverrideWorldGravity=true
    overridePhysicsAsset=&/M_MED_SummitReedGrit_Physics2
    simulationSpace=BaseBoneSpace
    baseBoneRef {boneName=pelvis}
    overrideWorldGravity {z=-980}
    simSpaceSettings {dampingAlpha=0, maxAngularAcceleration=150, maxAngularVelocity=15, maxLinearAcceleration=150, maxLinearVelocity=5, worldAlpha=1}
   }
   n4 = RigidBody(
    ComponentPose=@n3
    OverrideWorldGravity="0,0,-980"
    SimSpaceSettings=(WorldAlpha=1,VelocityScaleZ=1,DampingAlpha=0,MaxLinearVelocity=5,MaxAngularVelocity=15,MaxLinearAcceleration=150,MaxAngularAcceleration=150,ExternalLinearDragV=(X=0,Y=0,Z=0),ExternalLinearVelocity=(X=0,Y=0,Z=0),ExternalAngularVelocity=(X=0,Y=0,Z=0))
   ) {
    bClampLinearTranslationLimitToRefPose=true
    bOverrideWorldGravity=true
    overridePhysicsAsset=&/M_MED_SummitReedGrit_StrapeR_Physics
    simulationSpace=BaseBoneSpace
    baseBoneRef {boneName=pelvis}
    overrideWorldGravity {z=-980}
    simSpaceSettings {dampingAlpha=0, maxAngularAcceleration=150, maxAngularVelocity=15, maxLinearAcceleration=150, maxLinearVelocity=5, worldAlpha=1}
   }
   n5 = Anim Dynamics - Bone: dyn_strape02_fr_l(ComponentPose=Local To Component(LocalPose=Use cached pose 'ShirtDynamics'()), Alpha=1, ExternalForce="0,0,0") {
    angularDampingOverride=0.99
    angularSpringConstant=450
    bAngularSpring=true
    bGravityOverrideInSimSpace=true
    bOverrideAngularDamping=true
    bUseGravityOverride=true
    simulationSpace=BoneRelative
    boundBone {boneName=dyn_strape02_fr_l}
    componentAppliedLinearAccClamp=<1500 1500 1800>
    componentLinearAccScale=<0.3 0.3 0.3>
    componentLinearVelScale=<0.3 0.3 0.3>
    gravityOverride {x=-980}
    physicsBodyDefinitions [{
      boundBone {boneName=dyn_strape02_fr_l}
      boxExtents=<12 12 12>
      constraintSetup {angularLimitsMax=<0 15 0>, angularLimitsMin=<0 0 0>, angularTarget=<1 0 0>}
      localJointOffset=<2 0 0>
     }]
    relativeSpaceBone {boneName=thigh_l}
    simSpaceSettings {maxAngularAcceleration=1500, maxAngularVelocity=1500, simSpaceAngularAlpha=0.3}
   }
   n6 = Anim Dynamics - Bone: dyn_strape03_fr_l(ComponentPose=@n5, Alpha=1, ExternalForce="0,0,0") {
    angularDampingOverride=0.99
    angularSpringConstant=450
    bAngularSpring=true
    bGravityOverrideInSimSpace=true
    bOverrideAngularDamping=true
    bUseGravityOverride=true
    simulationSpace=BoneRelative
    boundBone {boneName=dyn_strape03_fr_l}
    componentAppliedLinearAccClamp=<1500 1500 1800>
    componentLinearAccScale=<0.3 0.3 0.3>
    componentLinearVelScale=<0.3 0.3 0.3>
    gravityOverride {x=-980}
    physicsBodyDefinitions [{
      boundBone {boneName=dyn_strape03_fr_l}
      boxExtents=<12 12 12>
      constraintSetup {angularLimitsMax=<0 30 0>, angularLimitsMin=<0 0 0>, angularTarget=<1 0 0>}
      localJointOffset=<2 0 0>
     }]
    relativeSpaceBone {boneName=thigh_l}
    simSpaceSettings {maxAngularAcceleration=1500, maxAngularVelocity=1500, simSpaceAngularAlpha=0.3}
   }
   n7 = Anim Dynamics - Bone: dyn_strape04_fr_l(ComponentPose=@n6, Alpha=1, ExternalForce="0,0,0") {
    angularDampingOverride=0.99
    angularSpringConstant=450
    bAngularSpring=true
    bGravityOverrideInSimSpace=true
    bOverrideAngularDamping=true
    bUseGravityOverride=true
    simulationSpace=BoneRelative
    boundBone {boneName=dyn_strape04_fr_l}
    componentAppliedLinearAccClamp=<1500 1500 1800>
    componentLinearAccScale=<0.3 0.3 0.3>
    componentLinearVelScale=<0.3 0.3 0.3>
    gravityOverride {x=-980}
    physicsBodyDefinitions [{
      boundBone {boneName=dyn_strape04_fr_l}
      boxExtents=<12 12 12>
      constraintSetup {angularLimitsMax=<0 30 0>, angularLimitsMin=<0 0 0>, angularTarget=<1 0 0>}
      localJointOffset=<2 0 0>
     }]
    relativeSpaceBone {boneName=thigh_l}
    simSpaceSettings {maxAngularAcceleration=1500, maxAngularVelocity=1500, simSpaceAngularAlpha=0.3}
   }
   n8 = Anim Dynamics - Bone: dyn_strape05_fr_l(ComponentPose=@n7, Alpha=1, ExternalForce="0,0,0") {
    angularDampingOverride=0.99
    angularSpringConstant=450
    bAngularSpring=true
    bGravityOverrideInSimSpace=true
    bOverrideAngularDamping=true
    bUseGravityOverride=true
    simulationSpace=BoneRelative
    boundBone {boneName=dyn_strape05_fr_l}
    componentAppliedLinearAccClamp=<1500 1500 1800>
    componentLinearAccScale=<0.3 0.3 0.3>
    componentLinearVelScale=<0.3 0.3 0.3>
    gravityOverride {x=-980}
    physicsBodyDefinitions [{
      boundBone {boneName=dyn_strape05_fr_l}
      boxExtents=<12 12 12>
      constraintSetup {angularLimitsMax=<0 0 0>, angularLimitsMin=<0 0 0>, angularTarget=<1 0 0>}
      localJointOffset=<2 0 0>
     }]
    relativeSpaceBone {boneName=thigh_l}
    simSpaceSettings {maxAngularAcceleration=1500, maxAngularVelocity=1500, simSpaceAngularAlpha=0.3}
   }
   n9 = Anim Dynamics - Bone: dyn_strape02_bk_l(ComponentPose=@n8, Alpha=1, ExternalForce="0,0,0") {
    angularDampingOverride=0.99
    angularSpringConstant=450
    bAngularSpring=true
    bGravityOverrideInSimSpace=true
    bOverrideAngularDamping=true
    bUseGravityOverride=true
    simulationSpace=BoneRelative
    boundBone {boneName=dyn_strape02_bk_l}
    componentAppliedLinearAccClamp=<1500 1500 1800>
    componentLinearAccScale=<0.3 0.3 0.3>
    componentLinearVelScale=<0.3 0.3 0.3>
    gravityOverride {x=-980}
    physicsBodyDefinitions [{
      boundBone {boneName=dyn_strape02_bk_l}
      boxExtents=<12 12 12>
      constraintSetup {angularLimitsMax=<0 0 0>, angularLimitsMin=<0 -15 0>, angularTarget=<1 0 0>}
      localJointOffset=<2 0 0>
     }]
    relativeSpaceBone {boneName=thigh_l}
    simSpaceSettings {maxAngularAcceleration=1500, maxAngularVelocity=1500, simSpaceAngularAlpha=0.3}
   }
   n10 = Anim Dynamics - Bone: dyn_strape03_bk_l(ComponentPose=@n9, Alpha=1, ExternalForce="0,0,0") {
    angularDampingOverride=0.99
    angularSpringConstant=450
    bAngularSpring=true
    bGravityOverrideInSimSpace=true
    bOverrideAngularDamping=true
    bUseGravityOverride=true
    simulationSpace=BoneRelative
    boundBone {boneName=dyn_strape03_bk_l}
    componentAppliedLinearAccClamp=<1500 1500 1800>
    componentLinearAccScale=<0.3 0.3 0.3>
    componentLinearVelScale=<0.3 0.3 0.3>
    gravityOverride {x=-980}
    physicsBodyDefinitions [{
      boundBone {boneName=dyn_strape03_bk_l}
      boxExtents=<12 12 12>
      constraintSetup {angularLimitsMax=<0 0 0>, angularLimitsMin=<0 -30 0>, angularTarget=<1 0 0>}
      localJointOffset=<2 0 0>
     }]
    relativeSpaceBone {boneName=thigh_l}
    simSpaceSettings {maxAngularAcceleration=1500, maxAngularVelocity=1500, simSpaceAngularAlpha=0.3}
   }
   n11 = Anim Dynamics - Bone: dyn_strape04_bk_l(ComponentPose=@n10, Alpha=1, ExternalForce="0,0,0") {
    angularDampingOverride=0.99
    angularSpringConstant=450
    bAngularSpring=true
    bGravityOverrideInSimSpace=true
    bOverrideAngularDamping=true
    bUseGravityOverride=true
    simulationSpace=BoneRelative
    boundBone {boneName=dyn_strape04_bk_l}
    componentAppliedLinearAccClamp=<1500 1500 1800>
    componentLinearAccScale=<0.3 0.3 0.3>
    componentLinearVelScale=<0.3 0.3 0.3>
    gravityOverride {x=-980}
    physicsBodyDefinitions [{
      boundBone {boneName=dyn_strape04_bk_l}
      boxExtents=<12 12 12>
      constraintSetup {angularLimitsMax=<0 0 0>, angularLimitsMin=<0 -30 0>, angularTarget=<1 0 0>}
      localJointOffset=<2 0 0>
     }]
    relativeSpaceBone {boneName=thigh_l}
    simSpaceSettings {maxAngularAcceleration=1500, maxAngularVelocity=1500, simSpaceAngularAlpha=0.3}
   }
   n12 = Anim Dynamics - Bone: dyn_strape05_bk_l(ComponentPose=@n11, Alpha=1, ExternalForce="0,0,0") {
    angularDampingOverride=0.99
    angularSpringConstant=450
    bAngularSpring=true
    bGravityOverrideInSimSpace=true
    bOverrideAngularDamping=true
    bUseGravityOverride=true
    simulationSpace=BoneRelative
    boundBone {boneName=dyn_strape05_bk_l}
    componentAppliedLinearAccClamp=<1500 1500 1800>
    componentLinearAccScale=<0.3 0.3 0.3>
    componentLinearVelScale=<0.3 0.3 0.3>
    gravityOverride {x=-980}
    physicsBodyDefinitions [{
      boundBone {boneName=dyn_strape05_bk_l}
      boxExtents=<12 12 12>
      constraintSetup {angularLimitsMax=<0 0 0>, angularLimitsMin=<0 0 0>, angularTarget=<1 0 0>}
      localJointOffset=<2 0 0>
     }]
    relativeSpaceBone {boneName=thigh_l}
    simSpaceSettings {maxAngularAcceleration=1500, maxAngularVelocity=1500, simSpaceAngularAlpha=0.3}
   }
   n13 = Anim Dynamics - Bone: dyn_strape02_fr_r(ComponentPose=@n12, Alpha=1, ExternalForce="0,0,0") {
    angularDampingOverride=0.99
    angularSpringConstant=450
    bAngularSpring=true
    bGravityOverrideInSimSpace=true
    bOverrideAngularDamping=true
    bUseGravityOverride=true
    simulationSpace=BoneRelative
    boundBone {boneName=dyn_strape02_fr_r}
    componentAppliedLinearAccClamp=<1500 1500 1800>
    componentLinearAccScale=<0.3 0.3 0.3>
    componentLinearVelScale=<0.3 0.3 0.3>
    gravityOverride {x=980}
    physicsBodyDefinitions [{
      boundBone {boneName=dyn_strape02_fr_r}
      boxExtents=<12 12 12>
      constraintSetup {angularLimitsMax=<0 15 0>, angularLimitsMin=<0 0 0>, angularTarget=<1 0 0>}
      localJointOffset=<-2 0 0>
     }]
    relativeSpaceBone {boneName=thigh_r}
    simSpaceSettings {maxAngularAcceleration=1500, maxAngularVelocity=1500, simSpaceAngularAlpha=0.3}
   }
   n14 = Anim Dynamics - Bone: dyn_strape03_fr_r(ComponentPose=@n13, Alpha=1, ExternalForce="0,0,0") {
    angularDampingOverride=0.99
    angularSpringConstant=450
    bAngularSpring=true
    bGravityOverrideInSimSpace=true
    bOverrideAngularDamping=true
    bUseGravityOverride=true
    simulationSpace=BoneRelative
    boundBone {boneName=dyn_strape03_fr_r}
    componentAppliedLinearAccClamp=<1500 1500 1800>
    componentLinearAccScale=<0.3 0.3 0.3>
    componentLinearVelScale=<0.3 0.3 0.3>
    gravityOverride {x=980}
    physicsBodyDefinitions [{
      boundBone {boneName=dyn_strape03_fr_r}
      boxExtents=<12 12 12>
      constraintSetup {angularLimitsMax=<0 30 0>, angularLimitsMin=<0 0 0>, angularTarget=<1 0 0>}
      localJointOffset=<-2 0 0>
     }]
    relativeSpaceBone {boneName=thigh_r}
    simSpaceSettings {maxAngularAcceleration=1500, maxAngularVelocity=1500, simSpaceAngularAlpha=0.3}
   }
   n15 = Anim Dynamics - Bone: dyn_strape04_fr_r(ComponentPose=@n14, Alpha=1, ExternalForce="0,0,0") {
    angularDampingOverride=0.99
    angularSpringConstant=450
    bAngularSpring=true
    bGravityOverrideInSimSpace=true
    bOverrideAngularDamping=true
    bUseGravityOverride=true
    simulationSpace=BoneRelative
    boundBone {boneName=dyn_strape04_fr_r}
    componentAppliedLinearAccClamp=<1500 1500 1800>
    componentLinearAccScale=<0.3 0.3 0.3>
    componentLinearVelScale=<0.3 0.3 0.3>
    gravityOverride {x=980}
    physicsBodyDefinitions [{
      boundBone {boneName=dyn_strape04_fr_r}
      boxExtents=<12 12 12>
      constraintSetup {angularLimitsMax=<0 30 0>, angularLimitsMin=<0 0 0>, angularTarget=<1 0 0>}
      localJointOffset=<-2 0 0>
     }]
    relativeSpaceBone {boneName=thigh_r}
    simSpaceSettings {maxAngularAcceleration=1500, maxAngularVelocity=1500, simSpaceAngularAlpha=0.3}
   }
   n16 = Anim Dynamics - Bone: dyn_strape05_fr_r(ComponentPose=@n15, Alpha=1, ExternalForce="0,0,0") {
    angularDampingOverride=0.99
    angularSpringConstant=450
    bAngularSpring=true
    bGravityOverrideInSimSpace=true
    bOverrideAngularDamping=true
    bUseGravityOverride=true
    simulationSpace=BoneRelative
    boundBone {boneName=dyn_strape05_fr_r}
    componentAppliedLinearAccClamp=<1500 1500 1800>
    componentLinearAccScale=<0.3 0.3 0.3>
    componentLinearVelScale=<0.3 0.3 0.3>
    gravityOverride {x=980}
    physicsBodyDefinitions [{
      boundBone {boneName=dyn_strape05_fr_r}
      boxExtents=<12 12 12>
      constraintSetup {angularLimitsMax=<0 0 0>, angularLimitsMin=<0 0 0>, angularTarget=<1 0 0>}
      localJointOffset=<-2 0 0>
     }]
    relativeSpaceBone {boneName=thigh_r}
    simSpaceSettings {maxAngularAcceleration=1500, maxAngularVelocity=1500, simSpaceAngularAlpha=0.3}
   }
   n17 = Anim Dynamics - Bone: dyn_strape02_bk_r(ComponentPose=@n16, Alpha=1, ExternalForce="0,0,0") {
    angularDampingOverride=0.99
    angularSpringConstant=450
    bAngularSpring=true
    bGravityOverrideInSimSpace=true
    bOverrideAngularDamping=true
    bUseGravityOverride=true
    simulationSpace=BoneRelative
    boundBone {boneName=dyn_strape02_bk_r}
    componentAppliedLinearAccClamp=<1500 1500 1800>
    componentLinearAccScale=<0.3 0.3 0.3>
    componentLinearVelScale=<0.3 0.3 0.3>
    gravityOverride {x=980}
    physicsBodyDefinitions [{
      boundBone {boneName=dyn_strape02_bk_r}
      boxExtents=<12 12 12>
      constraintSetup {angularLimitsMax=<0 0 0>, angularLimitsMin=<0 -15 0>, angularTarget=<1 0 0>}
      localJointOffset=<-2 0 0>
     }]
    relativeSpaceBone {boneName=thigh_r}
    simSpaceSettings {maxAngularAcceleration=1500, maxAngularVelocity=1500, simSpaceAngularAlpha=0.3}
   }
   n18 = Anim Dynamics - Bone: dyn_strape03_bk_r(ComponentPose=@n17, Alpha=1, ExternalForce="0,0,0") {
    angularDampingOverride=0.99
    angularSpringConstant=450
    bAngularSpring=true
    bGravityOverrideInSimSpace=true
    bOverrideAngularDamping=true
    bUseGravityOverride=true
    simulationSpace=BoneRelative
    boundBone {boneName=dyn_strape03_bk_r}
    componentAppliedLinearAccClamp=<1500 1500 1800>
    componentLinearAccScale=<0.3 0.3 0.3>
    componentLinearVelScale=<0.3 0.3 0.3>
    gravityOverride {x=980}
    physicsBodyDefinitions [{
      boundBone {boneName=dyn_strape03_bk_r}
      boxExtents=<12 12 12>
      constraintSetup {angularLimitsMax=<0 0 0>, angularLimitsMin=<0 -30 0>, angularTarget=<1 0 0>}
      localJointOffset=<-2 0 0>
     }]
    relativeSpaceBone {boneName=thigh_r}
    simSpaceSettings {maxAngularAcceleration=1500, maxAngularVelocity=1500, simSpaceAngularAlpha=0.3}
   }
   n19 = Anim Dynamics - Bone: dyn_strape04_bk_r(ComponentPose=@n18, Alpha=1, ExternalForce="0,0,0") {
    angularDampingOverride=0.99
    angularSpringConstant=450
    bAngularSpring=true
    bGravityOverrideInSimSpace=true
    bOverrideAngularDamping=true
    bUseGravityOverride=true
    simulationSpace=BoneRelative
    boundBone {boneName=dyn_strape04_bk_r}
    componentAppliedLinearAccClamp=<1500 1500 1800>
    componentLinearAccScale=<0.3 0.3 0.3>
    componentLinearVelScale=<0.3 0.3 0.3>
    gravityOverride {x=980}
    physicsBodyDefinitions [{
      boundBone {boneName=dyn_strape04_bk_r}
      boxExtents=<12 12 12>
      constraintSetup {angularLimitsMax=<0 0 0>, angularLimitsMin=<0 -30 0>, angularTarget=<1 0 0>}
      localJointOffset=<-2 0 0>
     }]
    relativeSpaceBone {boneName=thigh_r}
    simSpaceSettings {maxAngularAcceleration=1500, maxAngularVelocity=1500, simSpaceAngularAlpha=0.3}
   }
   n20 = Anim Dynamics - Bone: dyn_strape05_bk_r(ComponentPose=@n19, Alpha=1, ExternalForce="0,0,0") {
    angularDampingOverride=0.99
    angularSpringConstant=450
    bAngularSpring=true
    bGravityOverrideInSimSpace=true
    bOverrideAngularDamping=true
    bUseGravityOverride=true
    simulationSpace=BoneRelative
    boundBone {boneName=dyn_strape05_bk_r}
    componentAppliedLinearAccClamp=<1500 1500 1800>
    componentLinearAccScale=<0.3 0.3 0.3>
    componentLinearVelScale=<0.3 0.3 0.3>
    gravityOverride {x=980}
    physicsBodyDefinitions [{
      boundBone {boneName=dyn_strape05_bk_r}
      boxExtents=<12 12 12>
      constraintSetup {angularLimitsMax=<0 0 0>, angularLimitsMin=<0 0 0>, angularTarget=<1 0 0>}
      localJointOffset=<-2 0 0>
     }]
    relativeSpaceBone {boneName=thigh_r}
    simSpaceSettings {maxAngularAcceleration=1500, maxAngularVelocity=1500, simSpaceAngularAlpha=0.3}
   }
   Output Pose(Result=Blend Poses by bool(BlendPose_0=Component To Local(ComponentPose=@n4), BlendPose_1=Component To Local(ComponentPose=@n20), bActiveValue=Get bRigidBodyEnabled())) {layerGroup=None}
   n21 = Control Rig!(
    Source=Copy Pose From Mesh() {bUseAttachedParent=true}
    Alpha
    isDBNO:bool=Get bIsDBNO()
    isCrouching:bool=Get bIsCrouching()
    isSliding:bool=Get bIsSliding()
   ) -> (Pose) {controlRigAssetReference {blueprintRigClass=&/M_MED_SummitReedGrit_CtrlRig.M_MED_SummitReedGrit_CtrlRig_C}}
   Save cached pose 'Deformers'(Pose=@n21) {cacheName=Deformers}
   n22 = Anim Dynamics - Bone: dyn_pants_cloth_05_l(ComponentPose=Local To Component(LocalPose=Use cached pose 'Deformers'()), Alpha=1, ExternalForce="0,0,0") {
    angularDampingOverride=0.99
    angularSpringConstant=600
    bAngularSpring=true
    bLinearSpring=true
    bOverrideAngularDamping=true
    bOverrideLinearDamping=true
    bUseGravityOverride=true
    linearDampingOverride=0.9999
    linearSpringConstant=150
    simulationSpace=RootRelative
    boundBone {boneName=dyn_pants_cloth_05_l}
    componentAppliedLinearAccClamp=<3000 3000 3000>
    componentLinearAccScale=<0.2 0.2 0.5>
    componentLinearVelScale=<0.2 0.2 0.2>
    gravityOverride {z=-980}
    physicsBodyDefinitions [{
      boundBone {boneName=dyn_pants_cloth_05_l}
      boxExtents=<12 12 12>
      constraintSetup {angularLimitsMax=<0 10 10>, angularLimitsMin=<0 -10 -10>, angularTarget=<0 0 0>, linearAxesMax=<0 1 3>, linearAxesMin=<0 0 -0.2>}
      localJointOffset=<0 0 1>
     }]
    relativeSpaceBone {boneName=calf_l}
    simSpaceSettings {maxAngularAcceleration=1500, maxAngularVelocity=1500, simSpaceAngularAlpha=0.1}
   }
   n23 = Anim Dynamics - Bone: dyn_pants_cloth_04_l(ComponentPose=@n22, Alpha=1, ExternalForce="0,0,0") {
    angularDampingOverride=0.99
    angularSpringConstant=600
    bAngularSpring=true
    bLinearSpring=true
    bOverrideAngularDamping=true
    bOverrideLinearDamping=true
    bUseGravityOverride=true
    linearDampingOverride=0.9999
    linearSpringConstant=150
    simulationSpace=RootRelative
    boundBone {boneName=dyn_pants_cloth_04_l}
    componentAppliedLinearAccClamp=<3000 3000 3000>
    componentLinearAccScale=<0.2 0.2 0.5>
    componentLinearVelScale=<0.2 0.2 0.2>
    gravityOverride {z=-980}
    physicsBodyDefinitions [{
      boundBone {boneName=dyn_pants_cloth_04_l}
      boxExtents=<12 12 12>
      constraintSetup {angularLimitsMax=<0 10 10>, angularLimitsMin=<0 -10 -10>, angularTarget=<0 0 0>, linearAxesMax=<0 1 3>, linearAxesMin=<0 0 -0.2>}
      localJointOffset=<0 0 1>
     }]
    relativeSpaceBone {boneName=calf_l}
    simSpaceSettings {maxAngularAcceleration=1500, maxAngularVelocity=1500, simSpaceAngularAlpha=0.1}
   }
   n24 = Anim Dynamics - Bone: dyn_pants_cloth_01_l(ComponentPose=@n23, Alpha=1, ExternalForce="0,0,0") {
    angularDampingOverride=0.99
    angularSpringConstant=600
    bAngularSpring=true
    bLinearSpring=true
    bOverrideAngularDamping=true
    bOverrideLinearDamping=true
    bUseGravityOverride=true
    linearDampingOverride=0.9999
    linearSpringConstant=150
    simulationSpace=RootRelative
    boundBone {boneName=dyn_pants_cloth_01_l}
    componentAppliedLinearAccClamp=<3000 3000 3000>
    componentLinearAccScale=<0.2 0.2 0.5>
    componentLinearVelScale=<0.2 0.2 0.2>
    gravityOverride {z=-980}
    physicsBodyDefinitions [{
      boundBone {boneName=dyn_pants_cloth_01_l}
      boxExtents=<12 12 12>
      constraintSetup {angularLimitsMax=<0 10 10>, angularLimitsMin=<0 -10 -10>, angularTarget=<0 0 0>, linearAxesMax=<0 1 3>, linearAxesMin=<0 0 -0.2>}
      localJointOffset=<0 0 1>
     }]
    relativeSpaceBone {boneName=calf_l}
    simSpaceSettings {maxAngularAcceleration=1500, maxAngularVelocity=1500, simSpaceAngularAlpha=0.1}
   }
   n25 = Anim Dynamics - Bone: dyn_pants_cloth_03_l(ComponentPose=@n24, Alpha=1, ExternalForce="0,0,0") {
    angularDampingOverride=0.99
    angularSpringConstant=600
    bAngularSpring=true
    bLinearSpring=true
    bOverrideAngularDamping=true
    bOverrideLinearDamping=true
    bUseGravityOverride=true
    linearDampingOverride=0.9999
    linearSpringConstant=150
    simulationSpace=RootRelative
    boundBone {boneName=dyn_pants_cloth_03_l}
    componentAppliedLinearAccClamp=<3000 3000 3000>
    componentLinearAccScale=<0.2 0.2 0.5>
    componentLinearVelScale=<0.2 0.2 0.2>
    gravityOverride {z=-980}
    physicsBodyDefinitions [{
      boundBone {boneName=dyn_pants_cloth_03_l}
      boxExtents=<12 12 12>
      constraintSetup {angularLimitsMax=<0 10 10>, angularLimitsMin=<0 -10 -10>, angularTarget=<0 0 0>, linearAxesMax=<0 1 3>, linearAxesMin=<0 0 -0.2>}
      localJointOffset=<0 0 1>
     }]
    relativeSpaceBone {boneName=calf_l}
    simSpaceSettings {maxAngularAcceleration=1500, maxAngularVelocity=1500, simSpaceAngularAlpha=0.1}
   }
   n26 = Anim Dynamics - Bone: dyn_pants_cloth_02_l(ComponentPose=@n25, Alpha=1, ExternalForce="0,0,0") {
    angularDampingOverride=0.99
    angularSpringConstant=600
    bAngularSpring=true
    bLinearSpring=true
    bOverrideAngularDamping=true
    bOverrideLinearDamping=true
    bUseGravityOverride=true
    linearDampingOverride=0.9999
    linearSpringConstant=150
    simulationSpace=RootRelative
    boundBone {boneName=dyn_pants_cloth_02_l}
    componentAppliedLinearAccClamp=<3000 3000 3000>
    componentLinearAccScale=<0.2 0.2 0.5>
    componentLinearVelScale=<0.2 0.2 0.2>
    gravityOverride {z=-980}
    physicsBodyDefinitions [{
      boundBone {boneName=dyn_pants_cloth_02_l}
      boxExtents=<12 12 12>
      constraintSetup {angularLimitsMax=<0 10 10>, angularLimitsMin=<0 -10 -10>, angularTarget=<0 0 0>, linearAxesMax=<0 1 3>, linearAxesMin=<0 0 -0.2>}
      localJointOffset=<0 0 1>
     }]
    relativeSpaceBone {boneName=calf_l}
    simSpaceSettings {maxAngularAcceleration=1500, maxAngularVelocity=1500, simSpaceAngularAlpha=0.1}
   }
   n27 = Anim Dynamics - Bone: dyn_pants_cloth_05_r(ComponentPose=@n26, Alpha=1, ExternalForce="0,0,0") {
    angularDampingOverride=0.99
    angularSpringConstant=600
    bAngularSpring=true
    bLinearSpring=true
    bOverrideAngularDamping=true
    bOverrideLinearDamping=true
    bUseGravityOverride=true
    linearDampingOverride=0.9999
    linearSpringConstant=150
    simulationSpace=RootRelative
    boundBone {boneName=dyn_pants_cloth_05_r}
    componentAppliedLinearAccClamp=<3000 3000 3000>
    componentLinearAccScale=<0.2 0.2 0.5>
    componentLinearVelScale=<0.2 0.2 0.2>
    gravityOverride {z=-980}
    physicsBodyDefinitions [{
      boundBone {boneName=dyn_pants_cloth_05_r}
      boxExtents=<11 11 11>
      constraintSetup {angularLimitsMax=<0 10 10>, angularLimitsMin=<0 -10 -10>, angularTarget=<0 0 0>, linearAxesMax=<0 0 3>, linearAxesMin=<0 0 -0.1>}
      localJointOffset=<0 0 -1>
     }]
    relativeSpaceBone {boneName=calf_r}
    simSpaceSettings {maxAngularAcceleration=1500, maxAngularVelocity=1500, simSpaceAngularAlpha=0.1}
   }
   n28 = Anim Dynamics - Bone: dyn_pants_cloth_04_r(ComponentPose=@n27, Alpha=1, ExternalForce="0,0,0") {
    angularDampingOverride=0.99
    angularSpringConstant=600
    bAngularSpring=true
    bLinearSpring=true
    bOverrideAngularDamping=true
    bOverrideLinearDamping=true
    bUseGravityOverride=true
    linearDampingOverride=0.9999
    linearSpringConstant=150
    simulationSpace=RootRelative
    boundBone {boneName=dyn_pants_cloth_04_r}
    componentAppliedLinearAccClamp=<3000 3000 3000>
    componentLinearAccScale=<0.2 0.2 0.5>
    componentLinearVelScale=<0.2 0.2 0.2>
    gravityOverride {z=-980}
    physicsBodyDefinitions [{
      boundBone {boneName=dyn_pants_cloth_04_r}
      boxExtents=<11 11 11>
      constraintSetup {angularLimitsMax=<0 10 10>, angularLimitsMin=<0 -10 -10>, angularTarget=<0 0 0>, linearAxesMax=<0 0 3>, linearAxesMin=<0 0 -0.1>}
      localJointOffset=<0 0 -1>
     }]
    relativeSpaceBone {boneName=calf_r}
    simSpaceSettings {maxAngularAcceleration=1500, maxAngularVelocity=1500, simSpaceAngularAlpha=0.1}
   }
   n29 = Anim Dynamics - Bone: dyn_pants_cloth_03_r(ComponentPose=@n28, Alpha=1, ExternalForce="0,0,0") {
    angularDampingOverride=0.99
    angularSpringConstant=600
    bAngularSpring=true
    bLinearSpring=true
    bOverrideAngularDamping=true
    bOverrideLinearDamping=true
    bUseGravityOverride=true
    linearDampingOverride=0.9999
    linearSpringConstant=150
    simulationSpace=RootRelative
    boundBone {boneName=dyn_pants_cloth_03_r}
    componentAppliedLinearAccClamp=<3000 3000 3000>
    componentLinearAccScale=<0.2 0.2 0.5>
    componentLinearVelScale=<0.2 0.2 0.2>
    gravityOverride {z=-980}
    physicsBodyDefinitions [{
      boundBone {boneName=dyn_pants_cloth_03_r}
      boxExtents=<11 11 11>
      constraintSetup {angularLimitsMax=<0 10 10>, angularLimitsMin=<0 -10 -10>, angularTarget=<0 0 0>, linearAxesMax=<0 0 3>, linearAxesMin=<0 0 -0.1>}
      localJointOffset=<0 0 -1>
     }]
    relativeSpaceBone {boneName=calf_r}
    simSpaceSettings {maxAngularAcceleration=1500, maxAngularVelocity=1500, simSpaceAngularAlpha=0.1}
   }
   n30 = Anim Dynamics - Bone: dyn_pants_cloth_02_r(ComponentPose=@n29, Alpha=1, ExternalForce="0,0,0") {
    angularDampingOverride=0.99
    angularSpringConstant=600
    bAngularSpring=true
    bLinearSpring=true
    bOverrideAngularDamping=true
    bOverrideLinearDamping=true
    bUseGravityOverride=true
    linearDampingOverride=0.9999
    linearSpringConstant=150
    simulationSpace=RootRelative
    boundBone {boneName=dyn_pants_cloth_02_r}
    componentAppliedLinearAccClamp=<3000 3000 3000>
    componentLinearAccScale=<0.2 0.2 0.5>
    componentLinearVelScale=<0.2 0.2 0.2>
    gravityOverride {z=-980}
    physicsBodyDefinitions [{
      boundBone {boneName=dyn_pants_cloth_02_r}
      boxExtents=<11 11 11>
      constraintSetup {angularLimitsMax=<0 10 10>, angularLimitsMin=<0 -10 -10>, angularTarget=<0 0 0>, linearAxesMax=<0 0 3>, linearAxesMin=<0 0 -0.1>}
      localJointOffset=<0 0 -1>
     }]
    relativeSpaceBone {boneName=calf_r}
    simSpaceSettings {maxAngularAcceleration=1500, maxAngularVelocity=1500, simSpaceAngularAlpha=0.1}
   }
   Save cached pose 'PantsAnimDynamics'(Pose=Component To Local(ComponentPose=@n30)) {cacheName=PantsAnimDynamics}
   n31 = Anim Dynamics - Bone: dyn_shoeTag01_l(ComponentPose=Local To Component(LocalPose=Use cached pose 'PantsAnimDynamics'()), Alpha=1, ExternalForce="0,0,0") {
    angularDampingOverride=0.99
    angularSpringConstant=750
    bAngularSpring=true
    bOverrideAngularDamping=true
    bUseGravityOverride=true
    simulationSpace=RootRelative
    boundBone {boneName=dyn_shoeTag01_l}
    componentAppliedLinearAccClamp=<3300 3300 3300>
    componentLinearAccScale=<0.4 0.4 0.4>
    componentLinearVelScale=<0.4 0.4 0.5>
    gravityOverride {z=-980}
    physicsBodyDefinitions [{
      boundBone {boneName=dyn_shoeTag01_l}
      boxExtents=<16.2 16.2 16.2>
      constraintSetup {angularLimitsMax=<0 0 5>, angularLimitsMin=<0 -30 -10>, angularTarget=<1 0 0>}
      localJointOffset=<1.1 0 0>
     }]
    simSpaceSettings {maxAngularAcceleration=1800, maxAngularVelocity=1800, simSpaceAngularAlpha=0.11}
   }
   n32 = Anim Dynamics - Bone: dyn_shoeTag02_l(ComponentPose=@n31, Alpha=1, ExternalForce="0,0,0") {
    angularDampingOverride=0.99
    angularSpringConstant=750
    bAngularSpring=true
    bOverrideAngularDamping=true
    bUseGravityOverride=true
    simulationSpace=RootRelative
    boundBone {boneName=dyn_shoeTag02_l}
    componentAppliedLinearAccClamp=<3300 3300 3300>
    componentLinearAccScale=<0.4 0.4 0.4>
    componentLinearVelScale=<0.4 0.4 0.5>
    gravityOverride {z=-980}
    physicsBodyDefinitions [{
      boundBone {boneName=dyn_shoeTag02_l}
      boxExtents=<16.2 16.2 16.2>
      constraintSetup {angularLimitsMax=<0 0 5>, angularLimitsMin=<0 -30 -10>, angularTarget=<1 0 0>}
      localJointOffset=<1.1 0 0>
     }]
    simSpaceSettings {maxAngularAcceleration=1800, maxAngularVelocity=1800, simSpaceAngularAlpha=0.11}
   }
   n33 = Anim Dynamics - Bone: dyn_shoeTag01_r(ComponentPose=@n32, Alpha=1, ExternalForce="0,0,0") {
    angularDampingOverride=0.99
    angularSpringConstant=750
    bAngularSpring=true
    bOverrideAngularDamping=true
    bUseGravityOverride=true
    simulationSpace=RootRelative
    boundBone {boneName=dyn_shoeTag01_r}
    componentAppliedLinearAccClamp=<3300 3300 3300>
    componentLinearAccScale=<0.4 0.4 0.4>
    componentLinearVelScale=<0.4 0.4 0.7>
    gravityOverride {z=-980}
    physicsBodyDefinitions [{
      boundBone {boneName=dyn_shoeTag01_r}
      boxExtents=<14 14 14>
      constraintSetup {angularLimitsMax=<0 0 5>, angularLimitsMin=<0 -30 -10>, angularTarget=<1 0 0>}
      localJointOffset=<-1.1 0 0>
     }]
    simSpaceSettings {maxAngularAcceleration=1800, maxAngularVelocity=1800, simSpaceAngularAlpha=0.11}
   }
   n34 = Anim Dynamics - Bone: dyn_shoeTag02_r(ComponentPose=@n33, Alpha=1, ExternalForce="0,0,0") {
    angularDampingOverride=0.99
    angularSpringConstant=750
    bAngularSpring=true
    bOverrideAngularDamping=true
    bUseGravityOverride=true
    simulationSpace=RootRelative
    boundBone {boneName=dyn_shoeTag02_r}
    componentAppliedLinearAccClamp=<3300 3300 3300>
    componentLinearAccScale=<0.4 0.4 0.4>
    componentLinearVelScale=<0.4 0.4 0.7>
    gravityOverride {z=-980}
    physicsBodyDefinitions [{
      boundBone {boneName=dyn_shoeTag02_r}
      boxExtents=<14 14 14>
      constraintSetup {angularLimitsMax=<0 0 5>, angularLimitsMin=<0 -30 -10>, angularTarget=<1 0 0>}
      localJointOffset=<-1.1 0 0>
     }]
    simSpaceSettings {maxAngularAcceleration=1800, maxAngularVelocity=1800, simSpaceAngularAlpha=0.11}
   }
   Save cached pose 'ShoeTags'(Pose=Component To Local(ComponentPose=@n34)) {cacheName=ShoeTags}
   n35 = "Constraint - dyn_rope01_l (PPPPPPPP)"!(
    ComponentPose=Local To Component(LocalPose=Use cached pose 'ShoeTags'())
    Alpha=1
    ConstraintWeights_0=0.0195
    ConstraintWeights_1:float=0.0429
    ConstraintWeights_2:float=0.7706
    ConstraintWeights_3:float=0.0865
    ConstraintWeights_4:float=0.0239
    ConstraintWeights_5:float=0.0083
    ConstraintWeights_6:float=0.0208
    ConstraintWeights_7:float=0.0275
   ) -> (Pose) {
    boneToModify {boneName=dyn_rope01_l}
    constraintSetup [
     {targetBone {boneName=spine_03}}
     {targetBone {boneName=spine_04}}
     {targetBone {boneName=spine_05}}
     {targetBone {boneName=clavicle_l}}
     {targetBone {boneName=deform_upperarm_fr_l}}
     {targetBone {boneName=deform_upperarm_up_l}}
     {targetBone {boneName=neck_01}}
     {targetBone {boneName=deform_pec_l}}
    ]
    constraintWeights [1,1,1,1,1,1,1,1]
   }
   n36 = "Constraint - dyn_rope02_l (PPPPPPPP)"!(
    ComponentPose=@n35
    Alpha=1
    ConstraintWeights_0=0.0844
    ConstraintWeights_1:float=0.2099
    ConstraintWeights_2:float=0.3826
    ConstraintWeights_3:float=0.161
    ConstraintWeights_4:float=0.0396
    ConstraintWeights_5:float=0.005
    ConstraintWeights_6:float=0.0024
    ConstraintWeights_7:float=0.1153
   ) -> (Pose) {
    boneToModify {boneName=dyn_rope02_l}
    constraintSetup [
     {targetBone {boneName=spine_03}}
     {targetBone {boneName=spine_04}}
     {targetBone {boneName=spine_05}}
     {targetBone {boneName=clavicle_l}}
     {targetBone {boneName=deform_upperarm_fr_l}}
     {targetBone {boneName=neck_01}}
     {targetBone {boneName=neck_02}}
     {targetBone {boneName=deform_pec_l}}
    ]
    constraintWeights [1,1,1,1,1,1,1,1]
   }
   n37 = "Constraint - dyn_rope03_l (PPPPPPPP)"!(
    ComponentPose=@n36
    Alpha=1
    ConstraintWeights_0=0.1501
    ConstraintWeights_1:float=0.307
    ConstraintWeights_2:float=0.2497
    ConstraintWeights_3:float=0.1152
    ConstraintWeights_4:float=0.0425
    ConstraintWeights_5:float=0.0016
    ConstraintWeights_6:float=0.0015
    ConstraintWeights_7:float=0.1324
   ) -> (Pose) {
    boneToModify {boneName=dyn_rope03_l}
    constraintSetup [
     {targetBone {boneName=spine_03}}
     {targetBone {boneName=spine_04}}
     {targetBone {boneName=spine_05}}
     {targetBone {boneName=clavicle_l}}
     {targetBone {boneName=deform_upperarm_fr_l}}
     {targetBone {boneName=neck_01}}
     {targetBone {boneName=neck_02}}
     {targetBone {boneName=deform_pec_l}}
    ]
    constraintWeights [1,1,1,1,1,1,1,1]
   }
   n38 = "Constraint - dyn_rope04_l (PPP)"!(ComponentPose=@n37, Alpha=1, ConstraintWeights_0=0.2731, ConstraintWeights_1:float=0.5, ConstraintWeights_2:float=0.2269) -> (Pose) {
    boneToModify {boneName=dyn_rope04_l}
    constraintSetup [{targetBone {boneName=spine_03}}, {targetBone {boneName=spine_04}}, {targetBone {boneName=spine_05}}]
    constraintWeights [1,1,1]
   }
   n39 = "Constraint - dyn_rope05_l (PPPPP)"!(
    ComponentPose=@n38
    Alpha=1
    ConstraintWeights_0=0
    ConstraintWeights_1:float=0.4456
    ConstraintWeights_2:float=0.4192
    ConstraintWeights_3:float=0.1352
    ConstraintWeights_4:float=0
   ) -> (Pose) {
    boneToModify {boneName=dyn_rope05_l}
    constraintSetup [
     {targetBone {boneName=spine_02}}
     {targetBone {boneName=spine_03}}
     {targetBone {boneName=spine_04}}
     {targetBone {boneName=spine_05}}
     {targetBone {boneName=deform_pec_l}}
    ]
    constraintWeights [1,1,1,1,1]
   }
   n40 = "Constraint - dyn_rope01_r (PPPPPPPP)"!(
    ComponentPose=@n39
    Alpha=1
    ConstraintWeights_0=0.0195
    ConstraintWeights_1:float=0.0429
    ConstraintWeights_2:float=0.7706
    ConstraintWeights_3:float=0.0865
    ConstraintWeights_4:float=0.0239
    ConstraintWeights_5:float=0.0083
    ConstraintWeights_6:float=0.0208
    ConstraintWeights_7:float=0.0275
   ) -> (Pose) {
    boneToModify {boneName=dyn_rope01_r}
    constraintSetup [
     {targetBone {boneName=spine_03}}
     {targetBone {boneName=spine_04}}
     {targetBone {boneName=spine_05}}
     {targetBone {boneName=clavicle_r}}
     {targetBone {boneName=deform_upperarm_fr_r}}
     {targetBone {boneName=deform_upperarm_up_r}}
     {targetBone {boneName=neck_01}}
     {targetBone {boneName=deform_pec_r}}
    ]
    constraintWeights [1,1,1,1,1,1,1,1]
   }
   n41 = "Constraint - dyn_rope02_r (PPPPPPPP)"!(
    ComponentPose=@n40
    Alpha=1
    ConstraintWeights_0=0.0844
    ConstraintWeights_1:float=0.2099
    ConstraintWeights_2:float=0.3826
    ConstraintWeights_3:float=0.161
    ConstraintWeights_4:float=0.0396
    ConstraintWeights_5:float=0.005
    ConstraintWeights_6:float=0.0024
    ConstraintWeights_7:float=0.1153
   ) -> (Pose) {
    boneToModify {boneName=dyn_rope02_r}
    constraintSetup [
     {targetBone {boneName=spine_03}}
     {targetBone {boneName=spine_04}}
     {targetBone {boneName=spine_05}}
     {targetBone {boneName=clavicle_r}}
     {targetBone {boneName=deform_upperarm_fr_r}}
     {targetBone {boneName=neck_01}}
     {targetBone {boneName=neck_02}}
     {targetBone {boneName=deform_pec_r}}
    ]
    constraintWeights [1,1,1,1,1,1,1,1]
   }
   n42 = "Constraint - dyn_rope03_r (PPPPPPPP)"!(
    ComponentPose=@n41
    Alpha=1
    ConstraintWeights_0=0.1501
    ConstraintWeights_1:float=0.307
    ConstraintWeights_2:float=0.2497
    ConstraintWeights_3:float=0.1152
    ConstraintWeights_4:float=0.0425
    ConstraintWeights_5:float=0.0016
    ConstraintWeights_6:float=0.0015
    ConstraintWeights_7:float=0.1324
   ) -> (Pose) {
    boneToModify {boneName=dyn_rope03_r}
    constraintSetup [
     {targetBone {boneName=spine_03}}
     {targetBone {boneName=spine_04}}
     {targetBone {boneName=spine_05}}
     {targetBone {boneName=clavicle_r}}
     {targetBone {boneName=deform_upperarm_fr_r}}
     {targetBone {boneName=neck_01}}
     {targetBone {boneName=neck_02}}
     {targetBone {boneName=deform_pec_r}}
    ]
    constraintWeights [1,1,1,1,1,1,1,1]
   }
   n43 = "Constraint - dyn_rope04_r (PPP)"!(ComponentPose=@n42, Alpha=1, ConstraintWeights_0=0.2731, ConstraintWeights_1:float=0.5, ConstraintWeights_2:float=0.2269) -> (Pose) {
    boneToModify {boneName=dyn_rope04_r}
    constraintSetup [{targetBone {boneName=spine_03}}, {targetBone {boneName=spine_04}}, {targetBone {boneName=spine_05}}]
    constraintWeights [1,1,1]
   }
   n44 = "Constraint - dyn_rope05_r (PPPPP)"!(
    ComponentPose=@n43
    Alpha=1
    ConstraintWeights_0=0
    ConstraintWeights_1:float=0.4456
    ConstraintWeights_2:float=0.4192
    ConstraintWeights_3:float=0.1352
    ConstraintWeights_4:float=0
   ) -> (Pose) {
    boneToModify {boneName=dyn_rope05_r}
    constraintSetup [
     {targetBone {boneName=spine_02}}
     {targetBone {boneName=spine_03}}
     {targetBone {boneName=spine_04}}
     {targetBone {boneName=spine_05}}
     {targetBone {boneName=deform_pec_r}}
    ]
    constraintWeights [1,1,1,1,1]
   }
   n45 = "Constraint - def_chainCorrUp_l (PPPPPPPP)"!(
    ComponentPose=@n44
    Alpha=1
    ConstraintWeights_0=0.0016
    ConstraintWeights_1:float=0.0013
    ConstraintWeights_2:float=0.5583
    ConstraintWeights_3:float=0.3801
    ConstraintWeights_4:float=0.0061
    ConstraintWeights_5:float=0.0199
    ConstraintWeights_6:float=0.032
    ConstraintWeights_7:float=0.0006
   ) -> (Pose) {
    boneToModify {boneName=def_chainCorrUp_l}
    constraintSetup [
     {targetBone {boneName=spine_03}}
     {targetBone {boneName=spine_04}}
     {targetBone {boneName=spine_05}}
     {targetBone {boneName=clavicle_l}}
     {targetBone {boneName=deform_upperarm_up_l}}
     {targetBone {boneName=neck_01}}
     {targetBone {boneName=neck_02}}
     {targetBone {boneName=deform_pec_l}}
    ]
    constraintWeights [1,1,1,1,1,1,1,1]
   }
   n46 = "Constraint - def_chainCorrUp_r (PPPPPPP)"!(
    ComponentPose=@n45
    Alpha=1
    ConstraintWeights_0=0.0016
    ConstraintWeights_1:float=0.0013
    ConstraintWeights_2:float=0.5771
    ConstraintWeights_3:float=0.3922
    ConstraintWeights_4:float=0.0063
    ConstraintWeights_5:float=0.0208
    ConstraintWeights_6:float=0.0006
   ) -> (Pose) {
    boneToModify {boneName=def_chainCorrUp_r}
    constraintSetup [
     {targetBone {boneName=spine_03}}
     {targetBone {boneName=spine_04}}
     {targetBone {boneName=spine_05}}
     {targetBone {boneName=clavicle_r}}
     {targetBone {boneName=deform_upperarm_up_r}}
     {targetBone {boneName=neck_01}}
     {targetBone {boneName=deform_pec_r}}
    ]
    constraintWeights [1,1,1,1,1,1,1]
   }
   n47 = "Transform (Modify) Bone - Bone: def_chainCorrUp_l"(ComponentPose=@n46, Alpha=1, Translation="1,0,1.5", Rotation="30,0,0", Scale="0.8,0.8,0.8") {
    rotationMode=BMM_Additive
    rotationSpace=BCS_BoneSpace
    scaleMode=BMM_Replace
    scaleSpace=BCS_BoneSpace
    translationMode=BMM_Additive
    translationSpace=BCS_BoneSpace
    boneToModify {boneName=def_chainCorrUp_l}
    rotation {pitch=30}
    scale=<0.8 0.8 0.8>
    translation {x=1, z=1.5}
   }
   n48 = "Transform (Modify) Bone - Bone: def_chainCorrUp_r"(ComponentPose=@n47, Alpha=1, Translation="1,0,1.5", Rotation="30,0,0", Scale="0.8,0.8,0.8") {
    rotationMode=BMM_Additive
    rotationSpace=BCS_BoneSpace
    scaleMode=BMM_Replace
    scaleSpace=BCS_BoneSpace
    translationMode=BMM_Additive
    translationSpace=BCS_BoneSpace
    boneToModify {boneName=def_chainCorrUp_r}
    rotation {pitch=30}
    scale=<0.8 0.8 0.8>
    translation {x=1, z=1.5}
   }
   n49 = "Transform (Modify) Bone - Bone: dyn_rope01_l"(ComponentPose=@n48, Alpha=1, Translation="0,0,0", Rotation="0,45,0", Scale="1,1,1") {
    rotationMode=BMM_Additive
    rotationSpace=BCS_BoneSpace
    translationMode=BMM_Additive
    translationSpace=BCS_BoneSpace
    boneToModify {boneName=dyn_rope01_l}
    rotation {yaw=45}
   }
   n50 = "Transform (Modify) Bone - Bone: dyn_rope01_r"(ComponentPose=@n49, Alpha=1, Translation="0,0,0", Rotation="0,45,0", Scale="1,1,1") {
    rotationMode=BMM_Additive
    rotationSpace=BCS_BoneSpace
    translationMode=BMM_Additive
    translationSpace=BCS_BoneSpace
    boneToModify {boneName=dyn_rope01_r}
    rotation {yaw=45}
   }
   n51 = Anim Dynamics - Bone: dyn_rope01_l(ComponentPose=@n50, Alpha=1, ExternalForce="0,0,0") {
    angularDampingOverride=0.99
    angularSpringConstant=900
    bAngularSpring=true
    bOverrideAngularDamping=true
    bUseGravityOverride=true
    linearDampingOverride=0.9999
    linearSpringConstant=600
    simulationSpace=RootRelative
    boundBone {boneName=dyn_rope01_l}
    componentAppliedLinearAccClamp=<1500 1500 1500>
    componentLinearAccScale=<0.1 0.1 0.3>
    componentLinearVelScale=<0.1 0.1 0.1>
    gravityOverride {z=-980}
    physicsBodyDefinitions [{
      boundBone {boneName=dyn_rope01_l}
      boxExtents=<12 12 12>
      constraintSetup {angularLimitsMax=<0 0 5>, angularLimitsMin=<0 0 0>, angularTarget=<1 0 0>}
      localJointOffset=<1 0 0>
     }]
    simSpaceSettings {maxAngularAcceleration=1500, maxAngularVelocity=1500, simSpaceAngularAlpha=0.1}
   }
   n52 = Anim Dynamics - Bone: dyn_rope02_l(ComponentPose=@n51, Alpha=1, ExternalForce="0,0,0") {
    angularDampingOverride=0.99
    angularSpringConstant=900
    bAngularSpring=true
    bOverrideAngularDamping=true
    bUseGravityOverride=true
    linearDampingOverride=0.9999
    linearSpringConstant=600
    simulationSpace=RootRelative
    boundBone {boneName=dyn_rope02_l}
    componentAppliedLinearAccClamp=<1500 1500 1500>
    componentLinearAccScale=<0.1 0.1 0.3>
    componentLinearVelScale=<0.1 0.1 0.1>
    gravityOverride {z=-980}
    physicsBodyDefinitions [{
      boundBone {boneName=dyn_rope02_l}
      boxExtents=<12 12 12>
      constraintSetup {angularLimitsMax=<0 0 10>, angularLimitsMin=<0 0 0>, angularTarget=<1 0 0>}
      localJointOffset=<1 0 0>
     }]
    simSpaceSettings {maxAngularAcceleration=1500, maxAngularVelocity=1500, simSpaceAngularAlpha=0.1}
   }
   n53 = Anim Dynamics - Bone: dyn_rope03_l(ComponentPose=@n52, Alpha=1, ExternalForce="0,0,0") {
    angularDampingOverride=0.99
    angularSpringConstant=900
    bAngularSpring=true
    bOverrideAngularDamping=true
    bUseGravityOverride=true
    linearDampingOverride=0.9999
    linearSpringConstant=600
    simulationSpace=RootRelative
    boundBone {boneName=dyn_rope03_l}
    componentAppliedLinearAccClamp=<1500 1500 1500>
    componentLinearAccScale=<0.1 0.1 0.3>
    componentLinearVelScale=<0.1 0.1 0.1>
    gravityOverride {z=-980}
    physicsBodyDefinitions [{
      boundBone {boneName=dyn_rope03_l}
      boxExtents=<12 12 12>
      constraintSetup {angularLimitsMax=<0 0 15>, angularLimitsMin=<0 0 0>, angularTarget=<1 0 0>}
      localJointOffset=<1 0 0>
     }]
    simSpaceSettings {maxAngularAcceleration=1500, maxAngularVelocity=1500, simSpaceAngularAlpha=0.1}
   }
   n54 = Get AnimDynamics Neck Sim Space Settings() -> (AnimDynamics Neck Sim Space Settings:Anim Phys Sim Space Settings Structure)
   n55 = Anim Dynamics - Bone: dyn_rope04_l!(ComponentPose=@n53, Alpha=1, SimSpaceSettings:Anim Phys Sim Space Settings Structure=@n54, ExternalForce="0,0,0") -> (Pose) {
    angularDampingOverride=0.99
    angularSpringConstant=900
    bAngularSpring=true
    bOverrideAngularDamping=true
    bUseGravityOverride=true
    linearDampingOverride=0.9999
    linearSpringConstant=600
    simulationSpace=RootRelative
    boundBone {boneName=dyn_rope04_l}
    componentAppliedLinearAccClamp=<1500 1500 1500>
    componentLinearAccScale=<0.1 0.1 0.3>
    componentLinearVelScale=<0.1 0.1 0.1>
    gravityOverride {z=-980}
    physicsBodyDefinitions [{
      boundBone {boneName=dyn_rope04_l}
      boxExtents=<12 12 12>
      constraintSetup {angularLimitsMax=<0 10 30>, angularLimitsMin=<0 -5 0>, angularTarget=<1 0 0>}
      localJointOffset=<1 0 0>
     }]
    simSpaceSettings {maxAngularAcceleration=1500, maxAngularVelocity=1500, simSpaceAngularAlpha=0.3}
   }
   n56 = Anim Dynamics - Bone: dyn_rope05_l!(ComponentPose=@n55, Alpha=1, SimSpaceSettings:Anim Phys Sim Space Settings Structure=@n54, ExternalForce="0,0,0") -> (Pose) {
    angularDampingOverride=0.99
    angularSpringConstant=900
    bAngularSpring=true
    bOverrideAngularDamping=true
    bUseGravityOverride=true
    linearDampingOverride=0.9999
    linearSpringConstant=600
    simulationSpace=RootRelative
    boundBone {boneName=dyn_rope05_l}
    componentAppliedLinearAccClamp=<1500 1500 1500>
    componentLinearAccScale=<0.1 0.1 0.3>
    componentLinearVelScale=<0.1 0.1 0.1>
    gravityOverride {z=-980}
    physicsBodyDefinitions [{
      boundBone {boneName=dyn_rope05_l}
      boxExtents=<12 12 12>
      constraintSetup {angularLimitsMax=<0 7 10>, angularLimitsMin=<0 -7 0>, angularTarget=<1 0 0>}
      localJointOffset=<1 0 0>
     }]
    simSpaceSettings {maxAngularAcceleration=1500, maxAngularVelocity=1500, simSpaceAngularAlpha=0.3}
   }
   n57 = Anim Dynamics - Bone: dyn_rope01_r(ComponentPose=@n56, Alpha=1, ExternalForce="0,0,0") {
    angularDampingOverride=0.99
    angularSpringConstant=900
    bAngularSpring=true
    bOverrideAngularDamping=true
    bUseGravityOverride=true
    linearDampingOverride=0.9999
    linearSpringConstant=600
    simulationSpace=RootRelative
    boundBone {boneName=dyn_rope01_r}
    componentAppliedLinearAccClamp=<1500 1500 1500>
    componentLinearAccScale=<0.1 0.1 0.3>
    componentLinearVelScale=<0.1 0.1 0.1>
    gravityOverride {z=-980}
    physicsBodyDefinitions [{
      boundBone {boneName=dyn_rope01_r}
      boxExtents=<12 12 12>
      constraintSetup {angularLimitsMax=<0 0 5>, angularLimitsMin=<0 0 0>, angularTarget=<1 0 0>}
      localJointOffset=<1 0 0>
     }]
    simSpaceSettings {maxAngularAcceleration=1500, maxAngularVelocity=1500, simSpaceAngularAlpha=0.1}
   }
   n58 = Anim Dynamics - Bone: dyn_rope02_r(ComponentPose=@n57, Alpha=1, ExternalForce="0,0,0") {
    angularDampingOverride=0.99
    angularSpringConstant=900
    bAngularSpring=true
    bOverrideAngularDamping=true
    bUseGravityOverride=true
    linearDampingOverride=0.9999
    linearSpringConstant=600
    simulationSpace=RootRelative
    boundBone {boneName=dyn_rope02_r}
    componentAppliedLinearAccClamp=<1500 1500 1500>
    componentLinearAccScale=<0.1 0.1 0.3>
    componentLinearVelScale=<0.1 0.1 0.1>
    gravityOverride {z=-980}
    physicsBodyDefinitions [{
      boundBone {boneName=dyn_rope02_r}
      boxExtents=<12 12 12>
      constraintSetup {angularLimitsMax=<0 0 10>, angularLimitsMin=<0 0 0>, angularTarget=<1 0 0>}
      localJointOffset=<1 0 0>
     }]
    simSpaceSettings {maxAngularAcceleration=1500, maxAngularVelocity=1500, simSpaceAngularAlpha=0.1}
   }
   n59 = Anim Dynamics - Bone: dyn_rope03_r(ComponentPose=@n58, Alpha=1, ExternalForce="0,0,0") {
    angularDampingOverride=0.99
    angularSpringConstant=900
    bAngularSpring=true
    bOverrideAngularDamping=true
    bUseGravityOverride=true
    linearDampingOverride=0.9999
    linearSpringConstant=600
    simulationSpace=RootRelative
    boundBone {boneName=dyn_rope03_r}
    componentAppliedLinearAccClamp=<1500 1500 1500>
    componentLinearAccScale=<0.1 0.1 0.3>
    componentLinearVelScale=<0.1 0.1 0.1>
    gravityOverride {z=-980}
    physicsBodyDefinitions [{
      boundBone {boneName=dyn_rope03_r}
      boxExtents=<12 12 12>
      constraintSetup {angularLimitsMax=<0 0 20>, angularLimitsMin=<0 0 0>, angularTarget=<1 0 0>}
      localJointOffset=<1 0 0>
     }]
    simSpaceSettings {maxAngularAcceleration=1500, maxAngularVelocity=1500, simSpaceAngularAlpha=0.1}
   }
   n60 = Anim Dynamics - Bone: dyn_rope04_r!(ComponentPose=@n59, Alpha=1, SimSpaceSettings:Anim Phys Sim Space Settings Structure=@n54, ExternalForce="0,0,0") -> (Pose) {
    angularDampingOverride=0.99
    angularSpringConstant=900
    bAngularSpring=true
    bOverrideAngularDamping=true
    bUseGravityOverride=true
    linearDampingOverride=0.9999
    linearSpringConstant=600
    simulationSpace=RootRelative
    boundBone {boneName=dyn_rope04_r}
    componentAppliedLinearAccClamp=<1500 1500 1500>
    componentLinearAccScale=<0.1 0.1 0.3>
    componentLinearVelScale=<0.1 0.1 0.1>
    gravityOverride {z=-980}
    physicsBodyDefinitions [{
      boundBone {boneName=dyn_rope04_r}
      boxExtents=<12 12 12>
      constraintSetup {angularLimitsMax=<0 10 30>, angularLimitsMin=<0 -5 0>, angularTarget=<1 0 0>}
      localJointOffset=<1 0 0>
     }]
    simSpaceSettings {maxAngularAcceleration=1500, maxAngularVelocity=1500, simSpaceAngularAlpha=0.3}
   }
   n61 = Anim Dynamics - Bone: dyn_rope05_r!(ComponentPose=@n60, Alpha=1, SimSpaceSettings:Anim Phys Sim Space Settings Structure=@n54, ExternalForce="0,0,0") -> (Pose) {
    angularDampingOverride=0.99
    angularSpringConstant=900
    bAngularSpring=true
    bOverrideAngularDamping=true
    bUseGravityOverride=true
    linearDampingOverride=0.9999
    linearSpringConstant=600
    simulationSpace=RootRelative
    boundBone {boneName=dyn_rope05_r}
    componentAppliedLinearAccClamp=<1500 1500 1500>
    componentLinearAccScale=<0.1 0.1 0.3>
    componentLinearVelScale=<0.1 0.1 0.1>
    gravityOverride {z=-980}
    physicsBodyDefinitions [{
      boundBone {boneName=dyn_rope05_r}
      boxExtents=<12 12 12>
      constraintSetup {angularLimitsMax=<0 7 10>, angularLimitsMin=<0 -7 0>, angularTarget=<1 0 0>}
      localJointOffset=<1 0 0>
     }]
    simSpaceSettings {maxAngularAcceleration=1500, maxAngularVelocity=1500, simSpaceAngularAlpha=0.3}
   }
   n62 = "Constraint - dyn_chain_rear (PP)"!(ComponentPose=@n61, Alpha=1, ConstraintWeights_0=0.5, ConstraintWeights_1:float=0.5) -> (Pose) {
    boneToModify {boneName=dyn_chain_rear}
    constraintSetup [{targetBone {boneName=dyn_rope03_l}}, {targetBone {boneName=dyn_rope03_r}}]
    constraintWeights [1,1]
   }
   n63 = "Constraint - def_necklace_02 (PP)"!(ComponentPose=@n62, Alpha=1, ConstraintWeights_0=0.5, ConstraintWeights_1:float=0.5) -> (Pose) {
    boneToModify {boneName=def_necklace_02}
    constraintSetup [{targetBone {boneName=dyn_rope02_l}}, {targetBone {boneName=dyn_rope02_r}}]
    constraintWeights [1,1]
   }
   n64 = "Constraint - def_necklace_01 (PPP)"!(ComponentPose=@n63, Alpha=1, ConstraintWeights_0=0.5, ConstraintWeights_1:float=0.5, ConstraintWeights_2:float=0) -> (Pose) {
    boneToModify {boneName=def_necklace_01}
    constraintSetup [{targetBone {boneName=dyn_rope01_l}}, {targetBone {boneName=dyn_rope01_r}}, {targetBone {boneName=spine_05}}]
    constraintWeights [1,1,1]
   }
   n65 = Anim Dynamics - Bone: def_necklace_01(ComponentPose=@n64, Alpha=1, ExternalForce="0,0,0") {
    angularDampingOverride=0.99
    angularSpringConstant=900
    bAngularSpring=true
    bLinearSpring=true
    bOverrideAngularDamping=true
    bOverrideLinearDamping=true
    bUseGravityOverride=true
    linearDampingOverride=1
    linearSpringConstant=200
    simulationSpace=RootRelative
    boundBone {boneName=def_necklace_01}
    componentAppliedLinearAccClamp=<1500 1500 1500>
    componentLinearAccScale=<0.1 0.2 0.1>
    componentLinearVelScale=<0.1 0.1 0.1>
    gravityOverride {z=-980}
    physicsBodyDefinitions [{
      boundBone {boneName=def_necklace_01}
      boxExtents=<15 15 15>
      constraintSetup {angularLimitsMax=<0 0 0>, angularLimitsMin=<0 0 0>, angularTarget=<1 0 0>, linearAxesMax=<0 1 0.5>, linearAxesMin=<0 0 -0.5>}
      localJointOffset=<1 0 0>
     }]
    simSpaceSettings {maxAngularAcceleration=1500, maxAngularVelocity=1500, simSpaceAngularAlpha=0.1}
   }
   n66 = Anim Dynamics - Bone: def_necklace_02(ComponentPose=@n65, Alpha=1, ExternalForce="0,0,0") {
    angularDampingOverride=0.99
    angularSpringConstant=900
    bAngularSpring=true
    bLinearSpring=true
    bOverrideAngularDamping=true
    bOverrideLinearDamping=true
    bUseGravityOverride=true
    linearDampingOverride=1
    linearSpringConstant=200
    simulationSpace=RootRelative
    boundBone {boneName=def_necklace_02}
    componentAppliedLinearAccClamp=<1500 1500 1500>
    componentLinearAccScale=<0.1 0.2 0.1>
    componentLinearVelScale=<0.1 0.1 0.1>
    gravityOverride {z=-980}
    physicsBodyDefinitions [{
      boundBone {boneName=def_necklace_02}
      boxExtents=<13 13 13>
      constraintSetup {angularLimitsMax=<0 0 0>, angularLimitsMin=<0 0 0>, angularTarget=<1 0 0>, linearAxesMax=<0 1 0.5>, linearAxesMin=<0 0 -0.5>}
      localJointOffset=<0.7 0 0>
     }]
    simSpaceSettings {maxAngularAcceleration=1500, maxAngularVelocity=1500, simSpaceAngularAlpha=0.1}
   }
   n67 = Anim Dynamics - Bone: dyn_chain_rear(ComponentPose=@n66, Alpha=1, ExternalForce="0,0,0") {
    angularDampingOverride=0.99
    angularSpringConstant=900
    bAngularSpring=true
    bLinearSpring=true
    bOverrideAngularDamping=true
    bOverrideLinearDamping=true
    bUseGravityOverride=true
    linearDampingOverride=1
    linearSpringConstant=200
    simulationSpace=RootRelative
    boundBone {boneName=dyn_chain_rear}
    componentAppliedLinearAccClamp=<1500 1500 1500>
    componentLinearAccScale=<0.1 0.2 0.1>
    componentLinearVelScale=<0.1 0.1 0.1>
    gravityOverride {z=-980}
    physicsBodyDefinitions [{
      boundBone {boneName=dyn_chain_rear}
      boxExtents=<16 16 16>
      constraintSetup {angularLimitsMax=<0 0 0>, angularLimitsMin=<0 0 0>, angularTarget=<1 0 0>, linearAxesMax=<0 0.5 0.2>, linearAxesMin=<0 0 -0.2>}
      localJointOffset=<0.5 0 0>
     }]
    simSpaceSettings {maxAngularAcceleration=1500, maxAngularVelocity=1500, simSpaceAngularAlpha=0.1}
   }
   n68 = Anim Dynamics - Bone: dyn_chain(ComponentPose=@n67, Alpha=1, ExternalForce="0,0,0") {
    angularDampingOverride=0.99
    angularSpringConstant=900
    bAngularSpring=true
    bLinearSpring=true
    bOverrideAngularDamping=true
    bOverrideLinearDamping=true
    bUseGravityOverride=true
    linearDampingOverride=1
    linearSpringConstant=200
    simulationSpace=RootRelative
    boundBone {boneName=dyn_chain}
    componentAppliedLinearAccClamp=<1500 1500 1500>
    componentLinearAccScale=<0.1 0.2 0.1>
    componentLinearVelScale=<0.1 0.1 0.1>
    gravityOverride {z=-980}
    physicsBodyDefinitions [{
      boundBone {boneName=dyn_chain}
      boxExtents=<17 17 17>
      constraintSetup {angularLimitsMax=<0 0 0>, angularLimitsMin=<0 0 0>, angularTarget=<1 0 0>, linearAxesMax=<0 0.5 0.5>, linearAxesMin=<-0.2 0 -0.5>}
      localJointOffset=<0.3 0 0>
     }]
    simSpaceSettings {maxAngularAcceleration=1500, maxAngularVelocity=1500, simSpaceAngularAlpha=0.1}
   }
   Save cached pose 'Neck Dynamics'(Pose=Component To Local(ComponentPose=@n68)) {cacheName=Neck Dynamics}
   n69 = Anim Dynamics - Bone: dyn_shirt_l(ComponentPose=Local To Component(LocalPose=Use cached pose 'Neck Dynamics'()), Alpha=1, ExternalForce="0,0,0") {
    angularDampingOverride=0.99
    angularSpringConstant=600
    bAngularSpring=true
    bLinearSpring=true
    bOverrideAngularDamping=true
    bOverrideLinearDamping=true
    bUseGravityOverride=true
    linearDampingOverride=0.999
    linearSpringConstant=100
    simulationSpace=RootRelative
    boundBone {boneName=dyn_shirt_l}
    componentAppliedLinearAccClamp=<3000 3000 3000>
    componentLinearAccScale=<0.2 0.2 0.5>
    componentLinearVelScale=<0.2 0.2 0.2>
    gravityOverride {z=-980}
    physicsBodyDefinitions [{
      boundBone {boneName=dyn_shirt_l}
      boxExtents=<15 15 15>
      constraintSetup {angularLimitsMax=<10 5 0>, angularLimitsMin=<-10 -10 0>, angularTarget=<1 0 0>, linearAxesMax=<0 0 3>, linearAxesMin=<-7 0 0>}
      localJointOffset=<1 0 0>
     }]
    simSpaceSettings {maxAngularAcceleration=1500, maxAngularVelocity=1500, simSpaceAngularAlpha=0.1}
   }
   n70 = Anim Dynamics - Bone: dyn_shirt_r(ComponentPose=@n69, Alpha=1, ExternalForce="0,0,0") {
    angularDampingOverride=0.99
    angularSpringConstant=600
    bAngularSpring=true
    bLinearSpring=true
    bOverrideAngularDamping=true
    bOverrideLinearDamping=true
    bUseGravityOverride=true
    linearDampingOverride=0.999
    linearSpringConstant=100
    simulationSpace=RootRelative
    boundBone {boneName=dyn_shirt_r}
    componentAppliedLinearAccClamp=<3000 3000 3000>
    componentLinearAccScale=<0.2 0.2 0.5>
    componentLinearVelScale=<0.2 0.2 0.2>
    gravityOverride {z=-980}
    physicsBodyDefinitions [{
      boundBone {boneName=dyn_shirt_r}
      boxExtents=<15 15 15>
      constraintSetup {angularLimitsMax=<10 5 0>, angularLimitsMin=<-10 -10 0>, angularTarget=<1 0 0>, linearAxesMax=<0 0 3>, linearAxesMin=<-7 0 0>}
      localJointOffset=<1 0 0>
     }]
    simSpaceSettings {maxAngularAcceleration=1500, maxAngularVelocity=1500, simSpaceAngularAlpha=0.1}
   }
   n71 = Anim Dynamics - Bone: dyn_shirt(ComponentPose=@n70, Alpha=1, ExternalForce="0,0,0") {
    angularDampingOverride=0.99
    angularSpringConstant=600
    bAngularSpring=true
    bLinearSpring=true
    bOverrideAngularDamping=true
    bOverrideLinearDamping=true
    bUseGravityOverride=true
    linearDampingOverride=0.999
    linearSpringConstant=100
    simulationSpace=RootRelative
    boundBone {boneName=dyn_shirt}
    componentAppliedLinearAccClamp=<3000 3000 3000>
    componentLinearAccScale=<0.2 0.2 0.5>
    componentLinearVelScale=<0.2 0.2 0.2>
    gravityOverride {z=-980}
    physicsBodyDefinitions [{
      boundBone {boneName=dyn_shirt}
      boxExtents=<15 15 15>
      constraintSetup {angularLimitsMax=<10 5 0>, angularLimitsMin=<-10 -10 0>, angularTarget=<1 0 0>, linearAxesMax=<0 0 2>, linearAxesMin=<-3 0 -2>}
      localJointOffset=<-1 0 0>
     }]
    simSpaceSettings {maxAngularAcceleration=1500, maxAngularVelocity=1500, simSpaceAngularAlpha=0.1}
   }
   Save cached pose 'ShirtDynamics'(Pose=Component To Local(ComponentPose=@n71)) {cacheName=ShirtDynamics}
  ]}
 EventGraph {nodes [
   Event BlueprintUpdateAnimation()
   > n1 = "Gravity Override State Function (Nativized)"!(
    execute
    CharacterPartAnimInstance:Custom Character Part Anim Instance Object Reference=Self-Reference()
    GravityOverrideParameterSet:Fort Gravity Override Parameter Set Structure=Get Gravity Override Params() -> (Gravity Override Params:Fort Gravity Override Parameter Set Structure)
    RigidBodyAnimNodeParameterSet:Fort Rigid Body Anim Node Parameter Set Structure=Get RBAN Params() -> (RBAN Params:Fort Rigid Body Anim Node Parameter Set Structure)
    DebugDraw:bool=false
    GameTimeInSeconds:float=GetGameTimeInSeconds()
    DebugJointName:Name=root
    StateLogicLayerName:Name=none
   ) -> (
    then
    OutGravityOverride:Vector
    OutCurrentState:Name=None
    OutGravityOverrideParameters:Fort Gravity Override Parameters Structure
    OutRigidBodyAnimNodeParameters:Fort Rigid Body Anim Node Parameters Structure
   )
   > Set Gravity Override(Gravity Override=@n1)
   > Set RBAN Settings Out(RBAN Settings Out=@n1.OutRigidBodyAnimNodeParameters)
   n2 = GetPartSkeletalMeshComponent()
   > Set NecklaceRotate(NecklaceRotate:float=GetAlphaFromReferenceTransformDelta(SkeletalMeshComponent=@n2, BoneName=spine_05, BaseName=root, InputRange=(X=10,Y=75), TwistAxis="1,0,0")) -> (Output_Get:float)
   > Set Clavicle R RotateAlpha(Clavicle R RotateAlpha:float=GetAlphaFromReferenceTransformDelta(SkeletalMeshComponent=@n2, BoneName=clavicle_r, BaseName=spine_05, DeltaComponentType=EulerY, InputRange=(X=10,Y=50), TwistAxis="1,0,0")) -> (Output_Get:float)
   > Set Clavicle L RotateAlpha(Clavicle L RotateAlpha:float=GetAlphaFromReferenceTransformDelta(SkeletalMeshComponent=@n2, BoneName=clavicle_l, BaseName=spine_05, DeltaComponentType=EulerY, InputRange=(X=10,Y=50), TwistAxis="1,0,0")) -> (Output_Get:float)
   Event OnExitGrindRail() {is=title K2Node_Event}
   > ResetDynamics(InTeleportType=TeleportPhysics)
   > ResetDynamics(InTeleportType=ResetPhysics)
   n3 = Event BlueprintInitializeAnimation()
   Set AnimDynamics Neck Sim Space Settings(
    execute=[@n3.then, @n4.then]
    AnimDynamics Neck Sim Space Settings:Anim Phys Sim Space Settings Structure=Select(
     Option 0:Anim Phys Sim Space Settings Structure=Make Anim Phys Sim Space Settings!(SimSpaceAngularAlpha:float=0.1, MaxAngularVelocity:float=1500, MaxAngularAcceleration:float=1500, ExternalAngularVelocity:Vector="0,0,0") -> (AnimPhysSimSpaceSettings:Anim Phys Sim Space Settings Structure) {is=title K2Node_MakeStruct}
     Option 1:Anim Phys Sim Space Settings Structure=Make Anim Phys Sim Space Settings(SimSpaceAngularAlpha=0.3, MaxAngularVelocity=1500, MaxAngularAcceleration=1500, ExternalAngularVelocity="0,0,0")
     Index=Get bIsInFrontEnd()
    ) -> (ReturnValue:Anim Phys Sim Space Settings Structure)
   ) -> (Output_Get:Anim Phys Sim Space Settings Structure)
   n4 = Event OnCharacterCustomizationLoaded()
  ]}
]
variables [
 AnimDynamics Neck Sim Space Settings {
  type=AnimPhysSimSpaceSettings
  default {maxAngularAcceleration=1500, maxAngularVelocity=1500, simSpaceAngularAlpha=0.3, externalAngularVelocity=<0 0 0>}
 }
 Clavicle L RotateAlpha {type=number}
 Clavicle R RotateAlpha {type=number}
 Gravity Override {type=Vector}
 Gravity Override Params {
  type=FortGravityOverrideParameterSet
  default=(Skydive_GoingUp=(GravityOverride=(WindFrequency=(X=1.500000,Y=7.000000,Z=8.000000),WindAmplitude=(X=200.000000,Y=-500.000000,Z=400.000000),GravityOverride=(X=0.000000,Y=-300.000000,Z=1500.000000),JointName="root")),Skydive_FloatingDown=(GravityOverride=(WindFrequency=(X=1.500000,Y=7.000000,Z=8.000000),WindAmplitude=(X=200.000000,Y=-500.000000,Z=400.000000),WindOffset=(X=0.000000,Y=-600.000000,Z=400.000000),GravityOverride=(X=0.000000,Y=-700.000000,Z=600.000000),JointName="root")),Skydive_DivingDown=(GravityOverride=(WindFrequency=(X=1.500000,Y=12.000000,Z=10.000000),WindAmplitude=(X=50.000000,Y=300.000000,Z=100.000000),GravityOverride=(X=-1500.000000,Y=0.000000,Z=0.000000),JointName="pelvis")),Skydive_Parachuting=(GravityOverride=(WindFrequency=(X=1.500000,Y=7.000000,Z=8.000000),WindAmplitude=(X=200.000000,Y=-500.000000,Z=400.000000),GravityOverride=(X=0.000000,Y=-800.000000,Z=-980.000000),JointName="root")),OnGround_Standing=(GravityOverride=(GravityOverride=(X=0.000000,Y=0.000000,Z=-980.000000),JointName="root")),OnGround_Moving=(GravityOverride=(GravityOverride=(X=0.000000,Y=0.000000,Z=-980.000000),JointName="root")),OnGround_Moving_HighRotationRate=(GravityOverride=(GravityOverride=(X=0.000000,Y=0.000000,Z=-980.000000),JointName="root")),OnGround_Crouch_Moving=(GravityOverride=(GravityOverride=(X=0.000000,Y=0.000000,Z=-980.000000),JointName="root")),Emote_or_Melee=(GravityOverride=(GravityOverride=(X=0.000000,Y=0.000000,Z=-980.000000),JointName="root")),Swimming=(GravityOverride=(GravityOverride=(X=-800.000000,Y=0.000000,Z=0.000000),JointName="pelvis")),Front_End=(GravityOverride=(GravityOverride=(X=0.000000,Y=0.000000,Z=-980.000000),JointName="root")),DBNO=(GravityOverride=(GravityOverride=(X=-980.000000,Y=400.000000,Z=0.000000),JointName="pelvis")),DBNO_Carried=(GravityOverride=(GravityOverride=(X=-980.000000,Y=500.000000,Z=-500.000000),JointName="pelvis")),OnGround_Falling=(GravityOverride=(GravityOverride=(X=0.000000,Y=0.000000,Z=-980.000000),JointName="root")),Ziplining=(GravityOverride=(GravityOverride=(X=0.000000,Y=0.000000,Z=-980.000000),JointName="root")),Skydive_Surfing=(GravityOverride=(WindFrequency=(X=1.500000,Y=7.000000,Z=8.000000),WindAmplitude=(X=200.000000,Y=-500.000000,Z=400.000000),GravityOverride=(X=500.000000,Y=-300.000000,Z=-980.000000),JointName="root")),Riding_Standing=(GravityOverride=(GravityOverride=(X=0.000000,Y=-350.000000,Z=-980.000000),JointName="root")),Riding_Moving=(GravityOverride=(GravityOverride=(X=0.000000,Y=-450.000000,Z=-450.000000),JointName="root")),Riding_Moving_Targeting=(GravityOverride=(GravityOverride=(X=0.000000,Y=-200.000000,Z=-980.000000),JointName="root")),Tactical_Sprint=(GravityOverride=(GravityOverride=(X=0.000000,Y=-450.000000,Z=-500.000000),JointName="root")),Rail_Grinding=(GravityOverride=(GravityOverride=(X=0.000000,Y=0.000000,Z=-980.000000),JointName="root")),)
 }
 NecklaceRotate {type=number}
 RBAN Params {
  type=FortRigidBodyAnimNodeParameterSet
  default=(Skydive_GoingUp=(RigidBodyAnimNode=(SimSpaceSettings=(WorldAlpha=1.000000,DampingAlpha=0.000000,MaxLinearVelocity=15.000000,MaxAngularVelocity=30.000000,MaxLinearAcceleration=600.000000,MaxAngularAcceleration=150.000000,ExternalLinearDragV=(X=0.000000,Y=0.500000,Z=0.000000)))),Skydive_FloatingDown=(RigidBodyAnimNode=(ComponentLinearAccScale=(X=0.000000,Y=0.500000,Z=0.000000),ComponentLinearVelScale=(X=0.000000,Y=0.500000,Z=0.000000),SimSpaceSettings=(WorldAlpha=1.000000,DampingAlpha=0.000000,MaxLinearVelocity=15.000000,MaxAngularVelocity=30.000000,MaxLinearAcceleration=600.000000,MaxAngularAcceleration=150.000000,ExternalLinearDragV=(X=0.000000,Y=0.500000,Z=0.000000)))),Skydive_DivingDown=(RigidBodyAnimNode=(ComponentLinearAccScale=(X=0.100000,Y=0.000000,Z=0.000000),ComponentLinearVelScale=(X=0.100000,Y=0.000000,Z=0.000000),SimSpaceSettings=(WorldAlpha=1.000000,DampingAlpha=0.000000,MaxLinearVelocity=15.000000,MaxAngularVelocity=30.000000,MaxLinearAcceleration=600.000000,MaxAngularAcceleration=150.000000,ExternalLinearDragV=(X=0.000000,Y=0.500000,Z=0.000000)))),Skydive_Parachuting=(RigidBodyAnimNode=(ComponentLinearAccScale=(X=0.000000,Y=0.100000,Z=0.300000),ComponentLinearVelScale=(X=0.000000,Y=0.100000,Z=0.300000),SimSpaceSettings=(WorldAlpha=1.000000,DampingAlpha=0.000000,MaxLinearVelocity=15.000000,MaxAngularVelocity=30.000000,MaxLinearAcceleration=600.000000,MaxAngularAcceleration=150.000000,ExternalLinearDragV=(X=0.000000,Y=0.500000,Z=0.000000)))),OnGround_Standing=(RigidBodyAnimNode=(ComponentLinearAccScale=(X=0.200000,Y=0.500000,Z=0.000000),ComponentLinearVelScale=(X=0.200000,Y=0.500000,Z=0.000000),SimSpaceSettings=(WorldAlpha=1.000000,DampingAlpha=0.000000,MaxLinearVelocity=24.000000,MaxAngularVelocity=10.000000,MaxLinearAcceleration=1000.000000,MaxAngularAcceleration=250.000000,ExternalLinearDragV=(X=0.000000,Y=0.500000,Z=0.000000)))),OnGround_Moving=(RigidBodyAnimNode=(ComponentLinearAccScale=(X=0.300000,Y=0.300000,Z=0.000000),ComponentLinearVelScale=(X=0.300000,Y=0.300000,Z=0.000000),SimSpaceSettings=(WorldAlpha=1.000000,DampingAlpha=0.000000,MaxLinearVelocity=24.000000,MaxAngularVelocity=10.000000,MaxLinearAcceleration=1000.000000,MaxAngularAcceleration=250.000000,ExternalLinearDragV=(X=0.000000,Y=0.500000,Z=0.000000)))),OnGround_Moving_HighRotationRate=(RigidBodyAnimNode=(SimSpaceSettings=(WorldAlpha=1.000000,DampingAlpha=0.000000,MaxLinearVelocity=15.000000,MaxAngularVelocity=30.000000,MaxLinearAcceleration=600.000000,MaxAngularAcceleration=150.000000,ExternalLinearDragV=(X=0.000000,Y=0.500000,Z=0.000000)))),OnGround_Crouch_Moving=(RigidBodyAnimNode=(SimSpaceSettings=(WorldAlpha=1.000000,DampingAlpha=0.000000,MaxLinearVelocity=15.000000,MaxAngularVelocity=30.000000,MaxLinearAcceleration=600.000000,MaxAngularAcceleration=150.000000,ExternalLinearDragV=(X=0.000000,Y=0.500000,Z=0.000000)))),Emote_or_Melee=(RigidBodyAnimNode=(SimSpaceSettings=(WorldAlpha=1.000000,DampingAlpha=0.000000,MaxLinearVelocity=15.000000,MaxAngularVelocity=30.000000,MaxLinearAcceleration=600.000000,MaxAngularAcceleration=150.000000,ExternalLinearDragV=(X=0.000000,Y=0.500000,Z=0.000000)))),Swimming=(RigidBodyAnimNode=(SimSpaceSettings=(WorldAlpha=1.000000,DampingAlpha=0.000000,MaxLinearVelocity=15.000000,MaxAngularVelocity=30.000000,MaxLinearAcceleration=600.000000,MaxAngularAcceleration=150.000000,ExternalLinearDragV=(X=0.000000,Y=0.500000,Z=0.000000)))),Front_End=(RigidBodyAnimNode=(SimSpaceSettings=(WorldAlpha=0.700000,DampingAlpha=0.000000,MaxLinearVelocity=10.000000,MaxAngularVelocity=25.000000,MaxLinearAcceleration=400.000000,MaxAngularAcceleration=90.000000,ExternalLinearDragV=(X=0.000000,Y=0.300000,Z=0.000000)))),DBNO=(RigidBodyAnimNode=(SimSpaceSettings=(WorldAlpha=1.000000,DampingAlpha=0.000000,MaxLinearVelocity=15.000000,MaxAngularVelocity=30.000000,MaxLinearAcceleration=600.000000,MaxAngularAcceleration=150.000000,ExternalLinearDragV=(X=0.000000,Y=0.500000,Z=0.000000)))),DBNO_Carried=(RigidBodyAnimNode=(SimSpaceSettings=(WorldAlpha=1.000000,DampingAlpha=0.000000,MaxLinearVelocity=15.000000,MaxAngularVelocity=30.000000,MaxLinearAcceleration=600.000000,MaxAngularAcceleration=150.000000,ExternalLinearDragV=(X=0.000000,Y=0.500000,Z=0.000000)))),OnGround_Falling=(RigidBodyAnimNode=(SimSpaceSettings=(WorldAlpha=1.000000,DampingAlpha=0.000000,MaxLinearVelocity=15.000000,MaxAngularVelocity=30.000000,MaxLinearAcceleration=600.000000,MaxAngularAcceleration=150.000000,ExternalLinearDragV=(X=0.000000,Y=0.500000,Z=0.000000)))),Ziplining=(RigidBodyAnimNode=(SimSpaceSettings=(WorldAlpha=1.000000,DampingAlpha=0.000000,MaxLinearVelocity=15.000000,MaxAngularVelocity=30.000000,MaxLinearAcceleration=600.000000,MaxAngularAcceleration=150.000000,ExternalLinearDragV=(X=0.000000,Y=0.500000,Z=0.000000)))),Skydive_Surfing=(RigidBodyAnimNode=(SimSpaceSettings=(WorldAlpha=1.000000,DampingAlpha=0.000000,MaxLinearVelocity=15.000000,MaxAngularVelocity=30.000000,MaxLinearAcceleration=600.000000,MaxAngularAcceleration=150.000000,ExternalLinearDragV=(X=0.000000,Y=0.500000,Z=0.000000)))),Riding_Standing=(RigidBodyAnimNode=(SimSpaceSettings=(WorldAlpha=1.000000,DampingAlpha=0.000000,MaxLinearVelocity=15.000000,MaxAngularVelocity=30.000000,MaxLinearAcceleration=600.000000,MaxAngularAcceleration=150.000000,ExternalLinearDragV=(X=0.000000,Y=0.500000,Z=0.000000)))),Riding_Moving=(RigidBodyAnimNode=(SimSpaceSettings=(WorldAlpha=1.000000,DampingAlpha=0.000000,MaxLinearVelocity=15.000000,MaxAngularVelocity=30.000000,MaxLinearAcceleration=600.000000,MaxAngularAcceleration=150.000000,ExternalLinearDragV=(X=0.000000,Y=0.500000,Z=0.000000)))),Riding_Moving_Targeting=(RigidBodyAnimNode=(SimSpaceSettings=(WorldAlpha=1.000000,DampingAlpha=0.000000,MaxLinearVelocity=15.000000,MaxAngularVelocity=30.000000,MaxLinearAcceleration=600.000000,MaxAngularAcceleration=150.000000,ExternalLinearDragV=(X=0.000000,Y=0.500000,Z=0.000000)))),Tactical_Sprint=(RigidBodyAnimNode=(SimSpaceSettings=(WorldAlpha=1.000000,DampingAlpha=0.000000,MaxLinearVelocity=15.000000,MaxAngularVelocity=30.000000,MaxLinearAcceleration=600.000000,MaxAngularAcceleration=150.000000,ExternalLinearDragV=(X=0.000000,Y=0.500000,Z=0.000000)))),Rail_Grinding=(RigidBodyAnimNode=(SimSpaceSettings=(WorldAlpha=0.100000,DampingAlpha=0.000000,MaxLinearVelocity=15.000000,MaxAngularVelocity=30.000000,MaxLinearAcceleration=600.000000,MaxAngularAcceleration=150.000000,ExternalLinearDragV=(X=0.000000,Y=0.500000,Z=0.000000)))),)
 }
 RBAN Settings Out {type=FortRigidBodyAnimNodeParameters}
]
