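# Physics asset for the SummitReedGrit male medium character's FaceAcc part, setting up secondary motion for long hair.
# Seven three-capsule hair chains (back, sides, front) hang from small MSTR anchor spheres through mostly limited swing and twist constraints.
# Angularly free constraints cross-link tip capsules: center back to both back sides, and each side to its front neighbor.
# Head, neck, upper spine, clavicle and upper-arm bodies, unconstrained here, provide collision shapes around the hair.
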PhysicsAsset /BRCosmetics/Characters/Player/Male/Medium/Bodies/M_MED_SummitReedGrit/Meshes/Parts/M_MED_SummitReedGrit_FaceAcc_Physics
bodies {
 MSTR_hair_long_bk [{shapeType=Sphere, radius=0.505, shapeName=MSTR_hair_fr_r_sphere}]
 MSTR_hair_long_bk_l [{shapeType=Sphere, radius=0.505, shapeName=MSTR_hair_fr_r_sphere}]
 MSTR_hair_long_bk_r [{shapeType=Sphere, radius=0.505, shapeName=MSTR_hair_fr_r_sphere}]
 MSTR_hair_long_fr_l [{shapeType=Sphere, radius=0.505, shapeName=MSTR_hair_fr_l_sphere}]
 MSTR_hair_long_fr_r [{shapeType=Sphere, radius=0.505, shapeName=MSTR_hair_fr_r_sphere}]
 MSTR_hair_long_side_l [{shapeType=Sphere, radius=0.505, shapeName=MSTR_hair_side_l_sphere}]
 MSTR_hair_long_side_r [{shapeType=Sphere, radius=0.505, shapeName=MSTR_hair_side_r_sphere}]
 clavicle_l [{
   shapeType=Capsule
   length=2.6125
   radius=6.8304
   shapeName=clavicle_l_capsule
   center=<9.8574 -1.6177 -0.98>
   rotation=<rot -0.0001 -90 97.3912>
  }]
 clavicle_r [{
   shapeType=Capsule
   length=2.6125
   radius=6.8304
   shapeName=clavicle_r_capsule
   center=<-9.9246 1.8533 0.4619>
   rotation=<rot -0.0001 -90 -82.6088>
  }]
 dyn_hair_long_01_bk [{
   shapeType=Capsule
   length=4.3002
   radius=2.6199
   shapeName=dyn_hair_01_fr_l_capsule
   center=<4.9678 -1.7735 -0.395>
   rotation=<rot -89.2592 0.0001 0>
  }]
 dyn_hair_long_01_bk_l [{
   shapeType=Capsule
   length=2.8893
   radius=2.6199
   shapeName=dyn_hair_01_fr_r_capsule
   center=<4.6618 -1.8189 0.5535>
   rotation=<rot 88.1982 -179.9999 0>
  }]
 dyn_hair_long_01_bk_r [{
   shapeType=Capsule
   length=2.8893
   radius=2.6199
   shapeName=dyn_hair_01_fr_r_capsule
   center=<4.6664 -1.739 -0.5302>
   rotation=<rot 88.1982 -179.9999 0>
  }]
 dyn_hair_long_01_fr_l [{
   shapeType=Capsule
   length=2.2405
   radius=2.3803
   shapeName=dyn_hair_01_fr_l_capsule
   center=<2.4587 -1.905 -1.3055>
   rotation=<rot -83.5907 -92.5918 91.1315>
  }]
 dyn_hair_long_01_fr_r [{
   shapeType=Capsule
   length=2.2405
   radius=2.3803
   shapeName=dyn_hair_01_fr_r_capsule
   center=<-2.4485 1.8165 1.3155>
   rotation=<rot 83.5943 91.7156 -84.507>
  }]
 dyn_hair_long_01_side_l [{
   shapeType=Capsule
   length=2.506
   radius=2.1148
   shapeName=dyn_hair_01_fr_l_capsule
   center=<2.4387 -1.5496 -1.6555>
   rotation=<rot -86.6672 106.5901 -106.2873>
  }]
 dyn_hair_long_01_side_r [{
   shapeType=Capsule
   length=2.506
   radius=2.1148
   shapeName=dyn_hair_01_fr_r_capsule
   center=<-2.4496 1.531 1.6549>
   rotation=<rot 86.7607 -99.5656 80.8456>
  }]
 dyn_hair_long_02_bk [{shapeType=Capsule, length=6.0132, radius=2.6199, shapeName=dyn_hair_01_fr_l_capsule, center=<2.1754 -1.261 -0.4131>}]
 dyn_hair_long_02_bk_l [{
   shapeType=Capsule
   length=3.0771
   radius=2.432
   shapeName=dyn_hair_01_fr_l_capsule
   center=<1.8766 -1.8637 -0.1459>
   rotation=<rot 0 0.0001 60.6316>
  }]
 dyn_hair_long_02_bk_r [{
   shapeType=Capsule
   length=3.0771
   radius=2.432
   shapeName=dyn_hair_01_fr_l_capsule
   center=<1.8766 -1.8637 -0.1459>
   rotation=<rot 0 0.0002 -50.181>
  }]
 dyn_hair_long_02_fr_l [{shapeType=Capsule, length=2.5653, radius=1.7119, shapeName=dyn_hair_02_fr_l_capsule, center=<1.981 -1.911 -0.76>, rotation=<rot 0 0 23.562>}]
 dyn_hair_long_02_fr_r [{
   shapeType=Capsule
   length=2.5037
   radius=1.7735
   shapeName=dyn_hair_02_fr_r_capsule
   center=<-1.981 1.8934 0.7326>
   rotation=<rot 0.0001 0 -159.1094>
  }]
 dyn_hair_long_02_side_l [{
   shapeType=Capsule
   length=3.761
   radius=1.9444
   shapeName=dyn_hair_02_fr_l_capsule
   center=<1.9924 -1.336 -1.5093>
   rotation=<rot -2.7068 0 -0.0883>
  }]
 dyn_hair_long_02_side_r [{
   shapeType=Capsule
   length=3.867
   radius=1.8384
   shapeName=dyn_hair_02_fr_r_capsule
   center=<-1.9893 1.2439 1.574>
   rotation=<rot -2.7146 0 -178.2085>
  }]
 dyn_hair_long_03_bk [{shapeType=Capsule, length=6.0132, radius=2.6199, shapeName=dyn_hair_01_fr_l_capsule, center=<1.6519 -1.4051 -0.4131>}]
 dyn_hair_long_03_bk_l [{
   shapeType=Capsule
   length=3.4671
   radius=2.0421
   shapeName=dyn_hair_01_fr_l_capsule
   center=<0.7883 -0.9171 -0.0292>
   rotation=<rot 0.0005 0.0063 51.1518>
  }]
 dyn_hair_long_03_bk_r [{
   shapeType=Capsule
   length=3.4671
   radius=2.0421
   shapeName=dyn_hair_01_fr_l_capsule
   center=<0.7883 -0.9171 -0.0292>
   rotation=<rot 0.0005 0.0064 -47.3952>
  }]
 dyn_hair_long_03_fr_l [{
   shapeType=Capsule
   length=2.492
   radius=1.8287
   shapeName=dyn_hair_03_fr_l_capsule
   center=<0.9838 -1.5552 -0.6323>
   rotation=<rot 0.0001 0 17.8021>
  }]
 dyn_hair_long_03_fr_r [{
   shapeType=Capsule
   length=2.3414
   radius=1.9793
   shapeName=dyn_hair_03_fr_r_capsule
   center=<-0.9838 0.6031 0.9591>
   rotation=<rot 0.0001 0 -161.4455>
  }]
 dyn_hair_long_03_side_l [{
   shapeType=Capsule
   length=2.722
   radius=1.5987
   shapeName=dyn_hair_03_fr_l_capsule
   center=<0.8537 -1.7736 -1.622>
   rotation=<rot -6.7874 0 0>
  }]
 dyn_hair_long_03_side_r [{
   shapeType=Capsule
   length=2.7124
   radius=1.6084
   shapeName=dyn_hair_03_fr_r_capsule
   center=<-0.8537 1.1057 1.622>
   rotation=<rot -6.7874 0 -173.5506>
  }]
 head [
  {shapeType=Sphere, radius=6.5325, shapeName=head_sphere, center=<0 1.903 0>}
  {shapeType=Sphere, radius=6.5881, shapeName=head_sphere, center=<3.3557 -0.2429 0>}
 ]
 neck_01 [{shapeType=Sphere, radius=5.6639, shapeName=neck_01_sphere, center=<0 0.1605 0>}]
 neck_02 [{shapeType=Sphere, radius=5.3844, shapeName=neck_01_sphere, center=<0 -0.7872 0>}]
 spine_05 [
  {shapeType=Capsule, length=10.4987, radius=9.5997, shapeName=spine_05_capsule, center=<2.3981 -1.0539 0>, rotation=<rot 0 -90 0>}
  {shapeType=Capsule, length=5.2375, radius=7.6909, shapeName=spine_05_capsule, center=<6.2644 -1.4475 0>}
 ]
 upperarm_l [{
   shapeType=Capsule
   length=22.4727
   radius=4.8304
   shapeName=clavicle_l_capsule
   center=<9.9246 -1.3218 -0.4619>
   rotation=<rot -0.0002 -89.9999 87.3912>
  }]
 upperarm_r [{
   shapeType=Capsule
   length=22.4727
   radius=4.8304
   shapeName=clavicle_l_capsule
   center=<-9.9246 1.3218 0.4619>
   rotation=<rot -0.0002 -89.9999 87.3912>
  }]
}
constraints [
 {bone1Name=dyn_hair_long_01_bk_l, bone2Name=MSTR_hair_long_bk_l, swing1LimitDegrees=10, swing2LimitDegrees=10, twistLimitDegrees=10}
 {bone1Name=dyn_hair_long_02_bk_l, bone2Name=dyn_hair_long_01_bk_l, swing1LimitDegrees=30, swing2LimitDegrees=15, twistLimitDegrees=20}
 {bone1Name=dyn_hair_long_03_bk_l, bone2Name=dyn_hair_long_02_bk_l, swing1LimitDegrees=20, swing2LimitDegrees=10, twistLimitDegrees=15}
 {bone1Name=dyn_hair_long_01_bk, bone2Name=MSTR_hair_long_bk, swing1LimitDegrees=15, swing2LimitDegrees=10, twistLimitDegrees=5}
 {bone1Name=dyn_hair_long_02_bk, bone2Name=dyn_hair_long_01_bk, swing1LimitDegrees=25, swing2LimitDegrees=15, twistLimitDegrees=5}
 {
  bone1Name=dyn_hair_long_03_bk
  bone2Name=dyn_hair_long_02_bk
  swing1LimitDegrees=20
  swing2LimitDegrees=10
  swing2Motion=Locked
  twistLimitDegrees=5
 }
 {bone1Name=dyn_hair_long_01_bk_r, bone2Name=MSTR_hair_long_bk_r, swing1LimitDegrees=10, swing2LimitDegrees=10, twistLimitDegrees=10}
 {bone1Name=dyn_hair_long_02_bk_r, bone2Name=dyn_hair_long_01_bk_r, swing1LimitDegrees=30, swing2LimitDegrees=15, twistLimitDegrees=20}
 {bone1Name=dyn_hair_long_03_bk_r, bone2Name=dyn_hair_long_02_bk_r, swing1LimitDegrees=20, swing2LimitDegrees=10, twistLimitDegrees=15}
 {
  bone1Name=dyn_hair_long_03_bk
  bone2Name=dyn_hair_long_03_bk_r
  swing1LimitDegrees=45
  swing1Motion=Free
  swing2LimitDegrees=45
  swing2Motion=Free
  twistLimitDegrees=45
  twistMotion=Free
 }
 {
  bone1Name=dyn_hair_long_03_bk
  bone2Name=dyn_hair_long_03_bk_l
  swing1LimitDegrees=45
  swing1Motion=Free
  swing2LimitDegrees=45
  swing2Motion=Free
  twistLimitDegrees=45
  twistMotion=Free
 }
 {
  bone1Name=dyn_hair_long_01_side_l
  bone2Name=MSTR_hair_long_side_l
  swing1LimitDegrees=15
  swing2LimitDegrees=10
  twistLimitDegrees=0
  twistMotion=Locked
 }
 {bone1Name=dyn_hair_long_02_side_l, bone2Name=dyn_hair_long_01_side_l, swing1LimitDegrees=15, swing2LimitDegrees=10, twistLimitDegrees=5}
 {bone1Name=dyn_hair_long_03_side_l, bone2Name=dyn_hair_long_02_side_l, swing1LimitDegrees=20, swing2LimitDegrees=15, twistLimitDegrees=5}
 {
  bone1Name=dyn_hair_long_01_fr_l
  bone2Name=MSTR_hair_long_fr_l
  swing1LimitDegrees=15
  swing2LimitDegrees=15
  twistLimitDegrees=0
  twistMotion=Locked
 }
 {bone1Name=dyn_hair_long_02_fr_l, bone2Name=dyn_hair_long_01_fr_l, swing1LimitDegrees=15, swing2LimitDegrees=15, twistLimitDegrees=5}
 {bone1Name=dyn_hair_long_03_fr_l, bone2Name=dyn_hair_long_02_fr_l, swing1LimitDegrees=20, swing2LimitDegrees=15, twistLimitDegrees=5}
 {
  bone1Name=dyn_hair_long_03_side_l
  bone2Name=dyn_hair_long_03_fr_l
  swing1LimitDegrees=45
  swing1Motion=Free
  swing2LimitDegrees=45
  swing2Motion=Free
  twistLimitDegrees=45
  twistMotion=Free
 }
 {
  bone1Name=dyn_hair_long_01_side_r
  bone2Name=MSTR_hair_long_side_r
  swing1LimitDegrees=15
  swing2LimitDegrees=10
  twistLimitDegrees=0
  twistMotion=Locked
 }
 {bone1Name=dyn_hair_long_02_side_r, bone2Name=dyn_hair_long_01_side_r, swing1LimitDegrees=15, swing2LimitDegrees=10, twistLimitDegrees=5}
 {bone1Name=dyn_hair_long_03_side_r, bone2Name=dyn_hair_long_02_side_r, swing1LimitDegrees=20, swing2LimitDegrees=15, twistLimitDegrees=5}
 {bone1Name=dyn_hair_long_01_fr_r, bone2Name=MSTR_hair_long_fr_r, swing1LimitDegrees=15, swing2LimitDegrees=15, twistLimitDegrees=0}
 {bone1Name=dyn_hair_long_02_fr_r, bone2Name=dyn_hair_long_01_fr_r, swing1LimitDegrees=15, swing2LimitDegrees=15, twistLimitDegrees=5}
 {bone1Name=dyn_hair_long_03_fr_r, bone2Name=dyn_hair_long_02_fr_r, swing1LimitDegrees=20, swing2LimitDegrees=15, twistLimitDegrees=5}
 {
  bone1Name=dyn_hair_long_03_side_r
  bone2Name=dyn_hair_long_03_fr_r
  swing1LimitDegrees=45
  swing1Motion=Free
  swing2LimitDegrees=45
  swing2Motion=Free
  twistLimitDegrees=45
  twistMotion=Free
 }
]
