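# Anim Blueprint for the SummitReedGrit male face accessory part, adding secondary motion to its long-hair bones.
# It copies the attached parent's pose, runs a skydive-aware Control Rig, and caches the result.
# bRigidBodyEnabled selects between a RigidBody simulation and a fallback chain of per-bone AnimDynamics springs on front, side and back strands.
# Each update, a nativized state function derives gravity and rigid-body settings from per-movement-state parameter sets and stores them.
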AnimBlueprint /BRCosmetics/Characters/Player/Male/Medium/Bodies/M_MED_SummitReedGrit/Meshes/Parts/M_MED_SummitReedGrit_FaceAcc_AnimBP
target_skeleton=&/M_MED_SummitReedGrit_FaceAcc_Skeleton
graphs [
 AnimGraph {nodes [
   n1 = Anim Dynamics - Bone: dyn_hair_long_01_fr_l(ComponentPose=Local To Component(LocalPose=Use cached pose 'DefBake'()), Alpha=1, ExternalForce="0,0,0") {
    angularDampingOverride=0.999
    angularSpringConstant=600
    bAngularSpring=true
    bGravityOverrideInSimSpace=true
    bOverrideAngularDamping=true
    bUseGravityOverride=true
    simulationSpace=RootRelative
    boundBone {boneName=dyn_hair_long_01_fr_l}
    componentAppliedLinearAccClamp=<1500 1500 2000>
    componentLinearAccScale=<0.3 0.3 0.5>
    componentLinearVelScale=<0.3 0.3 0.2>
    gravityOverride {z=-980}
    physicsBodyDefinitions [{
      boundBone {boneName=dyn_hair_long_01_fr_l}
      boxExtents=<15 15 15>
      constraintSetup {angularLimitsMax=<0 5 0>, angularLimitsMin=<0 -10 -7>, angularTarget=<0 0 0>}
      localJointOffset=<1 0 0>
     }]
    relativeSpaceBone {boneName=head}
    simSpaceSettings {maxAngularAcceleration=1500, maxAngularVelocity=1500, simSpaceAngularAlpha=0.1}
   }
   n2 = Anim Dynamics - Bone: dyn_hair_long_02_fr_l(ComponentPose=@n1, Alpha=1, ExternalForce="0,0,0") {
    angularDampingOverride=0.999
    angularSpringConstant=600
    bAngularSpring=true
    bGravityOverrideInSimSpace=true
    bOverrideAngularDamping=true
    bUseGravityOverride=true
    simulationSpace=RootRelative
    boundBone {boneName=dyn_hair_long_02_fr_l}
    componentAppliedLinearAccClamp=<1500 1500 2000>
    componentLinearAccScale=<0.3 0.3 0.5>
    componentLinearVelScale=<0.3 0.3 0.2>
    gravityOverride {z=-980}
    physicsBodyDefinitions [{
      boundBone {boneName=dyn_hair_long_02_fr_l}
      boxExtents=<13 13 13>
      constraintSetup {angularLimitsMax=<0 5 10>, angularLimitsMin=<0 -10 -10>, angularTarget=<0 0 0>}
      localJointOffset=<1 0 0>
     }]
    relativeSpaceBone {boneName=head}
    simSpaceSettings {maxAngularAcceleration=1500, maxAngularVelocity=1500, simSpaceAngularAlpha=0.1}
   }
   n3 = Anim Dynamics - Bone: dyn_hair_long_03_fr_l(ComponentPose=@n2, Alpha=1, ExternalForce="0,0,0") {
    angularDampingOverride=0.999
    angularSpringConstant=600
    bAngularSpring=true
    bGravityOverrideInSimSpace=true
    bOverrideAngularDamping=true
    bUseGravityOverride=true
    simulationSpace=RootRelative
    boundBone {boneName=dyn_hair_long_03_fr_l}
    componentAppliedLinearAccClamp=<1500 1500 2000>
    componentLinearAccScale=<0.3 0.3 0.5>
    componentLinearVelScale=<0.3 0.3 0.2>
    gravityOverride {z=-980}
    physicsBodyDefinitions [{
      boundBone {boneName=dyn_hair_long_03_fr_l}
      boxExtents=<12 12 12>
      constraintSetup {angularLimitsMax=<0 5 0>, angularLimitsMin=<0 -10 -10>, angularTarget=<0 0 0>}
      localJointOffset=<1 0 0>
     }]
    relativeSpaceBone {boneName=head}
    simSpaceSettings {maxAngularAcceleration=1500, maxAngularVelocity=1500, simSpaceAngularAlpha=0.1}
   }
   n4 = Anim Dynamics - Bone: dyn_hair_long_01_fr_r(ComponentPose=@n3, Alpha=1, ExternalForce="0,0,0") {
    angularDampingOverride=0.999
    angularSpringConstant=600
    bAngularSpring=true
    bGravityOverrideInSimSpace=true
    bOverrideAngularDamping=true
    bUseGravityOverride=true
    simulationSpace=RootRelative
    boundBone {boneName=dyn_hair_long_01_fr_r}
    componentAppliedLinearAccClamp=<1500 1500 2000>
    componentLinearAccScale=<0.3 0.3 0.5>
    componentLinearVelScale=<0.3 0.3 0.2>
    gravityOverride {z=-980}
    physicsBodyDefinitions [{
      boundBone {boneName=dyn_hair_long_01_fr_r}
      boxExtents=<15 15 15>
      constraintSetup {angularLimitsMax=<0 5 0>, angularLimitsMin=<0 -10 -7>, angularTarget=<0 0 0>}
      localJointOffset=<-1 0 0>
     }]
    relativeSpaceBone {boneName=head}
    simSpaceSettings {maxAngularAcceleration=1500, maxAngularVelocity=1500, simSpaceAngularAlpha=0.1}
   }
   n5 = Anim Dynamics - Bone: dyn_hair_long_02_fr_r(ComponentPose=@n4, Alpha=1, ExternalForce="0,0,0") {
    angularDampingOverride=0.999
    angularSpringConstant=600
    bAngularSpring=true
    bGravityOverrideInSimSpace=true
    bOverrideAngularDamping=true
    bUseGravityOverride=true
    simulationSpace=RootRelative
    boundBone {boneName=dyn_hair_long_02_fr_r}
    componentAppliedLinearAccClamp=<1500 1500 2000>
    componentLinearAccScale=<0.3 0.3 0.5>
    componentLinearVelScale=<0.3 0.3 0.2>
    gravityOverride {z=-980}
    physicsBodyDefinitions [{
      boundBone {boneName=dyn_hair_long_02_fr_r}
      boxExtents=<13 13 13>
      constraintSetup {angularLimitsMax=<0 5 10>, angularLimitsMin=<0 -10 -10>, angularTarget=<0 0 0>}
      localJointOffset=<-1 0 0>
     }]
    relativeSpaceBone {boneName=head}
    simSpaceSettings {maxAngularAcceleration=1500, maxAngularVelocity=1500, simSpaceAngularAlpha=0.1}
   }
   n6 = Anim Dynamics - Bone: dyn_hair_long_03_fr_r(ComponentPose=@n5, Alpha=1, ExternalForce="0,0,0") {
    angularDampingOverride=0.999
    angularSpringConstant=600
    bAngularSpring=true
    bGravityOverrideInSimSpace=true
    bOverrideAngularDamping=true
    bUseGravityOverride=true
    simulationSpace=RootRelative
    boundBone {boneName=dyn_hair_long_03_fr_r}
    componentAppliedLinearAccClamp=<1500 1500 2000>
    componentLinearAccScale=<0.3 0.3 0.5>
    componentLinearVelScale=<0.3 0.3 0.2>
    gravityOverride {z=-980}
    physicsBodyDefinitions [{
      boundBone {boneName=dyn_hair_long_03_fr_r}
      boxExtents=<12 12 12>
      constraintSetup {angularLimitsMax=<0 5 0>, angularLimitsMin=<0 -10 -10>, angularTarget=<0 0 0>}
      localJointOffset=<-1 0 0>
     }]
    relativeSpaceBone {boneName=head}
    simSpaceSettings {maxAngularAcceleration=1500, maxAngularVelocity=1500, simSpaceAngularAlpha=0.1}
   }
   n7 = Anim Dynamics - Bone: dyn_hair_long_01_side_l(ComponentPose=@n6, Alpha=1, ExternalForce="0,0,0") {
    angularDampingOverride=0.999
    angularSpringConstant=600
    bAngularSpring=true
    bGravityOverrideInSimSpace=true
    bOverrideAngularDamping=true
    bUseGravityOverride=true
    simulationSpace=RootRelative
    boundBone {boneName=dyn_hair_long_01_side_l}
    componentAppliedLinearAccClamp=<1500 1500 2000>
    componentLinearAccScale=<0.3 0.3 0.5>
    componentLinearVelScale=<0.3 0.3 0.2>
    gravityOverride {z=-980}
    physicsBodyDefinitions [{
      boundBone {boneName=dyn_hair_long_01_side_l}
      boxExtents=<15 15 15>
      constraintSetup {angularLimitsMax=<0 5 0>, angularLimitsMin=<0 -10 -7>, angularTarget=<0 0 0>}
      localJointOffset=<1 0 0>
     }]
    relativeSpaceBone {boneName=head}
    simSpaceSettings {maxAngularAcceleration=1500, maxAngularVelocity=1500, simSpaceAngularAlpha=0.1}
   }
   n8 = Anim Dynamics - Bone: dyn_hair_long_02_side_l(ComponentPose=@n7, Alpha=1, ExternalForce="0,0,0") {
    angularDampingOverride=0.999
    angularSpringConstant=600
    bAngularSpring=true
    bGravityOverrideInSimSpace=true
    bOverrideAngularDamping=true
    bUseGravityOverride=true
    simulationSpace=RootRelative
    boundBone {boneName=dyn_hair_long_02_side_l}
    componentAppliedLinearAccClamp=<1500 1500 2000>
    componentLinearAccScale=<0.3 0.3 0.5>
    componentLinearVelScale=<0.3 0.3 0.2>
    gravityOverride {z=-980}
    physicsBodyDefinitions [{
      boundBone {boneName=dyn_hair_long_02_side_l}
      boxExtents=<13 13 13>
      constraintSetup {angularLimitsMax=<0 5 10>, angularLimitsMin=<0 -10 -7>, angularTarget=<0 0 0>}
      localJointOffset=<1 0 0>
     }]
    relativeSpaceBone {boneName=head}
    simSpaceSettings {maxAngularAcceleration=1500, maxAngularVelocity=1500, simSpaceAngularAlpha=0.1}
   }
   n9 = Anim Dynamics - Bone: dyn_hair_long_03_side_l(ComponentPose=@n8, Alpha=1, ExternalForce="0,0,0") {
    angularDampingOverride=0.999
    angularSpringConstant=600
    bAngularSpring=true
    bGravityOverrideInSimSpace=true
    bOverrideAngularDamping=true
    bUseGravityOverride=true
    simulationSpace=RootRelative
    boundBone {boneName=dyn_hair_long_03_side_l}
    componentAppliedLinearAccClamp=<1500 1500 2000>
    componentLinearAccScale=<0.3 0.3 0.5>
    componentLinearVelScale=<0.3 0.3 0.2>
    gravityOverride {z=-980}
    physicsBodyDefinitions [{
      boundBone {boneName=dyn_hair_long_03_side_l}
      boxExtents=<12 12 12>
      constraintSetup {angularLimitsMax=<0 5 0>, angularLimitsMin=<0 -10 -7>, angularTarget=<0 0 0>}
      localJointOffset=<1 0 0>
     }]
    relativeSpaceBone {boneName=head}
    simSpaceSettings {maxAngularAcceleration=1500, maxAngularVelocity=1500, simSpaceAngularAlpha=0.1}
   }
   n10 = Anim Dynamics - Bone: dyn_hair_long_01_side_r(ComponentPose=@n9, Alpha=1, ExternalForce="0,0,0") {
    angularDampingOverride=0.999
    angularSpringConstant=600
    bAngularSpring=true
    bGravityOverrideInSimSpace=true
    bOverrideAngularDamping=true
    bUseGravityOverride=true
    simulationSpace=RootRelative
    boundBone {boneName=dyn_hair_long_01_side_r}
    componentAppliedLinearAccClamp=<1500 1500 2000>
    componentLinearAccScale=<0.3 0.3 0.5>
    componentLinearVelScale=<0.3 0.3 0.2>
    gravityOverride {z=-980}
    physicsBodyDefinitions [{
      boundBone {boneName=dyn_hair_long_01_side_r}
      boxExtents=<15 15 15>
      constraintSetup {angularLimitsMax=<0 5 0>, angularLimitsMin=<0 -10 -7>, angularTarget=<0 0 0>}
      localJointOffset=<-1 0 0>
     }]
    relativeSpaceBone {boneName=head}
    simSpaceSettings {maxAngularAcceleration=1500, maxAngularVelocity=1500, simSpaceAngularAlpha=0.1}
   }
   n11 = Anim Dynamics - Bone: dyn_hair_long_02_side_r(ComponentPose=@n10, Alpha=1, ExternalForce="0,0,0") {
    angularDampingOverride=0.999
    angularSpringConstant=600
    bAngularSpring=true
    bGravityOverrideInSimSpace=true
    bOverrideAngularDamping=true
    bUseGravityOverride=true
    simulationSpace=RootRelative
    boundBone {boneName=dyn_hair_long_02_side_r}
    componentAppliedLinearAccClamp=<1500 1500 2000>
    componentLinearAccScale=<0.3 0.3 0.5>
    componentLinearVelScale=<0.3 0.3 0.2>
    gravityOverride {z=-980}
    physicsBodyDefinitions [{
      boundBone {boneName=dyn_hair_long_02_side_r}
      boxExtents=<13 13 13>
      constraintSetup {angularLimitsMax=<0 5 10>, angularLimitsMin=<0 -10 -10>, angularTarget=<0 0 0>}
      localJointOffset=<-1 0 0>
     }]
    relativeSpaceBone {boneName=head}
    simSpaceSettings {maxAngularAcceleration=1500, maxAngularVelocity=1500, simSpaceAngularAlpha=0.1}
   }
   n12 = Anim Dynamics - Bone: dyn_hair_long_03_side_r(ComponentPose=@n11, Alpha=1, ExternalForce="0,0,0") {
    angularDampingOverride=0.999
    angularSpringConstant=600
    bAngularSpring=true
    bGravityOverrideInSimSpace=true
    bOverrideAngularDamping=true
    bUseGravityOverride=true
    simulationSpace=RootRelative
    boundBone {boneName=dyn_hair_long_03_side_r}
    componentAppliedLinearAccClamp=<1500 1500 2000>
    componentLinearAccScale=<0.3 0.3 0.5>
    componentLinearVelScale=<0.3 0.3 0.2>
    gravityOverride {z=-980}
    physicsBodyDefinitions [{
      boundBone {boneName=dyn_hair_long_03_side_r}
      boxExtents=<12 12 12>
      constraintSetup {angularLimitsMax=<0 5 0>, angularLimitsMin=<0 -10 -10>, angularTarget=<0 0 0>}
      localJointOffset=<-1 0 0>
     }]
    relativeSpaceBone {boneName=head}
    simSpaceSettings {maxAngularAcceleration=1500, maxAngularVelocity=1500, simSpaceAngularAlpha=0.1}
   }
   n13 = Anim Dynamics - Bone: dyn_hair_long_01_bk_l(ComponentPose=@n12, Alpha=1, ExternalForce="0,0,0") {
    angularDampingOverride=0.999
    angularSpringConstant=600
    bAngularSpring=true
    bOverrideAngularDamping=true
    bUseGravityOverride=true
    simulationSpace=RootRelative
    boundBone {boneName=dyn_hair_long_01_bk_l}
    componentAppliedLinearAccClamp=<1500 1500 2000>
    componentLinearAccScale=<0.3 0.3 0.5>
    componentLinearVelScale=<0.3 0.3 0.2>
    gravityOverride {z=-980}
    physicsBodyDefinitions [{
      boundBone {boneName=dyn_hair_long_01_bk_l}
      boxExtents=<15 15 15>
      constraintSetup {angularLimitsMax=<0 0 0>, angularLimitsMin=<0 0 -10>, angularTarget=<0 0 0>}
      localJointOffset=<1 0 0>
     }]
    simSpaceSettings {maxAngularAcceleration=1500, maxAngularVelocity=1500, simSpaceAngularAlpha=0.1}
   }
   n14 = Anim Dynamics - Bone: dyn_hair_long_02_bk_l(ComponentPose=@n13, Alpha=1, ExternalForce="0,0,0") {
    angularDampingOverride=0.999
    angularSpringConstant=600
    bAngularSpring=true
    bOverrideAngularDamping=true
    bUseGravityOverride=true
    simulationSpace=RootRelative
    boundBone {boneName=dyn_hair_long_02_bk_l}
    componentAppliedLinearAccClamp=<1500 1500 2000>
    componentLinearAccScale=<0.3 0.3 0.5>
    componentLinearVelScale=<0.3 0.3 0.2>
    gravityOverride {z=-980}
    physicsBodyDefinitions [{
      boundBone {boneName=dyn_hair_long_02_bk_l}
      boxExtents=<13 13 13>
      constraintSetup {angularLimitsMax=<0 0 0>, angularLimitsMin=<0 0 -20>, angularTarget=<0 0 0>}
      localJointOffset=<1 0 0>
     }]
    simSpaceSettings {maxAngularAcceleration=1500, maxAngularVelocity=1500, simSpaceAngularAlpha=0.1}
   }
   n15 = Anim Dynamics - Bone: dyn_hair_long_03_bk_l(ComponentPose=@n14, Alpha=1, ExternalForce="0,0,0") {
    angularDampingOverride=0.999
    angularSpringConstant=600
    bAngularSpring=true
    bOverrideAngularDamping=true
    bUseGravityOverride=true
    simulationSpace=RootRelative
    boundBone {boneName=dyn_hair_long_03_bk_l}
    componentAppliedLinearAccClamp=<1500 1500 2000>
    componentLinearAccScale=<0.3 0.3 0.5>
    componentLinearVelScale=<0.3 0.3 0.2>
    gravityOverride {z=-980}
    physicsBodyDefinitions [{
      boundBone {boneName=dyn_hair_long_03_bk_l}
      boxExtents=<12 12 12>
      constraintSetup {angularLimitsMax=<0 0 0>, angularLimitsMin=<0 0 -20>, angularTarget=<0 0 0>}
      localJointOffset=<1 0 0>
     }]
    simSpaceSettings {maxAngularAcceleration=1500, maxAngularVelocity=1500, simSpaceAngularAlpha=0.1}
   }
   n16 = Anim Dynamics - Bone: dyn_hair_long_01_bk_r(ComponentPose=@n15, Alpha=1, ExternalForce="0,0,0") {
    angularDampingOverride=0.999
    angularSpringConstant=600
    bAngularSpring=true
    bOverrideAngularDamping=true
    bUseGravityOverride=true
    simulationSpace=RootRelative
    boundBone {boneName=dyn_hair_long_01_bk_r}
    componentAppliedLinearAccClamp=<1500 1500 2000>
    componentLinearAccScale=<0.3 0.3 0.5>
    componentLinearVelScale=<0.3 0.3 0.2>
    gravityOverride {z=-980}
    physicsBodyDefinitions [{
      boundBone {boneName=dyn_hair_long_01_bk_r}
      boxExtents=<15 15 15>
      constraintSetup {angularLimitsMax=<0 0 0>, angularLimitsMin=<0 0 -10>, angularTarget=<0 0 0>}
      localJointOffset=<1 0 0>
     }]
    simSpaceSettings {maxAngularAcceleration=1500, maxAngularVelocity=1500, simSpaceAngularAlpha=0.1}
   }
   n17 = Anim Dynamics - Bone: dyn_hair_long_02_bk_r(ComponentPose=@n16, Alpha=1, ExternalForce="0,0,0") {
    angularDampingOverride=0.999
    angularSpringConstant=600
    bAngularSpring=true
    bOverrideAngularDamping=true
    bUseGravityOverride=true
    simulationSpace=RootRelative
    boundBone {boneName=dyn_hair_long_02_bk_r}
    componentAppliedLinearAccClamp=<1500 1500 2000>
    componentLinearAccScale=<0.3 0.3 0.5>
    componentLinearVelScale=<0.3 0.3 0.2>
    gravityOverride {z=-980}
    physicsBodyDefinitions [{
      boundBone {boneName=dyn_hair_long_02_bk_r}
      boxExtents=<13 13 13>
      constraintSetup {angularLimitsMax=<0 0 0>, angularLimitsMin=<0 0 -20>, angularTarget=<0 0 0>}
      localJointOffset=<1 0 0>
     }]
    simSpaceSettings {maxAngularAcceleration=1500, maxAngularVelocity=1500, simSpaceAngularAlpha=0.1}
   }
   n18 = Anim Dynamics - Bone: dyn_hair_long_03_bk_r(ComponentPose=@n17, Alpha=1, ExternalForce="0,0,0") {
    angularDampingOverride=0.999
    angularSpringConstant=600
    bAngularSpring=true
    bOverrideAngularDamping=true
    bUseGravityOverride=true
    simulationSpace=RootRelative
    boundBone {boneName=dyn_hair_long_03_bk_r}
    componentAppliedLinearAccClamp=<1500 1500 2000>
    componentLinearAccScale=<0.3 0.3 0.5>
    componentLinearVelScale=<0.3 0.3 0.2>
    gravityOverride {z=-980}
    physicsBodyDefinitions [{
      boundBone {boneName=dyn_hair_long_03_bk_r}
      boxExtents=<12 12 12>
      constraintSetup {angularLimitsMax=<0 0 0>, angularLimitsMin=<0 0 -20>, angularTarget=<0 0 0>}
      localJointOffset=<1 0 0>
     }]
    simSpaceSettings {maxAngularAcceleration=1500, maxAngularVelocity=1500, simSpaceAngularAlpha=0.1}
   }
   n19 = Anim Dynamics - Bone: dyn_hair_long_01_bk(ComponentPose=@n18, Alpha=1, ExternalForce="0,0,0") {
    angularDampingOverride=0.999
    angularSpringConstant=600
    bAngularSpring=true
    bOverrideAngularDamping=true
    bUseGravityOverride=true
    simulationSpace=RootRelative
    boundBone {boneName=dyn_hair_long_01_bk}
    componentAppliedLinearAccClamp=<1500 1500 2000>
    componentLinearAccScale=<0.3 0.3 0.5>
    componentLinearVelScale=<0.3 0.3 0.2>
    gravityOverride {z=-980}
    physicsBodyDefinitions [{
      boundBone {boneName=dyn_hair_long_01_bk}
      boxExtents=<15 15 15>
      constraintSetup {angularLimitsMax=<0 0 0>, angularLimitsMin=<0 0 -10>, angularTarget=<0 0 0>}
      localJointOffset=<1 0 0>
     }]
    simSpaceSettings {maxAngularAcceleration=1500, maxAngularVelocity=1500, simSpaceAngularAlpha=0.1}
   }
   n20 = Anim Dynamics - Bone: dyn_hair_long_02_bk(ComponentPose=@n19, Alpha=1, ExternalForce="0,0,0") {
    angularDampingOverride=0.999
    angularSpringConstant=600
    bAngularSpring=true
    bOverrideAngularDamping=true
    bUseGravityOverride=true
    simulationSpace=RootRelative
    boundBone {boneName=dyn_hair_long_02_bk}
    componentAppliedLinearAccClamp=<1500 1500 2000>
    componentLinearAccScale=<0.3 0.3 0.5>
    componentLinearVelScale=<0.3 0.3 0.2>
    gravityOverride {z=-980}
    physicsBodyDefinitions [{
      boundBone {boneName=dyn_hair_long_02_bk}
      boxExtents=<13 13 13>
      constraintSetup {angularLimitsMax=<0 5 0>, angularLimitsMin=<0 -5 -20>, angularTarget=<0 0 0>}
      localJointOffset=<1 0 0>
     }]
    simSpaceSettings {maxAngularAcceleration=1500, maxAngularVelocity=1500, simSpaceAngularAlpha=0.1}
   }
   n21 = Anim Dynamics - Bone: dyn_hair_long_03_bk(ComponentPose=@n20, Alpha=1, ExternalForce="0,0,0") {
    angularDampingOverride=0.999
    angularSpringConstant=600
    bAngularSpring=true
    bOverrideAngularDamping=true
    bUseGravityOverride=true
    simulationSpace=RootRelative
    boundBone {boneName=dyn_hair_long_03_bk}
    componentAppliedLinearAccClamp=<1500 1500 2000>
    componentLinearAccScale=<0.3 0.3 0.5>
    componentLinearVelScale=<0.3 0.3 0.2>
    gravityOverride {z=-980}
    physicsBodyDefinitions [{
      boundBone {boneName=dyn_hair_long_03_bk}
      boxExtents=<12 12 12>
      constraintSetup {angularLimitsMax=<0 5 0>, angularLimitsMin=<0 -5 -20>, angularTarget=<0 0 0>}
      localJointOffset=<1 0 0>
     }]
    simSpaceSettings {maxAngularAcceleration=1500, maxAngularVelocity=1500, simSpaceAngularAlpha=0.1}
   }
   n22 = RigidBody!(
    ComponentPose=Local To Component(LocalPose=Use cached pose 'DefBake'())
    Alpha
    OverrideWorldGravity:Vector="0,0,0"
    ExternalForce
    SimSpaceSettings:Sim Space Settings Structure=(WorldAlpha=1,VelocityScaleZ=3,DampingAlpha=0,MaxLinearVelocity=15,MaxAngularVelocity=15,MaxLinearAcceleration=300,MaxAngularAcceleration=150,ExternalLinearDragV=(X=0,Y=0,Z=0),ExternalLinearVelocity=(X=0,Y=0,Z=0),ExternalAngularVelocity=(X=0,Y=0,Z=0))
   ) -> (Pose) {
    bClampLinearTranslationLimitToRefPose=true
    bOverrideWorldGravity=true
    overridePhysicsAsset=&/M_MED_SummitReedGrit_FaceAcc_Physics
    simulationSpace=BaseBoneSpace
    baseBoneRef {boneName=spine_05}
    simSpaceSettings {
     dampingAlpha=0
     maxAngularAcceleration=150
     maxAngularVelocity=15
     maxLinearAcceleration=300
     maxLinearVelocity=15
     velocityScaleZ=3
     worldAlpha=1
    }
   }
   Output Pose(Result=Blend Poses by bool(BlendPose_0=Component To Local(ComponentPose=@n22), BlendPose_1=Component To Local(ComponentPose=@n21), bActiveValue=Get bRigidBodyEnabled())) {layerGroup=None}
   n23 = Control Rig!(Source=Copy Pose From Mesh() {bUseAttachedParent=true}, Alpha, isSkyDiving:bool=Get bIsSkydiveFloating()) -> (Pose) {controlRigAssetReference {blueprintRigClass=&/M_MED_SummitReedGrit_FaceAcc_CtrlRig.M_MED_SummitReedGrit_FaceAcc_CtrlRig_C}}
   Save cached pose 'DefBake'(Pose=@n23) {cacheName=DefBake}
  ]}
 EventGraph {nodes [
   Event BlueprintUpdateAnimation()
   > n1 = "Gravity Override State Function (Nativized)"!(
    execute
    CharacterPartAnimInstance:Custom Character Part Anim Instance Object Reference=Self-Reference()
    GravityOverrideParameterSet:Fort Gravity Override Parameter Set Structure=Get Gravity Override Params() -> (Gravity Override Params:Fort Gravity Override Parameter Set Structure)
    RigidBodyAnimNodeParameterSet:Fort Rigid Body Anim Node Parameter Set Structure=Get RBAN Params() -> (RBAN Params:Fort Rigid Body Anim Node Parameter Set Structure)
    DebugDraw:bool=false
    GameTimeInSeconds:float=GetGameTimeInSeconds()
    DebugJointName:Name=root
    StateLogicLayerName:Name=none
   ) -> (
    then
    OutGravityOverride:Vector
    OutCurrentState:Name=None
    OutGravityOverrideParameters:Fort Gravity Override Parameters Structure
    OutRigidBodyAnimNodeParameters:Fort Rigid Body Anim Node Parameters Structure
   )
   > Set Gravity Override(Gravity Override=@n1)
   > Set RBAN Settings Out(RBAN Settings Out=@n1.OutRigidBodyAnimNodeParameters)
   n2 = Event OnCharacterCustomizationLoaded()
   > Set HeadComponent(HeadComponent:Skeletal Mesh Component Budgeted Object Reference=Get Array Item with EFort Custom Part Type(Array=@n2.CustomMeshParts, PartType=Head)) -> (Output_Get:Skeletal Mesh Component Budgeted Object Reference)
  ]}
]
variables [
 Gravity Override {type=Vector}
 Gravity Override Params {
  type=FortGravityOverrideParameterSet
  default=(Skydive_GoingUp=(GravityOverride=(WindFrequency=(X=1.500000,Y=7.000000,Z=8.000000),WindAmplitude=(X=200.000000,Y=-500.000000,Z=400.000000),GravityOverride=(X=0.000000,Y=-900.000000,Z=900.000000),JointName="root")),Skydive_FloatingDown=(GravityOverride=(WindFrequency=(X=1.500000,Y=7.000000,Z=8.000000),WindAmplitude=(X=-200.000000,Y=-300.000000,Z=-600.000000),WindOffset=(X=0.000000,Y=-300.000000,Z=-450.000000),GravityOverride=(X=0.000000,Y=-600.000000,Z=300.000000),JointName="root")),Skydive_DivingDown=(GravityOverride=(WindFrequency=(X=1.500000,Y=12.000000,Z=10.000000),WindAmplitude=(X=50.000000,Y=300.000000,Z=100.000000),GravityOverride=(X=-1500.000000,Y=0.000000,Z=0.000000),JointName="pelvis")),Skydive_Parachuting=(GravityOverride=(WindFrequency=(X=1.500000,Y=7.000000,Z=8.000000),WindAmplitude=(X=200.000000,Y=-500.000000,Z=400.000000),GravityOverride=(X=0.000000,Y=-800.000000,Z=-980.000000),JointName="root")),OnGround_Standing=(GravityOverride=(GravityOverride=(X=0.000000,Y=0.000000,Z=-980.000000),JointName="root")),OnGround_Moving=(GravityOverride=(GravityOverride=(X=0.000000,Y=0.000000,Z=-980.000000),JointName="root")),OnGround_Moving_HighRotationRate=(GravityOverride=(GravityOverride=(X=0.000000,Y=0.000000,Z=-980.000000),JointName="root")),OnGround_Crouch_Moving=(GravityOverride=(GravityOverride=(X=0.000000,Y=0.000000,Z=-980.000000),JointName="root")),Emote_or_Melee=(GravityOverride=(GravityOverride=(X=0.000000,Y=0.000000,Z=-980.000000),JointName="root")),Swimming=(GravityOverride=(GravityOverride=(X=-800.000000,Y=0.000000,Z=0.000000),JointName="pelvis")),Front_End=(GravityOverride=(GravityOverride=(X=0.000000,Y=0.000000,Z=-980.000000),JointName="root")),DBNO=(GravityOverride=(GravityOverride=(X=-980.000000,Y=400.000000,Z=0.000000),JointName="pelvis")),DBNO_Carried=(GravityOverride=(GravityOverride=(X=-980.000000,Y=500.000000,Z=-500.000000),JointName="pelvis")),OnGround_Falling=(GravityOverride=(GravityOverride=(X=0.000000,Y=0.000000,Z=-980.000000),JointName="root")),Ziplining=(GravityOverride=(GravityOverride=(X=0.000000,Y=0.000000,Z=-980.000000),JointName="root")),Skydive_Surfing=(GravityOverride=(WindFrequency=(X=1.500000,Y=7.000000,Z=8.000000),WindAmplitude=(X=200.000000,Y=-500.000000,Z=400.000000),GravityOverride=(X=500.000000,Y=-300.000000,Z=-980.000000),JointName="root")),Riding_Standing=(GravityOverride=(GravityOverride=(X=0.000000,Y=-350.000000,Z=-980.000000),JointName="root")),Riding_Moving=(GravityOverride=(GravityOverride=(X=0.000000,Y=-450.000000,Z=-450.000000),JointName="root")),Riding_Moving_Targeting=(GravityOverride=(GravityOverride=(X=0.000000,Y=-200.000000,Z=-980.000000),JointName="root")),Tactical_Sprint=(GravityOverride=(GravityOverride=(X=0.000000,Y=-450.000000,Z=-500.000000),JointName="root")),Rail_Grinding=(GravityOverride=(GravityOverride=(X=0.000000,Y=0.000000,Z=-980.000000),JointName="root")),)
 }
 HeadComponent {type=SkeletalMeshComponentBudgeted}
 RBAN Params {
  type=FortRigidBodyAnimNodeParameterSet
  default=(Skydive_GoingUp=(RigidBodyAnimNode=(SimSpaceSettings=(WorldAlpha=1.000000,DampingAlpha=0.000000,MaxLinearVelocity=15.000000,MaxAngularVelocity=30.000000,MaxLinearAcceleration=600.000000,MaxAngularAcceleration=150.000000,ExternalLinearDragV=(X=0.000000,Y=0.500000,Z=0.000000)))),Skydive_FloatingDown=(RigidBodyAnimNode=(ComponentLinearAccScale=(X=0.000000,Y=0.500000,Z=0.000000),ComponentLinearVelScale=(X=0.000000,Y=0.500000,Z=0.000000),SimSpaceSettings=(WorldAlpha=1.000000,DampingAlpha=0.000000,MaxLinearVelocity=15.000000,MaxAngularVelocity=30.000000,MaxLinearAcceleration=600.000000,MaxAngularAcceleration=150.000000,ExternalLinearDragV=(X=0.000000,Y=0.500000,Z=0.000000)))),Skydive_DivingDown=(RigidBodyAnimNode=(ComponentLinearAccScale=(X=0.100000,Y=0.000000,Z=0.000000),ComponentLinearVelScale=(X=0.100000,Y=0.000000,Z=0.000000),SimSpaceSettings=(WorldAlpha=1.000000,DampingAlpha=0.000000,MaxLinearVelocity=15.000000,MaxAngularVelocity=30.000000,MaxLinearAcceleration=600.000000,MaxAngularAcceleration=150.000000,ExternalLinearDragV=(X=0.000000,Y=0.500000,Z=0.000000)))),Skydive_Parachuting=(RigidBodyAnimNode=(ComponentLinearAccScale=(X=0.000000,Y=0.100000,Z=0.300000),ComponentLinearVelScale=(X=0.000000,Y=0.100000,Z=0.300000),SimSpaceSettings=(WorldAlpha=1.000000,DampingAlpha=0.000000,MaxLinearVelocity=15.000000,MaxAngularVelocity=30.000000,MaxLinearAcceleration=600.000000,MaxAngularAcceleration=150.000000,ExternalLinearDragV=(X=0.000000,Y=0.500000,Z=0.000000)))),OnGround_Standing=(RigidBodyAnimNode=(ComponentLinearAccScale=(X=0.200000,Y=0.000000,Z=0.000000),ComponentLinearVelScale=(X=0.200000,Y=0.000000,Z=0.000000),SimSpaceSettings=(WorldAlpha=1.000000,DampingAlpha=0.000000,MaxLinearVelocity=30.000000,MaxAngularVelocity=20.000000,MaxLinearAcceleration=1200.000000,MaxAngularAcceleration=80.000000,ExternalLinearDragV=(X=0.000000,Y=0.500000,Z=0.000000)))),OnGround_Moving=(RigidBodyAnimNode=(ComponentLinearAccScale=(X=0.200000,Y=0.000000,Z=0.000000),ComponentLinearVelScale=(X=0.200000,Y=0.000000,Z=0.000000),SimSpaceSettings=(WorldAlpha=1.000000,DampingAlpha=0.000000,MaxLinearVelocity=30.000000,MaxAngularVelocity=20.000000,MaxLinearAcceleration=1200.000000,MaxAngularAcceleration=80.000000,ExternalLinearDragV=(X=0.000000,Y=0.500000,Z=0.000000)))),OnGround_Moving_HighRotationRate=(RigidBodyAnimNode=(SimSpaceSettings=(WorldAlpha=1.000000,DampingAlpha=0.000000,MaxLinearVelocity=15.000000,MaxAngularVelocity=15.000000,MaxLinearAcceleration=600.000000,MaxAngularAcceleration=90.000000,ExternalLinearDragV=(X=0.000000,Y=0.500000,Z=0.000000)))),OnGround_Crouch_Moving=(RigidBodyAnimNode=(SimSpaceSettings=(WorldAlpha=1.000000,VelocityScaleZ=0.700000,DampingAlpha=0.000000,MaxLinearVelocity=10.000000,MaxAngularVelocity=15.000000,MaxLinearAcceleration=300.000000,MaxAngularAcceleration=90.000000,ExternalLinearDragV=(X=0.000000,Y=0.200000,Z=0.000000)))),Emote_or_Melee=(RigidBodyAnimNode=(SimSpaceSettings=(WorldAlpha=1.000000,DampingAlpha=0.000000,MaxLinearVelocity=15.000000,MaxAngularVelocity=30.000000,MaxLinearAcceleration=600.000000,MaxAngularAcceleration=150.000000,ExternalLinearDragV=(X=0.000000,Y=0.500000,Z=0.000000)))),Swimming=(RigidBodyAnimNode=(SimSpaceSettings=(WorldAlpha=1.000000,DampingAlpha=0.000000,MaxLinearVelocity=15.000000,MaxAngularVelocity=30.000000,MaxLinearAcceleration=600.000000,MaxAngularAcceleration=150.000000,ExternalLinearDragV=(X=0.000000,Y=0.500000,Z=0.000000)))),Front_End=(RigidBodyAnimNode=(SimSpaceSettings=(WorldAlpha=0.800000,DampingAlpha=0.000000,MaxLinearVelocity=12.000000,MaxAngularVelocity=25.000000,MaxLinearAcceleration=500.000000,MaxAngularAcceleration=120.000000,ExternalLinearDragV=(X=0.000000,Y=0.500000,Z=0.000000)))),DBNO=(RigidBodyAnimNode=(SimSpaceSettings=(WorldAlpha=1.000000,DampingAlpha=0.000000,MaxLinearVelocity=15.000000,MaxAngularVelocity=30.000000,MaxLinearAcceleration=600.000000,MaxAngularAcceleration=150.000000,ExternalLinearDragV=(X=0.000000,Y=0.500000,Z=0.000000)))),DBNO_Carried=(RigidBodyAnimNode=(SimSpaceSettings=(WorldAlpha=1.000000,DampingAlpha=0.000000,MaxLinearVelocity=15.000000,MaxAngularVelocity=30.000000,MaxLinearAcceleration=600.000000,MaxAngularAcceleration=150.000000,ExternalLinearDragV=(X=0.000000,Y=0.500000,Z=0.000000)))),OnGround_Falling=(RigidBodyAnimNode=(SimSpaceSettings=(WorldAlpha=1.000000,DampingAlpha=0.000000,MaxLinearVelocity=15.000000,MaxAngularVelocity=30.000000,MaxLinearAcceleration=600.000000,MaxAngularAcceleration=150.000000,ExternalLinearDragV=(X=0.000000,Y=0.500000,Z=0.000000)))),Ziplining=(RigidBodyAnimNode=(SimSpaceSettings=(WorldAlpha=1.000000,DampingAlpha=0.000000,MaxLinearVelocity=15.000000,MaxAngularVelocity=30.000000,MaxLinearAcceleration=600.000000,MaxAngularAcceleration=150.000000,ExternalLinearDragV=(X=0.000000,Y=0.500000,Z=0.000000)))),Skydive_Surfing=(RigidBodyAnimNode=(SimSpaceSettings=(WorldAlpha=1.000000,DampingAlpha=0.000000,MaxLinearVelocity=15.000000,MaxAngularVelocity=30.000000,MaxLinearAcceleration=600.000000,MaxAngularAcceleration=150.000000,ExternalLinearDragV=(X=0.000000,Y=0.500000,Z=0.000000)))),Riding_Standing=(RigidBodyAnimNode=(SimSpaceSettings=(WorldAlpha=1.000000,DampingAlpha=0.000000,MaxLinearVelocity=15.000000,MaxAngularVelocity=30.000000,MaxLinearAcceleration=600.000000,MaxAngularAcceleration=150.000000,ExternalLinearDragV=(X=0.000000,Y=0.500000,Z=0.000000)))),Riding_Moving=(RigidBodyAnimNode=(SimSpaceSettings=(WorldAlpha=1.000000,DampingAlpha=0.000000,MaxLinearVelocity=15.000000,MaxAngularVelocity=30.000000,MaxLinearAcceleration=600.000000,MaxAngularAcceleration=150.000000,ExternalLinearDragV=(X=0.000000,Y=0.500000,Z=0.000000)))),Riding_Moving_Targeting=(RigidBodyAnimNode=(SimSpaceSettings=(WorldAlpha=1.000000,DampingAlpha=0.000000,MaxLinearVelocity=15.000000,MaxAngularVelocity=30.000000,MaxLinearAcceleration=600.000000,MaxAngularAcceleration=150.000000,ExternalLinearDragV=(X=0.000000,Y=0.500000,Z=0.000000)))),Tactical_Sprint=(RigidBodyAnimNode=(SimSpaceSettings=(WorldAlpha=1.000000,DampingAlpha=0.000000,MaxLinearVelocity=15.000000,MaxAngularVelocity=30.000000,MaxLinearAcceleration=600.000000,MaxAngularAcceleration=150.000000,ExternalLinearDragV=(X=0.000000,Y=0.500000,Z=0.000000)))),Rail_Grinding=(RigidBodyAnimNode=(SimSpaceSettings=(WorldAlpha=0.100000,DampingAlpha=0.000000,MaxLinearVelocity=15.000000,MaxAngularVelocity=30.000000,MaxLinearAcceleration=600.000000,MaxAngularAcceleration=150.000000,ExternalLinearDragV=(X=0.000000,Y=0.500000,Z=0.000000)))),)
 }
 RBAN Settings Out {type=FortRigidBodyAnimNodeParameters}
]
