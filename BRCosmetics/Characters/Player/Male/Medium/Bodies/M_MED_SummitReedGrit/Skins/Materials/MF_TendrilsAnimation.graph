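# Material function that animates tendrils on the SummitReedGrit body skin by wobbling the UVs of an FX mask and outputting its green channel.
# Three rotated, tiled, panning noise layers are blended by the mask's red and blue channels, then remapped to a signed offset.
# The InPutAudio input scales distortion strength, and a blend mask channel suppresses it locally.
# The offset acts mainly along U, with a fraction along V.
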MaterialFunction /BRCosmetics/Characters/Player/Male/Medium/Bodies/M_MED_SummitReedGrit/Skins/Materials/MF_TendrilsAnimation
graph {nodes [
  Constant() {r=30}
  n1 = MaterialFunctionCall(
   UVs=TextureCoordinate() {uTiling=0.1, vTiling=0.5}
   "Rotation Angle (0-1)"=ScalarParameter() {group=TendrilsAnimation, parameterName=UV_Mask_Default_Rotation}
  ) {materialFunction=&CustomRotator}
  n2 = ScalarParameter() {defaultValue=25.5725, group=TendrilsAnimation, parameterName=UV_Mask_Default_Tiling}
  n3 = AppendVector(
   A=ScalarParameter() {group=TendrilsAnimation, parameterName=Speedx}
   B=ScalarParameter() {defaultValue=0.2, group=TendrilsAnimation, parameterName=SpeedY}
  )
  n4 = TextureSampleParameter2D(UVs=Panner(Coordinate=Multiply(A=@n1.Rotated Values, B=@n2) {constB=30}, Speed=@n3) {speedY=0.2}) {
   group=TendrilsAnimation
   parameterName=TendrilsAnimationNoise
   samplerType=SAMPLERTYPE_Grayscale
   texture=/Game/Global/Textures/Default/Noise/LowResBlurredNoiseNormalized
  }
  n5 = TextureSampleParameter2D() {
   group=TendrilsAnimation
   parameterName=TendrilsAnimationMask
   samplerType=SAMPLERTYPE_LinearColor
   texture=/BRCosmetics/Characters/Player/Male/Medium/Bodies/M_MED_SummitReedGrit/Skins/Textures/T_SummitReedGrit_Body_FX2
  }
  n6 = LinearInterpolate(A=@n4.R, B=NamedRerouteUsage(), Alpha=@n5.R)
  n7 = NamedRerouteUsage()
  n8 = ScalarParameter() {defaultValue=0.003, group=TendrilsAnimation, parameterName=UV_Distortion_intensity}
  n9 = FunctionInput() {bUsePreviewValueAsDefault=true, inputName=InPutAudio}
  n10 = TextureSampleParameter2D() {
   group=TendrilsAnimation
   parameterName=TendrilsAnimationBlendMask
   samplerType=SAMPLERTYPE_LinearColor
   texture=/BRCosmetics/Characters/Player/Male/Medium/Bodies/M_MED_SummitReedGrit/Skins/Textures/T_summitreedgrit_mask_Body_Fx3
  }
  n11 = Multiply(
   A=LinearInterpolate(Alpha=LinearInterpolate(A=@n6, B=@n7, Alpha=@n5.B)) {constA=-1}
   B=LinearInterpolate(
    A=LinearInterpolate(B=@n8, Alpha=@n9)
    B=Constant()
    Alpha=StaticComponentMaskParameter(None=@n10.RGB) {defaultR=true, group=TendrilsAnimation, parameterName=TendrilsAnimationBlendMaskChannel}
   )
  ) {constB=0.002}
  n12 = Multiply(A=@n11) {constB=0.1}
  n13 = TextureSampleParameter2D(UVs=Add(A=TextureCoordinate(), B=AppendVector(A=@n11, B=@n12))) {
   group=TendrilsAnimation
   parameterName=TendrilsAnimationMask
   samplerType=SAMPLERTYPE_LinearColor
   texture=&T_SummitReedGrit_Body_FX2
  }
  FunctionOutput(None=@n13.G)
  n14 = Multiply(
   A=MaterialFunctionCall(UVs=TextureCoordinate(), "Rotation Angle (0-1)"=Constant() {r=0.058}) {materialFunction=&CustomRotator}.Rotated Values
   B=AppendVector(A=Constant() {r=0.01}, B=Constant() {r=0.5})
  )
  n15 = ScalarParameter() {defaultValue=30, group=TendrilsAnimation, parameterName=UV_Mask_B_Tiling}
  n16 = TextureSampleParameter2D(UVs=Panner(Coordinate=Multiply(A=@n14, B=@n15) {constB=30}) {speedY=-0.2}) {
   group=TendrilsAnimation
   parameterName=TendrilsAnimationNoise
   samplerType=SAMPLERTYPE_Grayscale
   texture=&LowResBlurredNoiseNormalized
  }
  NamedRerouteDeclaration(None=@n16.R) {name=UV_Mask_B, nodeColor=<rgba 1 0.5391 0 1>}
  n17 = Multiply(
   A=MaterialFunctionCall(
    UVs=TextureCoordinate()
    "Rotation Angle (0-1)"=ScalarParameter() {defaultValue=0.3, group=TendrilsAnimation, parameterName=UV_Mask_R_Rotation}
   ) {materialFunction=&CustomRotator}.Rotated Values
   B=AppendVector(A=Constant() {r=0.01}, B=Constant() {r=0.5})
  )
  n18 = ScalarParameter() {defaultValue=30, group=TendrilsAnimation, parameterName=UV_Mask_R_Tiling}
  n19 = TextureSampleParameter2D(UVs=Panner(Coordinate=Multiply(A=@n17, B=@n18) {constB=30}, Speed=@n3) {speedY=0.2}) {
   group=TendrilsAnimation
   parameterName=TendrilsAnimationNoise
   samplerType=SAMPLERTYPE_Grayscale
   texture=&LowResBlurredNoiseNormalized
  }
  NamedRerouteDeclaration(None=@n19.R) {name=UV_Mask_R, nodeColor=<rgba 0.2188 0 1 1>}
 ]}
