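# Material function for the SummitReedGrit character skin that outputs an animated tentacle mask.
# A pre-skinned local-position sphere mask separates front from back, and that mask picks root positions for two projected, rotated UV sets.
# Three parameterized MF_SummitReedTentacle layers supply body masks, while time-driven sines through SphereMasks draw two head tentacles, one with an outline.
# Named reroutes carry these masks into a final Max, Multiply and Saturate composite.
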MaterialFunction /BRCosmetics/Characters/Player/Male/Medium/Bodies/M_MED_SummitReedGrit/Skins/Materials/MF_SummitReedTentacle/MF_SummitReedTentacle_AnimMask
graph {nodes [
  n1 = NamedRerouteUsage()
  n2 = NamedRerouteUsage()
  n3 = Add(A=Saturate(None=Power(Base=@n1) {constExponent=5}), B=Max(A=Power(Base=@n2), B=NamedRerouteUsage()))
  n4 = OneMinus(None=NamedRerouteUsage())
  n5 = Max(A=Max(A=NamedRerouteUsage(), B=NamedRerouteUsage()), B=Power(Base=Multiply(A=@n3, B=@n4)))
  n6 = NamedRerouteUsage()
  FunctionOutput(None=Saturate(None=Multiply(A=@n5, B=@n6)))
  n7 = Constant4Vector() {constant=<rgba 0.1713 0 163 1>}
  n8 = MaterialFunctionCall!(
   Root_Position=If(A=NamedRerouteUsage(), "A > B"=@n7.RGB, "A == B"=@n7.RGB, "A < B"=Constant4Vector() {constant=<rgba -0.3447 4.9093 162 1>}.RGB) {constB=0.05}
   UseClamp?=Constant() {r=1}
   FlipUV=Constant()
   UVSize=Constant() {r=6.0832}
   CenterOffsetX=Constant()
   CenterOffsetY=Constant()
   CenterOffsetZ=Constant()
  ) {materialFunction=/BRCosmetics/Gadgets/Assets/VinderTech_GliderChute/Glider_WiryPerk/Materials/MF_ProjectUV_WiryPerk}
  n9 = NamedRerouteUsage()
  n10 = Constant() {r=-0.105}
  n11 = Constant() {r=-0.09}
  MaterialExpressionNamedRerouteDeclaration_8 = NamedRerouteDeclaration(None=MaterialFunctionCall(
    UVs=NamedRerouteDeclaration(None=@n8.UV_BackProjection) {name=UVBasic_1, nodeColor=<rgba 0.5703 0 1 1>}
    "Rotation Angle (0-1)"=If(A=@n9, "A > B"=@n10, "A == B"=@n10, "A < B"=@n11) {constB=0.05}
   ) {materialFunction=&CustomRotator}.Rotated Values) {name=ProjectUV_1, nodeColor=<rgba 1 0.0234 0 1>}
  n12 = Convert(None=@MaterialExpressionNamedRerouteDeclaration_8) {
   nodeName=Break Float2
   convertInputs [{
     type=Vector2
     expressionInput {
      expression=&self:MaterialExpressionNamedRerouteDeclaration_8
      inputName=None
      mask=0
      maskA=0
      maskB=0
      maskG=0
      maskR=0
      outputIndex=0
     }
    }]
   convertOutputs [{type=Scalar}, {type=Scalar}]
  }
  NamedRerouteDeclaration(None=@n12) {name=AnimUV_G_1, nodeColor=<rgba 0.125 0 1 1>}
  n13 = SphereMask(A=ComponentMask(None=NamedRerouteUsage()) {g=true}, B=Constant(), Radius=Constant() {r=0.942}) {attenuationRadius=0.1}
  n14 = Time()
  n15 = Constant() {r=-0.6}
  n16 = Constant() {r=1.5231}
  n17 = NamedRerouteUsage()
  n18 = Sine(None=Subtract(A=Multiply(A=@n14, B=@n15), B=Multiply(A=@n16, B=@n17)))
  n19 = NamedRerouteUsage()
  n20 = NamedRerouteUsage()
  n21 = Constant() {r=0.5}
  n22 = Add(A=Multiply(A=Multiply(A=@n18, B=@n19), B=Constant() {r=0.25}), B=Multiply(A=Subtract(A=@n20, B=@n21)) {constB=8})
  n23 = LinearInterpolate(A=Constant() {r=-0.1051}, B=Constant() {r=0.552}, Alpha=NamedRerouteUsage())
  n24 = SphereMask(A=@n22, B=Constant(), Radius=@n23)
  NamedRerouteDeclaration(None=Multiply(A=@n13, B=Multiply(A=@n24))) {name=HeadTentacles2, nodeColor=<rgba 0 0.2969 1 1>}
  n25 = Constant() {r=0.1}
  n26 = Subtract(A=@n24, B=SphereMask(A=@n22, B=Constant(), Radius=Subtract(A=@n23, B=@n25)))
  NamedRerouteDeclaration(None=OneMinus(None=Multiply(A=@n13, B=@n26))) {name=HeadTentacles2_Outline, nodeColor=<rgba 1 0 0.2344 1>}
  n27 = MaterialFunctionCall!(
   Thickness and fineness B=ScalarParameter() {defaultValue=3.6, group=Tentacle, parameterName=L1_Thickness and fineness B}
   Amplitude boost
   SpeedA=ScalarParameter() {defaultValue=-0.41, group=Tentacle, parameterName=L1_SpeedA}
   SpeedB=ScalarParameter() {defaultValue=-0.32, group=Tentacle, parameterName=L1_SpeedB}
   FrequencyA=ScalarParameter() {defaultValue=2.4, group=Tentacle, parameterName=L1_FrequencyA}
   FrequencyB=ScalarParameter() {defaultValue=1.9, group=Tentacle, parameterName=L1_FrequencyB}
   Thickness and fineness A=ScalarParameter() {defaultValue=3.2, group=Tentacle, parameterName=L1_Thickness and fineness A}
   LayerInput=Constant() {r=1}
   Thickness and fineness C=ScalarParameter() {defaultValue=0.75, group=Tentacle, parameterName=L1_Thickness and fineness C}
   LengthA=ScalarParameter() {defaultValue=-0.2, group=Tentacle, parameterName=L1_LengthA}
   LengthB=ScalarParameter() {defaultValue=-0.25, group=Tentacle, parameterName=L1_LengthB}
   LayerOutline=ScalarParameter() {defaultValue=0.02, group=Tentacle, parameterName=L1_LayerOutline}
   AmplitudeA=ScalarParameter() {defaultValue=0.61, group=Tentacle, parameterName=L1_AmplitudeA}
   AmplitudeB=ScalarParameter() {defaultValue=0.35, group=Tentacle, parameterName=L1_AmplitudeB}
   SpeedBoost=ScalarParameter() {defaultValue=1, group=Tentacle, parameterName=L1_SpeedBoost}
  ) {materialFunction=&/MF_SummitReedTentacle}
  NamedRerouteDeclaration(None=@n27.Mask) {name=Layer1, nodeColor=<rgba 0 0.625 1 1>}
  n28 = MaterialFunctionCall(
   Thickness and fineness B=ScalarParameter() {defaultValue=3.6, group=Tentacle, parameterName=L2_Thickness and fineness B}
   SpeedA=ScalarParameter() {defaultValue=-0.41, group=Tentacle, parameterName=L2_SpeedA}
   SpeedB=ScalarParameter() {defaultValue=-0.32, group=Tentacle, parameterName=L2_SpeedB}
   FrequencyA=ScalarParameter() {defaultValue=2.4, group=Tentacle, parameterName=L2_FrequencyA}
   FrequencyB=ScalarParameter() {defaultValue=1.9, group=Tentacle, parameterName=L2_FrequencyB}
   Thickness and fineness A=ScalarParameter() {defaultValue=3.2, group=Tentacle, parameterName=L2_Thickness and fineness A}
   LayerInput=Constant() {r=2}
   Thickness and fineness C=ScalarParameter() {defaultValue=0.75, group=Tentacle, parameterName=L2_Thickness and fineness C}
   LengthA=ScalarParameter() {defaultValue=-0.2, group=Tentacle, parameterName=L2_LengthA}
   LengthB=ScalarParameter() {defaultValue=-0.25, group=Tentacle, parameterName=L2_LengthB}
   LayerOutline=ScalarParameter() {group=Tentacle, parameterName=L2_LayerOutline}
   AmplitudeA=ScalarParameter() {defaultValue=0.61, group=Tentacle, parameterName=L2_AmplitudeA}
   AmplitudeB=ScalarParameter() {defaultValue=0.35, group=Tentacle, parameterName=L2_AmplitudeB}
   SpeedBoost=ScalarParameter() {defaultValue=1, group=Tentacle, parameterName=L2_SpeedBoost}
  ) {materialFunction=&MF_SummitReedTentacle}
  NamedRerouteDeclaration(None=@n28.Mask) {name=Layer2, nodeColor=<rgba 0.4766 1 0 1>}
  n29 = MaterialFunctionCall(
   Thickness and fineness B=ScalarParameter() {defaultValue=3.6, group=Tentacle, parameterName=L3_Thickness and fineness B}
   SpeedA=ScalarParameter() {defaultValue=-0.35, group=Tentacle, parameterName=L3_SpeedA}
   SpeedB=ScalarParameter() {defaultValue=-0.29, group=Tentacle, parameterName=L3_SpeedB}
   FrequencyA=ScalarParameter() {defaultValue=3, group=Tentacle, parameterName=L3_FrequencyA}
   FrequencyB=ScalarParameter() {defaultValue=1.9, group=Tentacle, parameterName=L3_FrequencyB}
   Thickness and fineness A=ScalarParameter() {defaultValue=1.2, group=Tentacle, parameterName=L3_Thickness and fineness A}
   LayerInput=Constant() {r=3}
   Thickness and fineness C=ScalarParameter() {defaultValue=0.75, group=Tentacle, parameterName=L3_Thickness and fineness C}
   LengthA=ScalarParameter() {defaultValue=-0.1, group=Tentacle, parameterName=L3_LengthA}
   LengthB=ScalarParameter() {defaultValue=-0.2, group=Tentacle, parameterName=L3_LengthB}
   LayerOutline=ScalarParameter() {defaultValue=0.2, group=Tentacle, parameterName=L3_LayerOutline}
   AmplitudeA=ScalarParameter() {defaultValue=0.35, group=Tentacle, parameterName=L3_AmplitudeA}
   AmplitudeB=ScalarParameter() {defaultValue=0.21, group=Tentacle, parameterName=L3_AmplitudeB}
   SpeedBoost=ScalarParameter() {defaultValue=1, group=Tentacle, parameterName=L3_SpeedBoost}
  ) {materialFunction=&MF_SummitReedTentacle}
  NamedRerouteDeclaration(None=@n29.Mask) {name=Layer3, nodeColor=<rgba 0.4766 1 0 1>}
  NamedRerouteDeclaration(None=@n29.Outline) {name=Layer3Outline, nodeColor=<rgba 0.4766 1 0 1>}
  n30 = SphereMask(
   A=VertexInterpolator(VS=LocalPosition() {localOrigin=InstancePreSkinning}.XYZ).PS
   B=Constant4Vector() {constant=<rgba 0.1713 4.9093 163 1>}.RGB
   Radius=Constant() {r=7.2324}
  ) {attenuationRadius=0.1}
  n31 = Constant4Vector() {constant=<rgba 1.0437 0 163 1>}
  n32 = MaterialFunctionCall(
   Root_Position=If(
    A=NamedRerouteDeclaration(None=@n30) {name=Front and Back Masks, nodeColor=<rgba 0.8047 0 1 1>}
    "A > B"=@n31.RGB
    "A == B"=@n31.RGB
    "A < B"=Constant4Vector() {constant=<rgba -1.3632 4.9093 160 1>}.RGB
   ) {constB=0.05}
   UseClamp?=Constant() {r=1}
   FlipUV=Constant()
   UVSize=ScalarParameter() {defaultValue=6.0832, group=Tentacle, parameterName=UVSizerw}
   CenterOffsetX=Constant()
   CenterOffsetY=Constant()
   CenterOffsetZ=Constant()
  ) {materialFunction=&MF_ProjectUV_WiryPerk}
  n33 = NamedRerouteUsage()
  n34 = Constant() {r=0.055}
  n35 = Constant() {r=-0.058}
  MaterialExpressionNamedRerouteDeclaration_2 = NamedRerouteDeclaration(None=MaterialFunctionCall(
    UVs=NamedRerouteDeclaration(None=@n32.UV_BackProjection) {name=UVBasic, nodeColor=<rgba 0.5703 0 1 1>}
    "Rotation Angle (0-1)"=If(A=@n33, "A > B"=@n34, "A == B"=@n34, "A < B"=@n35) {constB=0.05}
   ) {materialFunction=&CustomRotator}.Rotated Values) {name=ProjectUV, nodeColor=<rgba 1 0.0234 0 1>}
  n36 = Convert(None=@MaterialExpressionNamedRerouteDeclaration_2) {
   nodeName=Break Float2
   convertInputs [{
     type=Vector2
     expressionInput {
      expression=&self:MaterialExpressionNamedRerouteDeclaration_2
      inputName=None
      mask=0
      maskA=0
      maskB=0
      maskG=0
      maskR=0
      outputIndex=0
     }
    }]
   convertOutputs [{type=Scalar}, {type=Scalar}]
  }
  NamedRerouteDeclaration(None=@n36) {name=AnimUV_R, nodeColor=<rgba 1 0.4453 0 1>}
  NamedRerouteDeclaration(None=@n36) {name=AnimUV_G, nodeColor=<rgba 0.125 0 1 1>}
  n37 = ComponentMask(None=NamedRerouteUsage()) {g=true}
  n38 = Constant()
  n39 = Constant() {r=0.942}
  n40 = If(A=NamedRerouteUsage(), "A > B"=@n39, "A == B"=@n39, "A < B"=Constant() {r=0.612}) {constB=0.05}
  n41 = Multiply(A=Time(), B=Constant() {r=-0.5})
  n42 = Multiply(A=Constant() {r=1}, B=NamedRerouteUsage())
  n43 = Multiply(A=Sine(None=Subtract(A=@n41, B=@n42)), B=NamedRerouteUsage())
  n44 = Constant() {r=0.31}
  n45 = Subtract(A=NamedRerouteUsage(), B=Constant() {r=0.5})
  n46 = SphereMask(
   A=Add(A=Multiply(A=@n43, B=@n44), B=Multiply(A=@n45) {constB=8})
   B=Constant()
   Radius=LinearInterpolate(A=Constant() {r=-0.1051}, B=Constant() {r=0.6}, Alpha=NamedRerouteUsage())
  )
  NamedRerouteDeclaration(None=Multiply(A=SphereMask(A=@n37, B=@n38, Radius=@n40) {attenuationRadius=0.1}, B=Multiply(A=@n46))) {name=HeadTentacles 1, nodeColor=<rgba 0 0.2969 1 1>}
  NamedRerouteDeclaration(None=@n12) {name=AnimUV_R_1, nodeColor=<rgba 1 0.4453 0 1>}
 ]}
