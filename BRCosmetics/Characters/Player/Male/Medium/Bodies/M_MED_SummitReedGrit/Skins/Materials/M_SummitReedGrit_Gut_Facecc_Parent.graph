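# Parent skin material for the SummitReedGrit character's Gut face, adding animated glowing tendrils on skeletal meshes.
# Between the CharacterParent PreFX and PostFX functions, it overrides Base Color and Emissive through named reroutes.
# A tentacle animation mask tinted by TendrilsColor builds the glow, while a masked multiply darkens base color.
# An audio-reactivity function supplies amplitude for threshold-gated audio values exposed as reroutes.
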Material /BRCosmetics/Characters/Player/Male/Medium/Bodies/M_MED_SummitReedGrit/Skins/Materials/M_SummitReedGrit_Gut_Facecc_Parent
graph {
 properties {
  bUseMaterialAttributes=true
  bUsedWithMeshDeformer=true
  bUsedWithMorphTargets=true
  bUsedWithSkeletalMesh=true
  pixelDepthOffsetMode=PDOM_Legacy
  shadingModel=MSM_FromMaterialExpression
  usedShadingModels=Default Lit | Subsurface | Clear Coat | Subsurface Profile | Cloth
 }
 nodes [
  n1 = GetMaterialAttributes(None=NamedRerouteUsage()) {attributeGetTypes [#1, #2]}
  n2 = NamedRerouteUsage()
  n3 = NamedRerouteUsage()
  n4 = NamedRerouteUsage()
  n5 = MaterialFunctionCall() {materialFunction=&MF_CharacterParent_2023_PreFX}
  MaterialFunctionCall(
   Material=SetMaterialAttributes!(MaterialAttributes=@n1.MaterialAttributes, Emissive Color=LinearInterpolate(A=@n2, B=@n3, Alpha=@n4), Base Color=NamedRerouteUsage()) {attributeSetTypes [#2, #1]}
   M=@n5.M
   Pre Skinned Local Position=@n5.M
   Distance Blend=@n5.M
   Fresnel=@n5.M
  ) {materialFunction=&MF_CharacterParent_2023_PostFX}
  NamedRerouteDeclaration(None=@n5.Result) {name=InputMA, nodeColor=<rgba 1 0 0.375 1>}
  n6 = ScalarParameter() {defaultValue=0.2, parameterName=AudioThreshold}
  n7 = MaterialFunctionCall!(InAmplitude, InSpectrum, CustomFakeSpectrum, CustomFakeAmplitude) {materialFunction=/BRCosmetics/Global/MaterialFunctions/AudioReactivity/MF_AudioReactivity_FrontendLoop}
  n8 = Constant() {r=1}
  n9 = If(A=@n7.Amplitude, "A > B"=@n8, "A == B"=@n8, "A < B"=Constant()) {constB=0.01}
  NamedRerouteDeclaration(None=LinearInterpolate(A=Multiply(A=@n6, B=@n9), Alpha=Power(Base=@n7.Amplitude)) {constA=0.2}) {name=InputAudio, nodeColor=<rgba 0.6641 0 1 1>}
  NamedRerouteDeclaration(None=@n1.Base Color) {name=InputBaseColor, nodeColor=<rgba 1 0.4219 0 1>}
  n10 = Multiply(A=NamedRerouteUsage(), B=NamedRerouteUsage())
  NamedRerouteDeclaration(None=Multiply(A=NamedRerouteUsage(), B=LinearInterpolate(Alpha=@n10) {constA=1, constB=0})) {name=OutPutBaseColor, nodeColor=<rgba 0.2187 1 0 1>}
  NamedRerouteDeclaration(None=@n1.Emissive Color) {name=InputEmissiveColor, nodeColor=<rgba 1 0.3281 0 1>}
  n11 = TextureSampleParameter2D() {group=TendrilsAnimation, parameterName=TendrilsEmissiveColor, texture=&T_BlacksRGB}
  NamedRerouteDeclaration(None=Max(A=@n11.R, B=NamedRerouteUsage())) {name=EmissiveR, nodeColor=<rgba 0.6172 0 1 1>}
  NamedRerouteDeclaration(None=@n11.RGB) {name=EmissiveRGB, nodeColor=<rgba 0 1 0.2969 1>}
  n12 = MaterialFunctionCall() {materialFunction=/BRCosmetics/Characters/Player/Male/Medium/Bodies/M_MED_SummitReedGrit/Skins/Materials/MF_SummitReedTentacle/MF_SummitReedTentacle_AnimMask}
  n13 = Add(
   A=Multiply(
    A=NamedRerouteDeclaration(None=@n12.Result) {name=TendrilsMask, nodeColor=<rgba 1 0 0.4922 1>}
    B=VectorParameter() {group=TendrilsAnimation, parameterName=TendrilsColor, defaultValue=<rgba 1 0.015 0.0125 1>}.RGB
   )
   B=NamedRerouteUsage()
  )
  n14 = NamedRerouteUsage()
  n15 = LinearInterpolate(Alpha=@n14) {constB=2}
  NamedRerouteDeclaration(None=Multiply(A=Multiply(A=@n13, B=@n15), B=@n14)) {name=OutPutEmissiveColor, nodeColor=<rgba 0 1 0.6953 1>}
  NamedRerouteDeclaration(None=LinearInterpolate(B=Constant() {r=0.55}, Alpha=@n9)) {name=Audio switch, nodeColor=<rgba 0 1 0.7422 1>}
 ]
}
